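# Control Rig for the F_MED_ToadLeaf female body that applies procedural corrective posing to its skirt, cape, knife, rope, sleeve, glute and foot bones.
# After a shared DeformRig function, a sequence derives stance weights from pelvis height, then layers parent constraints and additive bone offsets.
# Spherical pose readers on thighs and upper arms, plus remapped joint angles, weight those offsets.
# PlayingEmote, NotRban, DBNO, Gauntlets, FrontEndOn and high velocity gate situational fixes.
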ControlRigBlueprint /BRCosmetics/Characters/Player/Female/Medium/Bodies/F_MED_ToadLeaf/Meshes/F_MED_ToadLeaf_CtrlRig
graphs [
 RigVMFunctionLibrary {links [], nodes []}
 RigVMFunction_Sequence_1_ContainedGraph {nodes [
   n1 = RigVMFunctionEntryNode()
   RigVMFunction_Sequence = &RigVMFunction_Sequence(ExecuteContext=@n1.ExecuteContext)
   n2 = &RigVMFunction_Sequence(ExecuteContext=@RigVMFunction_Sequence.B)
   n3 = &RigVMFunction_Sequence(ExecuteContext=@n2.B)
   n4 = &RigVMFunction_Sequence(ExecuteContext=@n3.B)
   n5 = &RigVMFunction_Sequence(ExecuteContext=@n4.B)
   n6 = &RigVMFunction_Sequence(ExecuteContext=@n5.B)
   n7 = &RigVMFunction_Sequence(ExecuteContext=@n6.B)
   n8 = &RigVMFunction_Sequence(ExecuteContext=@n7.B)
   n9 = &RigVMFunction_Sequence(ExecuteContext=@n8.B)
   n10 = &RigVMFunction_Sequence(ExecuteContext=@n9.B)
   RigVMFunctionReturnNode(A=@RigVMFunction_Sequence.A, B=@n2.A, C=@n3.A, D=@n4.A, E=@n5.A, F=@n6.A, G=@n7.A, H=@n8.A, I=@n9.A, J=@n10.A, K=@n10.B)
  ]}
 RigVMModel {nodes [
   RigUnit_BeginExecution = &RigUnit_BeginExecution()
   FN_DeformRig_FNC = RigVMFunctionReferenceNode(
    DeformRig_ Curve="(deformRig_l_upperarm_up_alpha_72_#1=False,deformRig_r_upperarm_up_alpha_73_#2=False,deformRig_l_upperarm_fr_alpha_74_#3=False,deformRig_r_upperarm_fr_alpha_75_CB51B8 ...(truncated)"
    DeformRig_Bool="(upperarm_1_#4=True,elbow_3_#5=True,wrist_7_#6=True,groin_8_#7=True,glute_17_#8=True,knee_9_40C408324CA261 ...(truncated)"
    DeformRig_Debug=(upperarm_debug_18_#4=False,wrist_debug_20_#6=False,lat_debug_26_#9=False,groinGlute_debug_27_#7=False)
    DeformRig_Value="(upperarm_up_18_#10=(Rotation=(X=0,Y=0,Z=0,W=1),Translation=(X=0,Y=0,Z=8),Scale3D=(X=0.7,Y=1,Z=1)),upperarm_fwd_fr_19_#11=(Rotation=(X=0,Y=0,Z=0,W=1),Translation=(X=0,Y=10,Z=0),Scale3D=(X=0.5,Y=1,Z= ...(truncated)"
    ExecuteContext=@RigUnit_BeginExecution.ExecutePin
   )
   > RigVMFunction_Sequence = RigVMAggregateNode()
   Remap = &RigVMFunction_MathDoubleRemap(
    SourceMaximum=45
    SourceMinimum=55
    TargetMaximum=1
    TargetMinimum=0
    Value=&RigUnit_GetRelativeTransformForItem(Child=(Type=Bone,Name="pelvis"), Parent=(Type=Bone,Name="root"), bChildInitial=False, bParentInitial=False).RelativeTransform.Translation.Z
   )
   RigVMVariableNode(Value=@Remap.Result, ExecuteContext=@RigVMFunction_Sequence.A) {variable=Sit Lower}
   Multiply = &RigVMFunction_MathDoubleMul(
    B=&RigVMFunction_MathDoubleRemap(
     SourceMaximum=60
     SourceMinimum=50
     TargetMaximum=1
     TargetMinimum=0
     Value=&RigUnit_GetRelativeTransformForItem(Child=(Type=Bone,Name="pelvis"), Parent=(Type=Bone,Name="root"), bChildInitial=False, bParentInitial=False).RelativeTransform.Translation.Z
    ).Result
    A=&RigVMFunction_MathDoubleRemap(
     SourceMaximum=77
     SourceMinimum=85
     TargetMaximum=1
     TargetMinimum=0
     Value=&RigUnit_GetRelativeTransformForItem(Child=(Type=Bone,Name="pelvis"), Parent=(Type=Bone,Name="root"), bChildInitial=False, bParentInitial=False).RelativeTransform.Translation.Z
    ).Result
   )
   > RigVMVariableNode(Value=@Multiply.Result) {variable=Crouch}
   > RigVMVariableNode(Value=&RigVMFunction_MathDoubleRemap(
     SourceMaximum=85
     SourceMinimum=80
     TargetMaximum=1
     TargetMinimum=0
     Value=&RigUnit_GetRelativeTransformForItem(Child=(Type=Bone,Name="pelvis"), Parent=(Type=Bone,Name="root"), bChildInitial=False, bParentInitial=False).RelativeTransform.Translation.Z
    ).Result) {variable=Stand}
   n1 = &RigUnit_GetRelativeTransformForItem(Child=(Type=Bone,Name="pelvis"), Parent=(Type=Bone,Name="root"), bChildInitial=False, bParentInitial=False)
   > RigVMVariableNode(Value=&RigVMFunction_KalmanFloat(
     BufferSize=16
     Value=&RigVMFunction_MathDoubleRemap(SourceMaximum=80, SourceMinimum=85, TargetMaximum=1, TargetMinimum=0, Value=@n1.RelativeTransform.Translation.Z).Result
    ) {template="KalmanFilter::Execute(in Value,in BufferSize,out Result)"}.Result) {variable=No stand}
   n2 = &RigVMFunction_ControlFlowBranch(ExecuteContext=@RigVMFunction_Sequence.H, Condition=RigVMVariableNode() {variable=PlayingEmote}.Value)
   n3 = &RigUnit_SphericalPoseReader(
    ActiveRegionScaleFactors=(PositiveWidth=1,NegativeWidth=1,PositiveHeight=1,NegativeHeight=1)
    ActiveRegionSize=0.13
    Debug=(bDrawDebug=false,bDraw2D=False,bDrawLocalAxes=False,DebugScale=25,DebugSegments=20,DebugThickness=0.25)
    DriverAxis=(X=1,Y=0,Z=0)
    DriverItem=(Type=Bone,Name="thigh_r")
    FalloffRegionScaleFactors=(PositiveWidth=1,NegativeWidth=1,PositiveHeight=1,NegativeHeight=1)
    FalloffSize=0.26
    FlipHeightScaling=False
    FlipWidthScaling=False
    OptionalParentItem=(Type=Bone,Name="None")
    RotationOffset=(X=0,Y=-30,Z=-15)
    ExecutePin=@n2.True
   )
   > &RigUnit_ModifyTransforms(
    ItemToModify="((Item=(Type=Bone,Name=\"Base_skirt_Side_R\"),Transform=(Rotation=(X=0.0864,Y=-0.13,Z=0.0114,W=0.9877),Translation=(X=0,Y=0,Z=0),Scale3D=(X=1,Y=1,Z=1))),(Item=(Type=Bone,Name=\"Base_skirt_Bk_R\"),Transform=(Rotation=(X=-0.0302,Y=-0.171,Z=-0.171,W=0.9698),Trans ...(truncated)"
    WeightMinimum=0
    Weight=@n3.OutputParam
   )
   > Get Delta Ref = RigVMFunctionReferenceNode(Item=(Type=Bone,Name="thigh_r"))
   > &RigUnit_ModifyTransforms(
    ItemToModify=((Item=(Type=Bone,Name="Base_skirt_Side_R"),Transform=(Rotation=(X=0.0076,Y=-0.0868,Z=0.0868,W=0.9924),Translation=(X=0,Y=0,Z=0),Scale3D=(X=1,Y=1,Z=1))))
    WeightMinimum=0
    Weight=&RigVMFunction_MathDoubleRemap(SourceMaximum=50, SourceMinimum=30, TargetMaximum=1, TargetMinimum=0, Value=@"Get Delta Ref".Result.Rotation.Roll).Result
   )
   > n4 = &RigUnit_SphericalPoseReader(
    ActiveRegionScaleFactors=(PositiveWidth=1,NegativeWidth=1,PositiveHeight=1,NegativeHeight=1)
    ActiveRegionSize=0.13
    Debug=(bDrawDebug=true,bDraw2D=False,bDrawLocalAxes=False,DebugScale=25,DebugSegments=20,DebugThickness=0.25)
    DriverAxis=(X=1,Y=0,Z=0)
    DriverItem=(Type=Bone,Name="thigh_r")
    FalloffRegionScaleFactors=(PositiveWidth=1,NegativeWidth=1,PositiveHeight=1,NegativeHeight=1)
    FalloffSize=0.26
    FlipHeightScaling=False
    FlipWidthScaling=False
    OptionalParentItem=(Type=Bone,Name="None")
    RotationOffset=(X=80,Y=0,Z=-40)
   )
   > &RigUnit_ModifyTransforms(
    ItemToModify=((Item=(Type=Bone,Name="Base_skirt_Bk_R"),Transform=(Rotation=(X=0,Y=-0.2588,Z=0,W=0.9659),Translation=(X=-2,Y=0,Z=3),Scale3D=(X=1,Y=1,Z=1))))
    WeightMinimum=0
    Weight=@n4.OutputParam
   )
   > &RigUnit_ModifyTransforms(
    ItemToModify=((Item=(Type=Bone,Name="Base_skirt_Fr_R"),Transform=(Rotation=(X=0,Y=-0.1736,Z=0,W=0.9848),Translation=(X=0,Y=0,Z=0),Scale3D=(X=1,Y=1,Z=1))))
    WeightMinimum=0
    Weight=@n4.OutputParam
   )
   n5 = &RigVMFunction_ControlFlowBranch(ExecuteContext=@RigVMFunction_Sequence.I, Condition=RigVMVariableNode() {variable=NotRban}.Value)
   &RigUnit_ParentConstraint(
    AdvancedSettings=(InterpolationType=Average,RotationOrderForFilter=XZY)
    Child=(Type=Bone,Name="Base_skirt_Bk_R")
    Filter=(TranslationFilter=(bX=true,bY=true,bZ=true),RotationFilter=(bX=true,bY=true,bZ=true),ScaleFilter=(bX=True,bY=True,bZ=True))
    Parents=((Item=(Type=Bone,Name="thigh_r"),Weight=0.7),(Item=(Type=Bone,Name="pelvis"),Weight=0.3))
    Weight=1
    ExecutePin=@n5.True
   )
   > &RigUnit_ParentConstraint(
    AdvancedSettings=(InterpolationType=Average,RotationOrderForFilter=XZY)
    Child=(Type=Bone,Name="Base_skirt_Bk_L")
    Filter=(TranslationFilter=(bX=true,bY=true,bZ=true),RotationFilter=(bX=True,bY=True,bZ=True),ScaleFilter=(bX=True,bY=True,bZ=True))
    Parents=((Item=(Type=Bone,Name="thigh_l"),Weight=0.7),(Item=(Type=Bone,Name="pelvis"),Weight=0.3))
    Weight=1
   )
   > n6 = &RigUnit_SphericalPoseReader(
    ActiveRegionScaleFactors=(PositiveWidth=1,NegativeWidth=1,PositiveHeight=1,NegativeHeight=1)
    ActiveRegionSize=0.13
    Debug=(bDrawDebug=false,bDraw2D=False,bDrawLocalAxes=False,DebugScale=25,DebugSegments=20,DebugThickness=0.25)
    DriverAxis=(X=1,Y=0,Z=0)
    DriverItem=(Type=Bone,Name="thigh_r")
    FalloffRegionScaleFactors=(PositiveWidth=1,NegativeWidth=1,PositiveHeight=1,NegativeHeight=1)
    FalloffSize=0.26
    FlipHeightScaling=False
    FlipWidthScaling=False
    OptionalParentItem=(Type=Bone,Name="None")
    RotationOffset=(X=80,Y=0,Z=-40)
   )
   > &RigUnit_ModifyTransforms(
    ItemToModify="((Item=(Type=Bone,Name=\"Base_skirt_Bk_R\"),Transform=(Rotation=(X=0,Y=-0.1736,Z=0,W=0.9848),Translation=(X=0,Y=0,Z=0),Scale3D=(X=1,Y=1,Z=1))),(Item=(Type=Bone,Name=\"Base_skirt_Bk_L\"),Transform=(Rotation=(X=-0.0302,Y=-0.171,Z=-0.171,W=0.9698),Translation=(X= ...(truncated)"
    WeightMinimum=0
    Weight=@n6.OutputParam
   )
   > n7 = &RigUnit_SphericalPoseReader(
    ActiveRegionScaleFactors=(PositiveWidth=1,NegativeWidth=1,PositiveHeight=1,NegativeHeight=1)
    ActiveRegionSize=0.13
    Debug=(bDrawDebug=false,bDraw2D=False,bDrawLocalAxes=False,DebugScale=25,DebugSegments=20,DebugThickness=0.25)
    DriverAxis=(X=-1,Y=0,Z=0)
    DriverItem=(Type=Bone,Name="thigh_l")
    FalloffRegionScaleFactors=(PositiveWidth=1,NegativeWidth=1,PositiveHeight=1,NegativeHeight=1)
    FalloffSize=0.26
    FlipHeightScaling=False
    FlipWidthScaling=False
    OptionalParentItem=(Type=Bone,Name="None")
    RotationOffset=(X=100,Y=0,Z=140)
   )
   > &RigUnit_ModifyTransforms(
    ItemToModify="((Item=(Type=Bone,Name=\"Base_skirt_Bk_R\"),Transform=(Rotation=(X=0,Y=-0.1736,Z=0,W=0.9848),Translation=(X=0,Y=0,Z=0),Scale3D=(X=1,Y=1,Z=1))),(Item=(Type=Bone,Name=\"Base_skirt_Bk_L\"),Transform=(Rotation=(X=-0.0302,Y=-0.171,Z=-0.171,W=0.9698),Translation=(X= ...(truncated)"
    WeightMinimum=0
    Weight=@n7.OutputParam
   )
   > n8 = &RigUnit_SphericalPoseReader(
    ActiveRegionScaleFactors=(PositiveWidth=1,NegativeWidth=1,PositiveHeight=1,NegativeHeight=1)
    ActiveRegionSize=0.1
    Debug=(bDrawDebug=true,bDraw2D=False,bDrawLocalAxes=False,DebugScale=25,DebugSegments=20,DebugThickness=0.25)
    DriverAxis=(X=1,Y=0,Z=0)
    DriverItem=(Type=Bone,Name="upperarm_l")
    FalloffRegionScaleFactors=(PositiveWidth=1,NegativeWidth=1,PositiveHeight=1,NegativeHeight=1)
    FalloffSize=0.2
    FlipHeightScaling=False
    FlipWidthScaling=False
    OptionalParentItem=(Type=Bone,Name="None")
    RotationOffset=(X=0,Y=-110,Z=0)
   )
   > &RigUnit_ModifyTransforms(
    ItemToModify=((Item=(Type=Bone,Name="Base_Cape_R"),Transform=(Rotation=(X=0.0114,Y=0.13,Z=-0.0864,W=0.9877),Translation=(X=0,Y=0,Z=0),Scale3D=(X=1,Y=1,Z=1))))
    WeightMinimum=0
    Weight=@n8.OutputParam
   )
   > n9 = &RigUnit_SphericalPoseReader(
    ActiveRegionScaleFactors=(PositiveWidth=1,NegativeWidth=1,PositiveHeight=1,NegativeHeight=1)
    ActiveRegionSize=0.1
    Debug=(bDrawDebug=true,bDraw2D=False,bDrawLocalAxes=False,DebugScale=25,DebugSegments=20,DebugThickness=0.25)
    DriverAxis=(X=-1,Y=0,Z=0)
    DriverItem=(Type=Bone,Name="upperarm_r")
    FalloffRegionScaleFactors=(PositiveWidth=1,NegativeWidth=1,PositiveHeight=1,NegativeHeight=1)
    FalloffSize=0.2
    FlipHeightScaling=False
    FlipWidthScaling=False
    OptionalParentItem=(Type=Bone,Name="None")
    RotationOffset=(X=0,Y=70,Z=0)
   )
   > &RigUnit_ModifyTransforms(
    ItemToModify=((Item=(Type=Bone,Name="Base_Cape_L"),Transform=(Rotation=(X=-0.0114,Y=0.13,Z=0.0864,W=0.9877),Translation=(X=0,Y=0,Z=0),Scale3D=(X=1,Y=1,Z=1))))
    WeightMinimum=0
    Weight=@n9.OutputParam
   )
   > &RigUnit_ModifyTransforms(
    ItemToModify=((Item=(Type=Bone,Name="Base_Cape_C"),Transform=(Rotation=(X=0,Y=-0.0872,Z=0,W=0.9962),Translation=(X=0,Y=0,Z=0),Scale3D=(X=1,Y=1,Z=1))))
    WeightMinimum=0
    Weight=&RigVMFunction_MathDoubleMul(A=@n9.OutputParam, B=@n8.OutputParam).Result
   )
   n10 = &RigVMFunction_ControlFlowBranch(ExecuteContext=@RigVMFunction_Sequence.J, Condition=RigVMVariableNode() {variable=Gauntlets}.Value)
   &RigUnit_ModifyTransforms(
    ItemToModify="((Item=(Type=Bone,Name=\"Dyn_sleeve_Down_R\"),Transform=(Rotation=(X=0,Y=0,Z=0,W=1),Translation=(X=0,Y=0,Z=0),Scale3D=(X=0,Y=0,Z=0))),(Item=(Type=Bone,Name=\"Dyn_sleeve_In_R\"),Transform=(Rotation=(X=0,Y=0,Z=0,W=1),Translation=(X=0,Y=0,Z=0),Scale3D=(X=0,Y=0,Z= ...(truncated)"
    Weight=1
    WeightMinimum=0
    ExecutePin=@n10.True
   )
   > &RigUnit_ModifyTransforms(
    ItemToModify="((Item=(Type=Bone,Name=\"Dyn_sleeve_Down_L\"),Transform=(Rotation=(X=0,Y=0,Z=0,W=1),Translation=(X=0,Y=0,Z=0),Scale3D=(X=0,Y=0,Z=0))),(Item=(Type=Bone,Name=\"Dyn_sleeve_In_L\"),Transform=(Rotation=(X=0,Y=0,Z=0,W=1),Translation=(X=0,Y=0,Z=0),Scale3D=(X=0,Y=0,Z= ...(truncated)"
    Weight=1
    WeightMinimum=0
   )
   n11 = RigVMVariableNode() {variable=Sit Lower}
   &RigUnit_ModifyTransforms(
    ItemToModify=((Item=(Type=Bone,Name="Base_skirt_Fr_L"),Transform=(Rotation=(X=0,Y=0,Z=0,W=1),Translation=(X=5,Y=0,Z=0),Scale3D=(X=1,Y=1,Z=1))))
    WeightMinimum=0
    Weight=@n11.Value
    ExecutePin=@RigVMFunction_Sequence.C
   )
   > &RigUnit_ModifyTransforms(
    ItemToModify=((Item=(Type=Bone,Name="Dyn_knife_01"),Transform=(Rotation=(X=0,Y=0,Z=-0.1305,W=0.9914),Translation=(X=0,Y=0,Z=0),Scale3D=(X=1,Y=1,Z=1))))
    WeightMinimum=0
    Weight=@n11.Value
   )
   > &RigUnit_ModifyTransforms(
    ItemToModify=((Item=(Type=Bone,Name="Dyn_knife_04"),Transform=(Rotation=(X=0,Y=0,Z=-0.0698,W=0.9976),Translation=(X=0,Y=0,Z=0),Scale3D=(X=1,Y=1,Z=1))))
    WeightMinimum=0
    Weight=@n11.Value
   )
   > &RigUnit_ModifyTransforms(
    ItemToModify=((Item=(Type=Bone,Name="Base_skirt_Fr_R"),Transform=(Rotation=(X=0,Y=0,Z=0.0872,W=0.9962),Translation=(X=5,Y=0,Z=0),Scale3D=(X=1,Y=1,Z=1))))
    WeightMinimum=0
    Weight=@n11.Value
   )
   > &RigUnit_ModifyTransforms(
    ItemToModify=((Item=(Type=Bone,Name="Base_skirt_Side_L"),Transform=(Rotation=(X=0,Y=0,Z=0,W=1),Translation=(X=-1,Y=-2,Z=0),Scale3D=(X=1,Y=1,Z=1))))
    WeightMinimum=0
    Weight=@n11.Value
   )
   > n12 = &RigUnit_SphericalPoseReader(
    ActiveRegionScaleFactors=(PositiveWidth=1,NegativeWidth=1,PositiveHeight=1,NegativeHeight=1)
    ActiveRegionSize=0.1
    Debug=(bDrawDebug=false,bDraw2D=False,bDrawLocalAxes=False,DebugScale=25,DebugSegments=20,DebugThickness=0.25)
    DriverAxis=(X=-1,Y=0,Z=0)
    DriverItem=(Type=Bone,Name="thigh_l")
    FalloffRegionScaleFactors=(PositiveWidth=1,NegativeWidth=1,PositiveHeight=1,NegativeHeight=1)
    FalloffSize=0.2
    FlipHeightScaling=False
    FlipWidthScaling=False
    OptionalParentItem=(Type=Bone,Name="None")
    RotationOffset=(X=100,Y=0,Z=-10)
   )
   > &RigUnit_ModifyTransforms(
    ItemToModify=((Item=(Type=Bone,Name="Base_skirt_Fr_R"),Transform=(Rotation=(X=0,Y=0,Z=0,W=1),Translation=(X=5,Y=0,Z=0),Scale3D=(X=1,Y=1,Z=1))))
    WeightMinimum=0
    Weight=&RigVMFunction_MathDoubleMul(A=@n12.OutputParam, B=RigVMVariableNode() {variable=Stand}.Value).Result
   )
   > &RigUnit_ModifyTransforms(
    ItemToModify=((Item=(Type=Bone,Name="Base_skirt_Side_L"),Transform=(Rotation=(X=0,Y=0,Z=-0.2588,W=0.9659),Translation=(X=1,Y=2,Z=1),Scale3D=(X=1,Y=1,Z=1))))
    WeightMinimum=0
    Weight=RigVMVariableNode() {variable=No stand}.Value
   )
   > n13 = &RigUnit_SphericalPoseReader(
    ActiveRegionScaleFactors=(PositiveWidth=1,NegativeWidth=1,PositiveHeight=1,NegativeHeight=1)
    ActiveRegionSize=0.1
    Debug=(bDrawDebug=false,bDraw2D=False,bDrawLocalAxes=False,DebugScale=25,DebugSegments=20,DebugThickness=0.25)
    DriverAxis=(X=1,Y=0,Z=0)
    DriverItem=(Type=Bone,Name="thigh_r")
    FalloffRegionScaleFactors=(PositiveWidth=1,NegativeWidth=1,PositiveHeight=1,NegativeHeight=1)
    FalloffSize=0.2
    FlipHeightScaling=False
    FlipWidthScaling=False
    OptionalParentItem=(Type=Bone,Name="None")
    RotationOffset=(X=15,Y=-90,Z=0)
   )
   > &RigUnit_ModifyTransforms(
    ItemToModify=((Item=(Type=Bone,Name="Base_skirt_Fr_R"),Transform=(Rotation=(X=0,Y=-0.1305,Z=0,W=0.9914),Translation=(X=0,Y=0,Z=0),Scale3D=(X=1,Y=1,Z=1))))
    WeightMinimum=0
    Weight=&RigVMFunction_MathDoubleMul(A=@n13.OutputParam, B=RigVMVariableNode() {variable=Stand}.Value).Result
   )
   n14 = RigVMVariableNode() {variable=Crouch}
   > &RigUnit_ModifyTransforms(
    ItemToModify=((Item=(Type=Bone,Name="Dyn_knife_03"),Transform=(Rotation=(X=0,Y=0,Z=-0.0698,W=0.9976),Translation=(X=0,Y=0,Z=0),Scale3D=(X=1,Y=1,Z=1))))
    WeightMinimum=0
    Weight=@n14.Value
   )
   > &RigUnit_ModifyTransforms(
    ItemToModify=((Item=(Type=Bone,Name="Dyn_knife_04"),Transform=(Rotation=(X=0,Y=0,Z=-0.0698,W=0.9976),Translation=(X=0,Y=0,Z=0),Scale3D=(X=1,Y=1,Z=1))))
    WeightMinimum=0
    Weight=@n14.Value
   )
   n15 = &RigUnit_SphericalPoseReader(
    ActiveRegionScaleFactors=(PositiveWidth=1,NegativeWidth=1,PositiveHeight=1,NegativeHeight=1)
    ActiveRegionSize=0.1
    Debug=(bDrawDebug=false,bDraw2D=False,bDrawLocalAxes=False,DebugScale=25,DebugSegments=20,DebugThickness=0.25)
    DriverAxis=(X=-1,Y=0,Z=0)
    DriverItem=(Type=Bone,Name="thigh_l")
    FalloffRegionScaleFactors=(PositiveWidth=1,NegativeWidth=1,PositiveHeight=1,NegativeHeight=1)
    FalloffSize=0.2
    FlipHeightScaling=False
    FlipWidthScaling=False
    OptionalParentItem=(Type=Bone,Name="None")
    RotationOffset=(X=100,Y=0,Z=-10)
    ExecutePin=@RigVMFunction_Sequence.D
   )
   > &RigUnit_ModifyTransforms(
    ItemToModify=((Item=(Type=Bone,Name="Base_skirt_Fr_L"),Transform=(Rotation=(X=0,Y=0,Z=0,W=1),Translation=(X=5,Y=0,Z=0),Scale3D=(X=1,Y=1,Z=1))))
    WeightMinimum=0
    Weight=&RigVMFunction_MathDoubleMul(A=@n15.OutputParam, B=RigVMVariableNode() {variable=Crouch}.Value).Result
   )
   > n16 = &RigUnit_SphericalPoseReader(
    ActiveRegionScaleFactors=(PositiveWidth=1,NegativeWidth=1,PositiveHeight=1,NegativeHeight=1)
    ActiveRegionSize=0.1
    Debug=(bDrawDebug=false,bDraw2D=False,bDrawLocalAxes=False,DebugScale=25,DebugSegments=20,DebugThickness=0.25)
    DriverAxis=(X=1,Y=0,Z=0)
    DriverItem=(Type=Bone,Name="thigh_r")
    FalloffRegionScaleFactors=(PositiveWidth=1,NegativeWidth=1,PositiveHeight=1,NegativeHeight=1)
    FalloffSize=0.2
    FlipHeightScaling=False
    FlipWidthScaling=False
    OptionalParentItem=(Type=Bone,Name="None")
    RotationOffset=(X=-85,Y=0,Z=0)
   )
   > &RigUnit_ModifyTransforms(
    ItemToModify=((Item=(Type=Bone,Name="Base_skirt_Fr_R"),Transform=(Rotation=(X=0,Y=0,Z=0,W=1),Translation=(X=5,Y=0,Z=0),Scale3D=(X=1,Y=1,Z=1))))
    WeightMinimum=0
    Weight=&RigVMFunction_MathDoubleMul(A=@n16.OutputParam, B=RigVMVariableNode() {variable=Crouch}.Value).Result
   )
   > &RigUnit_ModifyTransforms(
    ItemToModify=((Item=(Type=Bone,Name="Base_rope"),Transform=(Rotation=(X=0,Y=0,Z=0,W=1),Translation=(X=-1,Y=2,Z=0),Scale3D=(X=1,Y=1,Z=1))))
    WeightMinimum=0
    Weight=RigVMVariableNode() {variable=Crouch}.Value
   )
   n17 = &RigVMFunction_MathDoubleRemap(
    SourceMaximum=-70
    SourceMinimum=-50
    TargetMaximum=1
    TargetMinimum=0
    Value=&RigVMFunction_MathQuaternionToEuler(
     RotationOrder=ZYX
     Value=&RigUnit_GetRelativeTransformForItem(Child=(Type=Bone,Name="spine_05"), Parent=(Type=Bone,Name="pelvis"), bChildInitial=False, bParentInitial=False).RelativeTransform.Rotation
    ).Result.X
   )
   > &RigUnit_ModifyTransforms(
    ItemToModify=((Item=(Type=Bone,Name="Dyn_knife_01"),Transform=(Rotation=(X=-0.0053,Y=-0.087,Z=-0.0608,W=0.9943),Translation=(X=0,Y=0,Z=0),Scale3D=(X=1,Y=1,Z=1))))
    WeightMinimum=0
    Weight=@n17.Result
   )
   > &RigUnit_ModifyTransforms(
    ItemToModify=((Item=(Type=Bone,Name="Dyn_knife_02"),Transform=(Rotation=(X=0,Y=0,Z=-0.0436,W=0.999),Translation=(X=0,Y=0,Z=0),Scale3D=(X=1,Y=1,Z=1))))
    WeightMinimum=0
    Weight=@n17.Result
   )
   > Get Delta Ref_1 = RigVMFunctionReferenceNode(Item=(Type=Bone,Name="pelvis"))
   > &RigUnit_ModifyTransforms(
    ItemToModify="((Item=(Type=Bone,Name=\"Base_skirt_Bk_R\"),Transform=(Rotation=(X=0,Y=-0.0698,Z=0,W=0.9976),Translation=(X=0,Y=0,Z=1),Scale3D=(X=1,Y=1,Z=1))),(Item=(Type=Bone,Name=\"Base_skirt_Bk_L\"),Transform=(Rotation=(X=0,Y=-0.0436,Z=0,W=0.999),Translation=(X=0,Y=0,Z=1), ...(truncated)"
    WeightMinimum=0
    Weight=&RigVMFunction_MathDoubleRemap(SourceMaximum=60, SourceMinimum=40, TargetMaximum=1, TargetMinimum=0, Value=@"Get Delta Ref_1".Result.Rotation.Yaw).Result
   )
   &RigUnit_ParentConstraint(
    AdvancedSettings=(InterpolationType=Average,RotationOrderForFilter=XZY)
    Child=(Type=Bone,Name="Base_skirt_Side_R")
    Filter=(TranslationFilter=(bX=false,bY=false,bZ=false),RotationFilter=(bX=true,bY=false,bZ=false),ScaleFilter=(bX=True,bY=True,bZ=True))
    Parents=((Item=(Type=Bone,Name="pelvis"),Weight=0.4),(Item=(Type=Bone,Name="thigh_r"),Weight=0.6))
    Weight=1
    bMaintainOffset=true
    ExecutePin=@RigVMFunction_Sequence.B
   )
   > &RigUnit_ParentConstraint(
    AdvancedSettings=(InterpolationType=Average,RotationOrderForFilter=XZY)
    Child=(Type=Bone,Name="Base_skirt_Side_L")
    Filter=(TranslationFilter=(bX=false,bY=false,bZ=false),RotationFilter=(bX=true,bY=false,bZ=false),ScaleFilter=(bX=True,bY=True,bZ=True))
    Parents=((Item=(Type=Bone,Name="pelvis"),Weight=0.4),(Item=(Type=Bone,Name="thigh_l"),Weight=0.6))
    Weight=1
    bMaintainOffset=true
   )
   > &RigUnit_ParentConstraint(
    AdvancedSettings=(InterpolationType=Average,RotationOrderForFilter=XZY)
    Child=(Type=Bone,Name="Base_skirt_Bk_R")
    Filter=(TranslationFilter=(bX=true,bY=true,bZ=true),RotationFilter=(bX=true,bY=true,bZ=true),ScaleFilter=(bX=True,bY=True,bZ=True))
    Parents=((Item=(Type=Bone,Name="pelvis"),Weight=0.6),(Item=(Type=Bone,Name="thigh_r"),Weight=0.4))
    Weight=1
   )
   > &RigUnit_ParentConstraint(
    AdvancedSettings=(InterpolationType=Average,RotationOrderForFilter=XZY)
    Child=(Type=Bone,Name="Base_skirt_Bk_L")
    Filter=(TranslationFilter=(bX=true,bY=true,bZ=true),RotationFilter=(bX=True,bY=True,bZ=True),ScaleFilter=(bX=True,bY=True,bZ=True))
    Parents=((Item=(Type=Bone,Name="pelvis"),Weight=0.6),(Item=(Type=Bone,Name="thigh_l"),Weight=0.4))
    Weight=1
   )
   > &RigUnit_ModifyTransforms(
    ItemToModify=((Item=(Type=Bone,Name="Base_skirt_Side_L"),Transform=(Rotation=(X=0,Y=0,Z=0,W=1),Translation=(X=-1,Y=0,Z=1),Scale3D=(X=1,Y=1,Z=1))))
    Weight=1
    WeightMinimum=0
   )
   > &RigUnit_ModifyTransforms(
    ItemToModify=((Item=(Type=Bone,Name="Base_skirt_Side_R"),Transform=(Rotation=(X=0,Y=0,Z=0,W=1),Translation=(X=0,Y=0,Z=1),Scale3D=(X=1,Y=1,Z=1))))
    Weight=1
    WeightMinimum=0
   )
   > &RigUnit_ModifyTransforms(
    ItemToModify=((Item=(Type=Bone,Name="Base_skirt_Bk_R"),Transform=(Rotation=(X=0,Y=0,Z=0,W=1),Translation=(X=0,Y=0,Z=1),Scale3D=(X=1,Y=1,Z=1))))
    Weight=1
    WeightMinimum=0
   )
   > &RigUnit_ModifyTransforms(
    ItemToModify=((Item=(Type=Bone,Name="Base_skirt_Bk_L"),Transform=(Rotation=(X=0,Y=0,Z=0,W=1),Translation=(X=0,Y=0,Z=1),Scale3D=(X=1,Y=1,Z=1))))
    Weight=1
    WeightMinimum=0
   )
   > &RigUnit_ModifyTransforms(
    ItemToModify=((Item=(Type=Bone,Name="Dyn_knife_03"),Transform=(Rotation=(X=0,Y=0,Z=0,W=1),Translation=(X=-1.5,Y=0,Z=0),Scale3D=(X=0.9,Y=0.9,Z=0.9))))
    Weight=1
    WeightMinimum=0
   )
   > &RigUnit_ModifyTransforms(
    ItemToModify=((Item=(Type=Bone,Name="Dyn_knife_04"),Transform=(Rotation=(X=0,Y=0,Z=0,W=1),Translation=(X=0,Y=0,Z=0),Scale3D=(X=0.9,Y=0.9,Z=0.9))))
    Weight=1
    WeightMinimum=0
   )
   > &RigUnit_ModifyTransforms(
    ItemToModify=((Item=(Type=Bone,Name="deform_glute_base_l"),Transform=(Rotation=(X=0,Y=0,Z=0,W=1),Translation=(X=-1,Y=-1,Z=0),Scale3D=(X=1,Y=1,Z=1))))
    Weight=1
    WeightMinimum=0
   )
   > n18 = &RigUnit_ModifyTransforms(
    ItemToModify=((Item=(Type=Bone,Name="deform_glute_base_r"),Transform=(Rotation=(X=0,Y=0,Z=0,W=1),Translation=(X=1,Y=1,Z=1),Scale3D=(X=1,Y=1,Z=1))))
    Weight=1
    WeightMinimum=0
   )
   n19 = &RigVMFunction_ControlFlowBranch(ExecuteContext=@n18.ExecutePin, Condition=RigVMVariableNode() {variable=DBNO}.Value)
   &RigUnit_ModifyTransforms(
    ItemToModify="((Item=(Type=Bone,Name=\"Base_skirt_Bk_L\"),Transform=(Rotation=(X=0,Y=0.2164,Z=0,W=0.9763),Translation=(X=0,Y=0,Z=0),Scale3D=(X=1,Y=1,Z=1))),(Item=(Type=Bone,Name=\"Base_skirt_Bk_R\"),Transform=(Rotation=(X=0,Y=0.2164,Z=0,W=0.9763),Translation=(X=0,Y=0,Z=0),S ...(truncated)"
    Weight=1
    WeightMinimum=0
    ExecutePin=@n19.True
   )
   Greater = /Script/RigVM.RigVMFunction_MathDoubleGreater(B=800, A=RigVMVariableNode() {variable=Velocity}.Value) {is=struct RigVMUnitNode, template="Greater::Execute(in A,in B,out Result)"}
   n20 = &RigVMFunction_ControlFlowBranch(Condition=@Greater.Result, ExecuteContext=@RigVMFunction_Sequence.G)
   n21 = &RigUnit_SphericalPoseReader(
    ActiveRegionScaleFactors=(PositiveWidth=1,NegativeWidth=1,PositiveHeight=1,NegativeHeight=1)
    ActiveRegionSize=0.1
    Debug=(bDrawDebug=false,bDraw2D=False,bDrawLocalAxes=False,DebugScale=25,DebugSegments=20,DebugThickness=0.25)
    DriverAxis=(X=1,Y=0,Z=0)
    DriverItem=(Type=Bone,Name="thigh_r")
    FalloffRegionScaleFactors=(PositiveWidth=1,NegativeWidth=1,PositiveHeight=1,NegativeHeight=1)
    FalloffSize=0.2
    FlipHeightScaling=False
    FlipWidthScaling=False
    OptionalParentItem=(Type=Bone,Name="None")
    RotationOffset=(X=-55,Y=-30,Z=0)
    ExecutePin=@n20.True
   )
   > &RigUnit_ModifyTransforms(
    ItemToModify=((Item=(Type=Bone,Name="Base_skirt_Side_R"),Transform=(Rotation=(X=-0.0076,Y=-0.0868,Z=-0.0868,W=0.9924),Translation=(X=0,Y=0,Z=0),Scale3D=(X=1,Y=1,Z=1))))
    WeightMinimum=0
    Weight=@n21.OutputParam
   )
   RigVMFunction_ControlFlowBranch = &RigVMFunction_ControlFlowBranch(Condition=RigVMVariableNode() {variable=PlayingEmote}.Value, ExecuteContext=@RigVMFunction_Sequence.F)
   &RigUnit_ParentConstraint(
    AdvancedSettings=(InterpolationType=Average,RotationOrderForFilter=XZY)
    Child=(Type=Bone,Name="Base_skirt_Side_R")
    Filter=(TranslationFilter=(bX=false,bY=false,bZ=false),RotationFilter=(bX=true,bY=false,bZ=false),ScaleFilter=(bX=True,bY=True,bZ=True))
    Parents=((Item=(Type=Bone,Name="thigh_r"),Weight=1))
    Weight=1
    bMaintainOffset=true
    ExecutePin=@RigVMFunction_ControlFlowBranch.True
   )
   > &RigUnit_ModifyTransforms(
    ItemToModify=((Item=(Type=Bone,Name="Base_skirt_Side_R"),Transform=(Rotation=(X=0,Y=0,Z=0,W=1),Translation=(X=0,Y=0,Z=1),Scale3D=(X=1,Y=1,Z=1))))
    Weight=1
    WeightMinimum=0
   )
   > &RigUnit_ModifyTransforms(
    ItemToModify=((Item=(Type=Bone,Name="Base_skirt_Bk_R"),Transform=(Rotation=(X=0,Y=0,Z=0,W=1),Translation=(X=0,Y=0,Z=1),Scale3D=(X=1,Y=1,Z=1))))
    Weight=1
    WeightMinimum=0
   )
   > &RigUnit_ModifyTransforms(
    ItemToModify=((Item=(Type=Bone,Name="Base_skirt_Bk_L"),Transform=(Rotation=(X=0,Y=0,Z=0,W=1),Translation=(X=0,Y=0,Z=1),Scale3D=(X=1,Y=1,Z=1))))
    Weight=1
    WeightMinimum=0
   )
   > &RigUnit_ModifyTransforms(
    ItemToModify=((Item=(Type=Bone,Name="deform_glute_base_r"),Transform=(Rotation=(X=0,Y=0,Z=0,W=1),Translation=(X=1,Y=1,Z=1),Scale3D=(X=1,Y=1,Z=1))))
    Weight=1
    WeightMinimum=0
   )
   > &RigUnit_ModifyTransforms(
    ItemToModify=((Item=(Type=Bone,Name="deform_glute_base_l"),Transform=(Rotation=(X=0,Y=0,Z=0,W=1),Translation=(X=-1,Y=-1,Z=-1),Scale3D=(X=1,Y=1,Z=1))))
    Weight=1
    WeightMinimum=0
   )
   > &RigUnit_ModifyTransforms(
    ItemToModify="((Item=(Type=Bone,Name=\"Dyn_knife_04\"),Transform=(Rotation=(X=-0.0019,Y=-0.0436,Z=-0.0436,W=0.9981),Translation=(X=-1,Y=0,Z=0),Scale3D=(X=1,Y=1,Z=1))),(Item=(Type=Bone,Name=\"Dyn_knife_03\"),Transform=(Rotation=(X=-0.0038,Y=-0.0435,Z=-0.0871,W=0.9952),Transl ...(truncated)"
    Weight=1
    WeightMinimum=0
   )
   > n22 = &RigUnit_SphericalPoseReader(
    ActiveRegionScaleFactors=(PositiveWidth=1,NegativeWidth=1,PositiveHeight=1,NegativeHeight=1)
    ActiveRegionSize=0.1
    Debug=(bDrawDebug=false,bDraw2D=False,bDrawLocalAxes=False,DebugScale=25,DebugSegments=20,DebugThickness=0.25)
    DriverAxis=(X=1,Y=0,Z=0)
    DriverItem=(Type=Bone,Name="thigh_r")
    FalloffRegionScaleFactors=(PositiveWidth=1,NegativeWidth=1,PositiveHeight=1,NegativeHeight=1)
    FalloffSize=0.2
    FlipHeightScaling=False
    FlipWidthScaling=False
    OptionalParentItem=(Type=Bone,Name="None")
    RotationOffset=(X=-90,Y=0,Z=-25)
   )
   > &RigUnit_ModifyTransforms(
    ItemToModify=((Item=(Type=Bone,Name="Base_skirt_Fr_R"),Transform=(Rotation=(X=0,Y=0,Z=0,W=1),Translation=(X=5,Y=0,Z=0),Scale3D=(X=1,Y=1,Z=1))))
    WeightMinimum=0
    Weight=@n22.OutputParam
   )
   > &RigUnit_ModifyTransforms(
    ItemToModify=((Item=(Type=Bone,Name="Base_rope"),Transform=(Rotation=(X=-0.2578,Y=0.0842,Z=0.0226,W=0.9623),Translation=(X=0,Y=0,Z=0),Scale3D=(X=1,Y=1,Z=1))))
    WeightMinimum=0
    Weight=@n22.OutputParam
   )
   > n23 = &RigUnit_SphericalPoseReader(
    ActiveRegionScaleFactors=(PositiveWidth=1,NegativeWidth=1,PositiveHeight=1,NegativeHeight=1)
    ActiveRegionSize=0.1
    Debug=(bDrawDebug=false,bDraw2D=False,bDrawLocalAxes=False,DebugScale=25,DebugSegments=20,DebugThickness=0.25)
    DriverAxis=(X=-1,Y=0,Z=0)
    DriverItem=(Type=Bone,Name="upperarm_r")
    FalloffRegionScaleFactors=(PositiveWidth=1,NegativeWidth=1,PositiveHeight=1,NegativeHeight=1)
    FalloffSize=0.2
    FlipHeightScaling=False
    FlipWidthScaling=False
    OptionalParentItem=(Type=Bone,Name="None")
    RotationOffset=(X=10,Y=40,Z=0)
   )
   > n24 = &RigUnit_SphericalPoseReader(
    ActiveRegionScaleFactors=(PositiveWidth=1,NegativeWidth=1,PositiveHeight=1,NegativeHeight=1)
    ActiveRegionSize=0.1
    Debug=(bDrawDebug=false,bDraw2D=False,bDrawLocalAxes=False,DebugScale=25,DebugSegments=20,DebugThickness=0.25)
    DriverAxis=(X=1,Y=0,Z=0)
    DriverItem=(Type=Bone,Name="upperarm_l")
    FalloffRegionScaleFactors=(PositiveWidth=1,NegativeWidth=1,PositiveHeight=1,NegativeHeight=1)
    FalloffSize=0.2
    FlipHeightScaling=False
    FlipWidthScaling=False
    OptionalParentItem=(Type=Bone,Name="None")
    RotationOffset=(X=-15,Y=-140,Z=0)
   )
   n25 = &RigVMFunction_MathDoubleMul(B=@n23.OutputParam, A=@n24.OutputParam)
   > &RigUnit_ModifyTransforms(
    ItemToModify=((Item=(Type=Bone,Name="Base_Cape_C"),Transform=(Rotation=(X=0,Y=-0.0872,Z=0,W=0.9962),Translation=(X=0,Y=0,Z=0),Scale3D=(X=1,Y=1,Z=1))))
    WeightMinimum=0
    Weight=@n25.Result
   )
   > &RigUnit_ModifyTransforms(
    ItemToModify=((Item=(Type=Bone,Name="Dyn_Cape_R"),Transform=(Rotation=(X=0,Y=0.0872,Z=0,W=0.9962),Translation=(X=0,Y=0,Z=0),Scale3D=(X=1,Y=1,Z=1))))
    WeightMinimum=0
    Weight=@n25.Result
   )
   > &RigUnit_ModifyTransforms(
    ItemToModify=((Item=(Type=Bone,Name="Dyn_Cape_L"),Transform=(Rotation=(X=0,Y=0.0872,Z=0,W=0.9962),Translation=(X=0,Y=0,Z=0),Scale3D=(X=1,Y=1,Z=1))))
    WeightMinimum=0
    Weight=@n25.Result
   )
   n26 = &RigUnit_GetRelativeTransformForItem(Child=(Type=Bone,Name="thigh_l"), Parent=(Type=Bone,Name="pelvis"), bChildInitial=False, bParentInitial=False)
   > &RigUnit_ModifyTransforms(
    ItemToModify=((Item=(Type=Bone,Name="Base_knife"),Transform=(Rotation=(X=0,Y=0,Z=-0.1736,W=0.9848),Translation=(X=0,Y=0,Z=0),Scale3D=(X=1,Y=1,Z=1))))
    WeightMinimum=0
    Weight=&RigVMFunction_MathDoubleRemap(
     SourceMaximum=-45
     SourceMinimum=-20
     TargetMaximum=1
     TargetMinimum=0
     Value=&RigVMFunction_MathQuaternionToEuler(RotationOrder=ZYX, Value=@n26.RelativeTransform.Rotation).Result.Z
    ).Result
   )
   n27 = &RigUnit_GetRelativeTransformForItem(Child=(Type=Bone,Name="spine_03"), Parent=(Type=Bone,Name="pelvis"), bChildInitial=False, bParentInitial=False)
   > &RigUnit_ModifyTransforms(
    ItemToModify=((Item=(Type=Bone,Name="Base_knife"),Transform=(Rotation=(X=0,Y=0,Z=-0.1736,W=0.9848),Translation=(X=0,Y=0,Z=0),Scale3D=(X=1,Y=1,Z=1))))
    WeightMinimum=0
    Weight=&RigVMFunction_MathDoubleRemap(
     SourceMaximum=-45
     SourceMinimum=-20
     TargetMaximum=1
     TargetMinimum=0
     Value=&RigVMFunction_MathQuaternionToEuler(RotationOrder=ZYX, Value=@n27.RelativeTransform.Rotation).Result.Y
    ).Result
   )
   > n28 = &RigUnit_SphericalPoseReader(
    ActiveRegionScaleFactors=(PositiveWidth=1,NegativeWidth=1,PositiveHeight=1,NegativeHeight=1)
    ActiveRegionSize=0.1
    Debug=(bDrawDebug=false,bDraw2D=False,bDrawLocalAxes=False,DebugScale=25,DebugSegments=20,DebugThickness=0.25)
    DriverAxis=(X=1,Y=0,Z=0)
    DriverItem=(Type=Bone,Name="thigh_r")
    FalloffRegionScaleFactors=(PositiveWidth=1,NegativeWidth=1,PositiveHeight=1,NegativeHeight=1)
    FalloffSize=0.2
    FlipHeightScaling=False
    FlipWidthScaling=False
    OptionalParentItem=(Type=Bone,Name="None")
    RotationOffset=(X=-90,Y=0,Z=-35)
   )
   > &RigUnit_ModifyTransforms(
    ItemToModify="((Item=(Type=Bone,Name=\"Base_skirt_Fr_R\"),Transform=(Rotation=(X=0.0038,Y=0.0435,Z=-0.0871,W=0.9952),Translation=(X=0,Y=0,Z=0),Scale3D=(X=1,Y=1,Z=1))),(Item=(Type=Bone,Name=\"Base_skirt_Fr_L\"),Transform=(Rotation=(X=0,Y=0,Z=0.0436,W=0.999),Translation=(X=0, ...(truncated)"
    WeightMinimum=0
    Weight=@n28.OutputParam
   )
   RigVMFunction_MathQuaternionToEuler = &RigVMFunction_MathQuaternionToEuler(
    RotationOrder=ZYX
    Value=&RigUnit_GetRelativeTransformForItem(Child=(Type=Bone,Name="foot_l"), Parent=(Type=Bone,Name="calf_l"), bChildInitial=False, bParentInitial=False).RelativeTransform.Rotation
   )
   > &RigUnit_ModifyTransforms(
    ItemToModify=((Item=(Type=Bone,Name="Correct_foot_L"),Transform=(Rotation=(X=0,Y=0,Z=0,W=1),Translation=(X=7,Y=0,Z=-12),Scale3D=(X=1,Y=1,Z=1))))
    WeightMinimum=0
    Weight=&RigVMFunction_MathDoubleRemap(SourceMaximum=40, SourceMinimum=30, TargetMaximum=1, TargetMinimum=0, Value=@RigVMFunction_MathQuaternionToEuler.Result.Z).Result
   )
   n29 = &RigUnit_GetRelativeTransformForItem(Child=(Type=Bone,Name="foot_r"), Parent=(Type=Bone,Name="calf_r"), bChildInitial=False, bParentInitial=False)
   > &RigUnit_ModifyTransforms(
    ItemToModify=((Item=(Type=Bone,Name="Correct_foot_R"),Transform=(Rotation=(X=0,Y=0,Z=0,W=1),Translation=(X=7,Y=0,Z=8),Scale3D=(X=1,Y=1,Z=1))))
    WeightMinimum=0
    Weight=&RigVMFunction_MathDoubleRemap(
     SourceMaximum=35
     SourceMinimum=25
     TargetMaximum=1
     TargetMinimum=0
     Value=&RigVMFunction_MathQuaternionToEuler(RotationOrder=ZYX, Value=@n29.RelativeTransform.Rotation).Result.Z
    ).Result
   )
   &RigUnit_ModifyTransforms(
    ItemToModify=((Item=(Type=Bone,Name="Base_knife"),Transform=(Rotation=(X=0,Y=0,Z=-0.0872,W=0.9962),Translation=(X=0,Y=-1,Z=0),Scale3D=(X=1,Y=1,Z=1))))
    WeightMinimum=0
    Weight=&RigVMFunction_MathDoubleRemap(
     SourceMaximum=30
     SourceMinimum=35
     TargetMaximum=1
     TargetMinimum=0
     Value=&RigUnit_GetRelativeTransformForItem(Child=(Type=Bone,Name="pelvis"), Parent=(Type=Bone,Name="root"), bChildInitial=False, bParentInitial=False).RelativeTransform.Translation.Z
    ).Result
    ExecutePin=@RigVMFunction_Sequence.E
   )
   n30 = &RigVMFunction_MathDoubleRemap(
    SourceMaximum=15
    SourceMinimum=23
    TargetMaximum=1
    TargetMinimum=0
    Value=&RigUnit_GetRelativeTransformForItem(Child=(Type=Bone,Name="pelvis"), Parent=(Type=Bone,Name="root"), bChildInitial=False, bParentInitial=False).RelativeTransform.Translation.Z
   )
   > &RigUnit_ModifyTransforms(
    ItemToModify=((Item=(Type=Bone,Name="Base_knife"),Transform=(Rotation=(X=0.0057,Y=0.0432,Z=-0.1304,W=0.9905),Translation=(X=0,Y=0,Z=0),Scale3D=(X=1,Y=1,Z=1))))
    WeightMinimum=0
    Weight=@n30.Result
   )
   > &RigUnit_ModifyTransforms(
    ItemToModify=((Item=(Type=Bone,Name="Base_rope"),Transform=(Rotation=(X=-0.171,Y=-0.171,Z=-0.0302,W=0.9698),Translation=(X=0,Y=0,Z=0),Scale3D=(X=1,Y=1,Z=1))))
    WeightMinimum=0
    Weight=@n30.Result
   )
   n31 = &RigUnit_GetRelativeTransformForItem(Child=(Type=Bone,Name="spine_05"), Parent=(Type=Bone,Name="pelvis"), bChildInitial=False, bParentInitial=False)
   > &RigUnit_ModifyTransforms(
    ItemToModify=((Item=(Type=Bone,Name="Base_Cape_C"),Transform=(Rotation=(X=0,Y=-0.6428,Z=0,W=0.766),Translation=(X=0,Y=0,Z=0),Scale3D=(X=1,Y=1,Z=1))))
    WeightMinimum=0
    Weight=&RigVMFunction_MathDoubleRemap(
     SourceMaximum=-90
     SourceMinimum=-70
     TargetMaximum=1
     TargetMinimum=0
     Value=&RigVMFunction_MathQuaternionToEuler(RotationOrder=ZYX, Value=@n31.RelativeTransform.Rotation).Result.Z
    ).Result
   )
   n32 = &RigVMFunction_ControlFlowBranch(Condition=RigVMVariableNode() {variable=FrontEndOn}.Value, ExecuteContext=@RigVMFunction_Sequence.K)
   n33 = &RigUnit_SphericalPoseReader(
    ActiveRegionScaleFactors=(PositiveWidth=1,NegativeWidth=1,PositiveHeight=1,NegativeHeight=1)
    ActiveRegionSize=0.025
    Debug=(bDrawDebug=true,bDraw2D=False,bDrawLocalAxes=False,DebugScale=25,DebugSegments=20,DebugThickness=0.25)
    DriverAxis=(X=1,Y=0,Z=0)
    DriverItem=(Type=Bone,Name="upperarm_l")
    FalloffRegionScaleFactors=(PositiveWidth=1,NegativeWidth=1,PositiveHeight=1,NegativeHeight=1)
    FalloffSize=0.2
    FlipHeightScaling=False
    FlipWidthScaling=False
    OptionalParentItem=(Type=Bone,Name="None")
    RotationOffset=(X=-13,Y=-120,Z=0)
    ExecutePin=@n32.True
   )
   > n34 = &RigUnit_SphericalPoseReader(
    ActiveRegionScaleFactors=(PositiveWidth=1,NegativeWidth=1,PositiveHeight=1,NegativeHeight=1)
    ActiveRegionSize=0.025
    Debug=(bDrawDebug=True,bDraw2D=False,bDrawLocalAxes=False,DebugScale=25,DebugSegments=20,DebugThickness=0.25)
    DriverAxis=(X=-1,Y=0,Z=0)
    DriverItem=(Type=Bone,Name="upperarm_r")
    FalloffRegionScaleFactors=(PositiveWidth=1,NegativeWidth=1,PositiveHeight=1,NegativeHeight=1)
    FalloffSize=0.2
    FlipHeightScaling=False
    FlipWidthScaling=False
    OptionalParentItem=(Type=Bone,Name="None")
    RotationOffset=(X=12,Y=60,Z=0)
   )
   n35 = &RigVMFunction_MathDoubleMul(A=@n33.OutputParam, B=@n34.OutputParam)
   > &RigUnit_ModifyTransforms(
    ItemToModify=((Item=(Type=Bone,Name="Base_Cape_C"),Transform=(Rotation=(X=0,Y=0.0872,Z=0,W=0.9962),Translation=(X=0,Y=0,Z=2.5),Scale3D=(X=1,Y=0.8,Z=1))))
    WeightMinimum=0
    Weight=@n35.Result
   )
   > &RigUnit_ModifyTransforms(
    ItemToModify=((Item=(Type=Bone,Name="Base_Cape_L"),Transform=(Rotation=(X=0,Y=0,Z=0,W=1),Translation=(X=-1,Y=0,Z=0),Scale3D=(X=1,Y=1,Z=1))))
    WeightMinimum=0
    Weight=@n35.Result
   )
   > &RigUnit_ModifyTransforms(
    ItemToModify=((Item=(Type=Bone,Name="Base_Cape_R"),Transform=(Rotation=(X=0,Y=0,Z=0,W=1),Translation=(X=-1,Y=0,Z=0),Scale3D=(X=1,Y=1,Z=1))))
    WeightMinimum=0
    Weight=@n35.Result
   )
  ]}
]
hierarchy [
 root {loc=<0 0 0>, rot=<0 0 0 1>}
 pelvis {parent=root, loc=<0 -0.5301 98.0571>, rot=<0.0127 0.707 -0.0127 -0.707>}
 spine_01 {parent=pelvis, loc=<6.1208 -0.9328 0>, rot=<0 0 0.034 -0.9994>}
 spine_02 {parent=spine_01, loc=<6.5997 -0.2095 0>, rot=<0 0 0.0238 -0.9997>}
 spine_03 {parent=spine_02, loc=<6.4914 -0.1237 0>, rot=<0 0 0.0149 -0.9999>}
 spine_04 {parent=spine_03, loc=<6.9598 0.0794 0>, rot=<0 0 -0.0132 -0.9999>}
 spine_05 {parent=spine_04, loc=<6.9987 0 0>, rot=<0 0 -0.0015 -1>}
 neck_01 {parent=spine_05, loc=<11.3138 0.1529 0>, rot=<0 0 -0.2003 -0.9797>}
 neck_02 {parent=neck_01, loc=<4.5627 0 0>, rot=<0 0 0.0504 -0.9987>}
 head {parent=neck_02, loc=<5.3914 0 0>, rot=<0 0 0.1108 -0.9938>}
 FACIAL_C_FacialRoot {parent=head, loc=<2.1462 -0.7638 -0.0002>, rot=<0.4858 -0.5138 -0.4858 -0.5138>}
 FACIAL_C_Forehead {parent=FACIAL_C_FacialRoot, loc=<-0.0009 -2.9256 8.867>, rot=<-0.0492 0 0 -0.9988>}
 FACIAL_L_ForeheadIn {parent=FACIAL_C_FacialRoot, loc=<1.6442 -2.8794 8.7178>, rot=<-0.0664 -0.0696 0.0039 -0.9954>}
 FACIAL_R_ForeheadIn {parent=FACIAL_C_FacialRoot, loc=<-1.6457 -2.8797 8.7163>, rot=<-0.0664 0.0697 -0.0038 -0.9953>}
 FACIAL_L_ForeheadMid {parent=FACIAL_C_FacialRoot, loc=<3.063 -3.1235 8.2093>, rot=<-0.0766 -0.1755 -0.0216 -0.9813>}
 FACIAL_R_ForeheadMid {parent=FACIAL_C_FacialRoot, loc=<-3.0637 -3.1236 8.2077>, rot=<-0.0766 0.1756 0.0216 -0.9812>}
 FACIAL_L_ForeheadOut {parent=FACIAL_C_FacialRoot, loc=<4.3265 -3.0733 7.3279>, rot=<-0.1269 -0.2551 -0.0806 -0.9552>}
 FACIAL_R_ForeheadOut {parent=FACIAL_C_FacialRoot, loc=<-4.3265 -3.073 7.3269>, rot=<-0.1269 0.2552 0.0807 -0.9551>}
 FACIAL_L_EyesackUpper {parent=FACIAL_C_FacialRoot, loc=<3.124 -2.0358 8.7328>, rot=<0.0431 -0.0838 -0.065 -0.9934>}
 FACIAL_R_EyesackUpper {parent=FACIAL_C_FacialRoot, loc=<-3.1234 -2.0365 8.734>, rot=<0.0431 0.0839 0.065 -0.9934>}
 FACIAL_L_EyelidUpperFurrow {parent=FACIAL_C_FacialRoot, loc=<2.9977 -1.8032 8.6547>, rot=<-0.0357 -0.1503 -0.0805 -0.9847>}
 FACIAL_R_EyelidUpperFurrow {parent=FACIAL_C_FacialRoot, loc=<-2.9957 -1.8051 8.6564>, rot=<-0.0357 0.1504 0.0806 -0.9847>}
 FACIAL_L_EyelidUpperA {parent=FACIAL_C_FacialRoot, loc=<2.7937 -1.0337 6.7665>, rot=<-0.1535 0 0 -0.9881>}
 FACIAL_R_EyelidUpperA {parent=FACIAL_C_FacialRoot, loc=<-2.7922 -1.0348 6.7696>, rot=<-0.1535 0 0 -0.9881>}
 FACIAL_L_Eye {parent=FACIAL_C_FacialRoot, loc=<2.7904 -1.0475 6.7647>, rot=<-0.0058 0 0 -1>}
 FACIAL_R_Eye {parent=FACIAL_C_FacialRoot, loc=<-2.7915 -1.0468 6.7674>, rot=<-0.0058 0 0 -1>}
 FACIAL_L_EyelidLowerA {parent=FACIAL_C_FacialRoot, loc=<2.7937 -1.0337 6.7665>, rot=<0.2785 0 0 -0.9604>}
 FACIAL_R_EyelidLowerA {parent=FACIAL_C_FacialRoot, loc=<-2.7922 -1.0348 6.7696>, rot=<0.2785 0 0 -0.9604>}
 FACIAL_L_EyeCornerInner {parent=FACIAL_C_FacialRoot, loc=<1.5718 -0.2293 8.0765>, rot=<-0.0058 0 0 -1>}
 FACIAL_R_EyeCornerInner {parent=FACIAL_C_FacialRoot, loc=<-1.5714 -0.2367 8.0799>, rot=<-0.0058 0 0 -1>}
 FACIAL_L_EyeCornerOuter {parent=FACIAL_C_FacialRoot, loc=<4.5678 -0.881 7.3683>, rot=<-0.0053 -0.3826 -0.0022 -0.9239>}
 FACIAL_R_EyeCornerOuter {parent=FACIAL_C_FacialRoot, loc=<-4.5669 -0.8801 7.37>, rot=<-0.0053 0.3826 0.0022 -0.9239>}
 FACIAL_L_EyesackLower {parent=FACIAL_C_FacialRoot, loc=<3.288 0.6064 7.3155>, rot=<-0.028 -0.1395 -0.023 -0.9896>}
 FACIAL_R_EyesackLower {parent=FACIAL_C_FacialRoot, loc=<-3.2867 0.6018 7.3134>, rot=<-0.028 0.1396 0.0231 -0.9895>}
 FACIAL_L_CheekInner {parent=FACIAL_C_FacialRoot, loc=<2.0453 1.315 7.7799>, rot=<-0.1288 -0.0868 -0.1976 -0.9679>}
 FACIAL_R_CheekInner {parent=FACIAL_C_FacialRoot, loc=<-2.0463 1.3157 7.7782>, rot=<-0.1288 0.0869 0.1976 -0.9679>}
 FACIAL_L_CheekOuter {parent=FACIAL_C_FacialRoot, loc=<3.8159 1.3856 6.1618>, rot=<0.0499 -0.3775 0.0982 -0.9194>}
 FACIAL_R_CheekOuter {parent=FACIAL_C_FacialRoot, loc=<-3.8151 1.3865 6.1599>, rot=<0.0499 0.3775 -0.0981 -0.9194>}
 FACIAL_L_NasolabialBulge {parent=FACIAL_C_FacialRoot, loc=<2.9755 4.6013 6.0608>, rot=<0.0526 -0.3774 0.0217 -0.9243>}
 FACIAL_R_NasolabialBulge {parent=FACIAL_C_FacialRoot, loc=<-2.978 4.6027 6.0583>, rot=<0.0526 0.3775 -0.0216 -0.9243>}
 FACIAL_L_NasolabialFurrow {parent=FACIAL_C_FacialRoot, loc=<2.7188 3.4221 8.4841>, rot=<-0.0386 -0.1237 -0.3643 -0.9222>}
 FACIAL_R_NasolabialFurrow {parent=FACIAL_C_FacialRoot, loc=<-2.7196 3.4213 8.4832>, rot=<-0.0386 0.1237 0.3644 -0.9222>}
 FACIAL_L_CheekLower {parent=FACIAL_C_FacialRoot, loc=<3.4056 4.3172 5.3527>, rot=<0.0294 -0.4272 0.0553 -0.902>}
 FACIAL_R_CheekLower {parent=FACIAL_C_FacialRoot, loc=<-3.4069 4.3182 5.35>, rot=<0.0294 0.4272 -0.0553 -0.902>}
 FACIAL_C_Nose {parent=FACIAL_C_FacialRoot, loc=<-0.0007 0.6564 9.2655>, rot=<-0.1361 0 0 -0.9907>}
 FACIAL_C_NoseLower {parent=FACIAL_C_Nose, loc=<0.0002 2.6158 -0.377>, rot=<0.1305 0 0 -0.9914>}
 FACIAL_L_Nostril {parent=FACIAL_C_Nose, loc=<0.8582 2.0803 -0.3389>, rot=<0.0173 0.1305 0.0023 -0.9913>}
 FACIAL_R_Nostril {parent=FACIAL_C_Nose, loc=<-0.8563 2.0813 -0.3429>, rot=<0.0173 -0.1305 -0.0023 -0.9913>}
 FACIAL_C_NoseTip {parent=FACIAL_C_Nose, loc=<-0.0003 1.973 0.5252>, rot=<0.0872 0 0 -0.9962>}
 FACIAL_C_MouthUpper {parent=FACIAL_C_FacialRoot, loc=<-0.0017 4.7371 7.0688>, rot=<-0.0057 0 0 -1>}
 FACIAL_C_LipUpper {parent=FACIAL_C_MouthUpper, loc=<0.0005 -0.1655 1.9674>, rot=<-0.0872 0 0 -0.9962>}
 FACIAL_L_LipUpper {parent=FACIAL_C_MouthUpper, loc=<0.8885 -0.1208 1.8453>, rot=<-0.0778 -0.1601 -0.0379 -0.9833>}
 FACIAL_R_LipUpper {parent=FACIAL_C_MouthUpper, loc=<-0.8886 -0.1214 1.8448>, rot=<-0.0778 0.1601 0.0379 -0.9833>}
 FACIAL_L_LipUpperOuter {parent=FACIAL_C_MouthUpper, loc=<1.6302 0.0083 1.3419>, rot=<-0.0984 -0.2293 -0.0983 -0.9634>}
 FACIAL_R_LipUpperOuter {parent=FACIAL_C_MouthUpper, loc=<-1.63 0.0085 1.339>, rot=<-0.0984 0.2293 0.0983 -0.9634>}
 FACIAL_L_LipCorner {parent=FACIAL_C_MouthUpper, loc=<2.2279 0.3866 0.6858>, rot=<0 -0.2078 0 -0.9782>}
 FACIAL_R_LipCorner {parent=FACIAL_C_MouthUpper, loc=<-2.2287 0.3872 0.6868>, rot=<0 0.2078 0 -0.9782>}
 FACIAL_L_Masseter {parent=FACIAL_C_FacialRoot, loc=<5.8947 1.9119 3.8935>, rot=<0.0637 0.1129 0.0086 -0.9915>}
 FACIAL_R_Masseter {parent=FACIAL_C_FacialRoot, loc=<-5.8955 1.9136 3.8914>, rot=<0.0636 -0.1128 -0.0085 -0.9915>}
 FACIAL_C_UnderChin {parent=FACIAL_C_FacialRoot, loc=<0 8.8679 3.9382>, rot=<0.4949 0 0 -0.869>}
 FACIAL_L_UnderChin {parent=FACIAL_C_FacialRoot, loc=<2.7204 7.6459 2.7899>, rot=<0.3977 -0.3553 0.0803 -0.8421>}
 FACIAL_R_UnderChin {parent=FACIAL_C_FacialRoot, loc=<-2.7213 7.6463 2.7896>, rot=<0.3977 0.3553 -0.0802 -0.8421>}
 FACIAL_C_TeethUpper {parent=FACIAL_C_FacialRoot, loc=<-0.003 5.1341 7.9669>, rot=<-0.0057 0 0 -1>}
 FACIAL_C_LowerLipRotation {parent=FACIAL_C_FacialRoot, loc=<0.0012 2.6819 2.1495>, rot=<0.3366 0 0 -0.9416>}
 FACIAL_C_MouthLower {parent=FACIAL_C_LowerLipRotation, loc=<-0.0027 -1.5294 5.1073>, rot=<-0.342 0 0 -0.9397>}
 FACIAL_C_LipLower {parent=FACIAL_C_MouthLower, loc=<-0.0016 0.9423 1.6917>, rot=<0.0872 0 0 -0.9962>}
 FACIAL_L_LipLower {parent=FACIAL_C_MouthLower, loc=<0.7999 0.9254 1.5331>, rot=<0.0927 -0.1294 0.0272 -0.9869>}
 FACIAL_R_LipLower {parent=FACIAL_C_MouthLower, loc=<-0.8038 0.9264 1.5309>, rot=<0.0927 0.1294 -0.0272 -0.9869>}
 FACIAL_L_LipLowerOuter {parent=FACIAL_C_MouthLower, loc=<1.5324 0.7667 1.0904>, rot=<0.0913 -0.2321 0.0318 -0.9679>}
 FACIAL_R_LipLowerOuter {parent=FACIAL_C_MouthLower, loc=<-1.5347 0.7668 1.0882>, rot=<0.0913 0.2321 -0.0318 -0.9679>}
 FACIAL_C_TeethLower {parent=FACIAL_C_LowerLipRotation, loc=<-0.0039 -1.7131 6.0386>, rot=<-0.342 0 0 -0.9397>}
 FACIAL_C_Tongue1 {parent=FACIAL_C_TeethLower, loc=<0.0034 0.1773 -4.6246>, rot=<-0.2588 0 0 -0.9659>}
 FACIAL_C_Tongue2 {parent=FACIAL_C_Tongue1, loc=<-0.0008 0.3914 1.3319>, rot=<0.2588 0 0 -0.9659>}
 FACIAL_C_Tongue3 {parent=FACIAL_C_Tongue2, loc=<-0.0007 0.1987 1.2451>, rot=<0.1305 0 0 -0.9914>}
 FACIAL_C_Tongue4 {parent=FACIAL_C_Tongue3, loc=<-0.0001 -0.1744 1.0793>, rot=<0.0436 0 0 -0.999>}
 FACIAL_C_TongueUpper3 {parent=FACIAL_C_Tongue3, loc=<0.0006 -0.3192 -0.1919>, rot=<0 0 0 1>}
 FACIAL_C_TongueLower3 {parent=FACIAL_C_Tongue3, loc=<0 0.8165 0.0286>, rot=<0 0 0 1>}
 FACIAL_L_TongueSide3 {parent=FACIAL_C_Tongue3, loc=<1.4976 0.0025 0.1017>, rot=<0 0 0 1>}
 FACIAL_R_TongueSide3 {parent=FACIAL_C_Tongue3, loc=<-1.4977 0.0034 0.1>, rot=<0 0 0 1>}
 FACIAL_C_TongueUpper2 {parent=FACIAL_C_Tongue2, loc=<0.0007 -0.5628 -0.2339>, rot=<0 0 0 1>}
 FACIAL_L_TongueSide2 {parent=FACIAL_C_Tongue2, loc=<1.7793 0.0005 0.122>, rot=<0 0 0 1>}
 FACIAL_R_TongueSide2 {parent=FACIAL_C_Tongue2, loc=<-1.7795 0.0011 0.12>, rot=<0 0 0 1>}
 FACIAL_C_Jaw {parent=FACIAL_C_FacialRoot, loc=<0.0012 2.6819 2.1495>, rot=<0.3366 0 0 -0.9416>}
 FACIAL_C_Jawline {parent=FACIAL_C_Jaw, loc=<-0.0017 2.2167 6.9396>, rot=<0.2586 0 0 -0.966>}
 FACIAL_L_ChinSide {parent=FACIAL_C_Jaw, loc=<2.4796 0.2882 6.7508>, rot=<-0.2208 -0.3738 0.0499 -0.8995>}
 FACIAL_R_ChinSide {parent=FACIAL_C_Jaw, loc=<-2.4812 0.288 6.7512>, rot=<-0.2208 0.3738 -0.0499 -0.8995>}
 FACIAL_C_Chin {parent=FACIAL_C_Jaw, loc=<-0.0017 0.7893 7.1548>, rot=<-0.2591 0 0 -0.9659>}
 FACIAL_L_Ear {parent=FACIAL_C_FacialRoot, loc=<6.2834 -0.0974 0.7791>, rot=<-0.032 -0.0005 -0.0261 -0.9991>}
 FACIAL_R_Ear {parent=FACIAL_C_FacialRoot, loc=<-6.2817 -0.0954 0.7736>, rot=<-0.032 0.0006 0.0262 -0.9991>}
 FACIAL_L_Jawline {parent=FACIAL_C_FacialRoot, loc=<4.0158 5.778 4.3247>, rot=<0.2038 0.1581 -0.0537 -0.9647>}
 FACIAL_R_Jawline {parent=FACIAL_C_FacialRoot, loc=<-4.0183 5.7792 4.3243>, rot=<0.2038 -0.158 0.0538 -0.9647>}
 FACIAL_C_NoseBridge {parent=FACIAL_C_FacialRoot, loc=<-0.0005 -1.3768 9.6101>, rot=<0.0381 0 0 -0.9993>}
 FACIAL_L_EyelidUpperB {parent=FACIAL_C_FacialRoot, loc=<2.7937 -1.0337 6.7665>, rot=<-0.2306 0 0 -0.9731>}
 FACIAL_L_EyelidLowerB {parent=FACIAL_C_FacialRoot, loc=<2.7937 -1.0337 6.7665>, rot=<0.3118 0 0 -0.9501>}
 FACIAL_R_EyelidLowerB {parent=FACIAL_C_FacialRoot, loc=<-2.7922 -1.0348 6.7696>, rot=<0.3201 0 0 -0.9474>}
 FACIAL_R_EyelidUpperB {parent=FACIAL_C_FacialRoot, loc=<-2.7922 -1.0348 6.7696>, rot=<-0.2306 0 0 -0.9731>}
 clavicle_l {parent=spine_05, loc=<5.4831 -0.1053 -3.2946>, rot=<0.0465 -0.7289 -0.009 -0.6829>}
 upperarm_l {parent=clavicle_l, loc=<9.8658 -1.2136 -0.4105>, rot=<0.0164 -0.3245 -0.0269 -0.9454>}
 lowerarm_l {parent=upperarm_l, loc=<26.465 -0.8345 0.259>, rot=<0.0036 0.0161 -0.2298 -0.9731>}
 hand_l {parent=lowerarm_l, loc=<22.7355 -0.3203 -0.6105>, rot=<-0.6784 -0.0223 0.0735 -0.7307>}
 thumb_01_l {parent=hand_l, loc=<1.3118 -0.3191 -2.459>, rot=<0.6546 -0.4042 0.0664 -0.6354>}
 thumb_02_l {parent=thumb_01_l, loc=<5.9336 -0.0448 0.0401>, rot=<-0.0037 0.0213 0.298 -0.9543>}
 thumb_03_l {parent=thumb_02_l, loc=<2.2223 0.3682 -0.0907>, rot=<0.0118 -0.013 0.0066 -0.9998>}
 index_metacarpal_l {parent=hand_l, loc=<1.6535 0.0183 -2.96>, rot=<0.0158 -0.0124 -0.0173 -0.9996>}
 index_01_l {parent=index_metacarpal_l, loc=<8.3008 0.0357 -0.0524>, rot=<-0.0044 0.0485 0.0924 -0.9945>}
 index_02_l {parent=index_01_l, loc=<3.3587 -0.3004 0.0452>, rot=<-0.002 0.0027 0.3225 -0.9466>}
 index_03_l {parent=index_02_l, loc=<2.5578 -0.123 -0.0252>, rot=<-0.002 0.0067 0.2558 -0.9667>}
 middle_metacarpal_l {parent=hand_l, loc=<2.1856 -0.1725 -0.5984>, rot=<0.006 0.0168 -0.0329 -0.9993>}
 middle_01_l {parent=middle_metacarpal_l, loc=<7.3275 -0.2447 0.0129>, rot=<-0.0208 0.0042 0.0858 -0.9961>}
 middle_02_l {parent=middle_01_l, loc=<4.0123 0.1401 0.1175>, rot=<-0.0218 0.0258 0.3512 -0.9357>}
 middle_03_l {parent=middle_02_l, loc=<2.6077 0.0534 0.1373>, rot=<0.0082 0.0188 0.1708 -0.9851>}
 ring_metacarpal_l {parent=hand_l, loc=<2.2074 -0.3625 1.3638>, rot=<-0.0284 0.0722 -0.0458 -0.9959>}
 ring_01_l {parent=ring_metacarpal_l, loc=<7.0295 -0.6197 -0.1446>, rot=<-0.0489 -0.0252 0.1892 -0.9804>}
 ring_02_l {parent=ring_01_l, loc=<3.6174 0.4647 -0.1027>, rot=<0 0 0.2806 -0.9598>}
 ring_03_l {parent=ring_02_l, loc=<2.4152 0.114 0.0199>, rot=<0 0 0.1189 -0.9929>}
 pinky_metacarpal_l {parent=hand_l, loc=<1.7616 -0.9753 3.3814>, rot=<-0.0397 0.0814 -0.069 -0.9935>}
 pinky_01_l {parent=pinky_metacarpal_l, loc=<7.1709 -0.4831 -0.1094>, rot=<-0.0654 -0.0035 0.2242 -0.9723>}
 pinky_02_l {parent=pinky_01_l, loc=<2.8274 0.0893 -0.1043>, rot=<0.0037 0.0011 0.2854 -0.9584>}
 pinky_03_l {parent=pinky_02_l, loc=<2.2046 0.0307 -0.0078>, rot=<-0.0002 0.0098 0.1445 -0.9895>}
 deform_wrist_dn_base_l {parent=hand_l, loc=<0 -1.5326 0>, rot=<0 0 0 -1>}
 deform_wrist_dn_l {parent=deform_wrist_dn_base_l, loc=<0 0 0>, rot=<0 0 0 1>}
 deform_wrist_up_base_l {parent=hand_l, loc=<0 2.3009 0>, rot=<0 0 0 -1>}
 deform_wrist_up_l {parent=deform_wrist_up_base_l, loc=<0 0 0>, rot=<0 0 0 1>}
 lowerarm_twist_02_l {parent=lowerarm_l, loc=<10.6707 0 0>, rot=<0 0 0 1>}
 lowerarm_twist_01_l {parent=lowerarm_l, loc=<18.316 0 0>, rot=<0 0 0 1>}
 deform_elbow_in_base_l {parent=lowerarm_l, loc=<0.1834 3.6062 -0.2572>, rot=<0 0 0 -1>}
 deform_elbow_in_l {parent=deform_elbow_in_base_l, loc=<0 0 0>, rot=<0 0 0 1>}
 deform_elbow_out_base_l {parent=lowerarm_l, loc=<0 -4 0>, rot=<0 0 0 1>}
 deform_elbow_out_l {parent=deform_elbow_out_base_l, loc=<0 0 0>, rot=<0 0 0 1>}
 Dyn_sleeve_Down_L {parent=lowerarm_l, loc=<3.9686 -1.17 -6.591>, rot=<0.992 -0.0213 -0.0192 0.123>}
 Dyn_sleeve_In_L {parent=lowerarm_l, loc=<4.9169 4.9093 -2.9854>, rot=<0.7889 -0.0129 0.0098 -0.6143>}
 Dyn_sleeve_Up_L {parent=lowerarm_l, loc=<4.2508 1.603 3.6845>, rot=<0.1209 0.0129 0.0345 -0.992>}
 Dyn_sleeve_Out_L {parent=lowerarm_l, loc=<3.2904 -4.7531 0.068>, rot=<-0.6638 0.0417 0.0196 -0.7465>}
 upperarm_twist_01_l {parent=upperarm_l, loc=<7.2135 -0.1134 0>, rot=<0 0 0 1>}
 deform_upperarm_bk_base_l {parent=upperarm_twist_01_l, loc=<-6.4053 -4.3746 0.314>, rot=<0 0 0 -1>}
 deform_upperarm_bk_l {parent=deform_upperarm_bk_base_l, loc=<0 0 0>, rot=<0 0 0 1>}
 deform_upperarm_fr_base_l {parent=upperarm_twist_01_l, loc=<-5.6312 5.092 0.3354>, rot=<0 0 0 -1>}
 deform_upperarm_fr_l {parent=deform_upperarm_fr_base_l, loc=<0 0 0>, rot=<0 0 0 1>}
 deform_upperarm_up_base_l {parent=upperarm_twist_01_l, loc=<-8.891 0.4825 3.9517>, rot=<0 0 0 -1>}
 deform_upperarm_up_l {parent=deform_upperarm_up_base_l, loc=<0 0 0>, rot=<0 0 0 1>}
 Base_Cape_L {parent=upperarm_twist_01_l, loc=<1.977 -7.5974 -1.345>, rot=<-0.4161 -0.48 0.5047 -0.5845>}
 Dyn_Cape_L {parent=Base_Cape_L, loc=<0.2981 0 0>, rot=<0 0 0 1>}
 upperarm_twist_02_l {parent=upperarm_l, loc=<19.7998 -0.1134 0>, rot=<0 0 0 1>}
 clavicle_r {parent=spine_05, loc=<5.4831 -0.1053 3.2946>, rot=<0.7289 0.0465 -0.6829 0.009>}
 upperarm_r {parent=clavicle_r, loc=<-9.8658 1.2136 0.4105>, rot=<0.0164 -0.3245 -0.0269 -0.9454>}
 lowerarm_r {parent=upperarm_r, loc=<-26.465 0.8345 -0.259>, rot=<0.0036 0.0161 -0.2298 -0.9731>}
 hand_r {parent=lowerarm_r, loc=<-22.7355 0.3203 0.6106>, rot=<-0.6784 -0.0223 0.0735 -0.7307>}
 thumb_01_r {parent=hand_r, loc=<-1.3118 0.3191 2.459>, rot=<0.6546 -0.4042 0.0664 -0.6354>}
 thumb_02_r {parent=thumb_01_r, loc=<-5.9336 0.0448 -0.0401>, rot=<-0.0037 0.0213 0.298 -0.9543>}
 thumb_03_r {parent=thumb_02_r, loc=<-2.2223 -0.3682 0.0907>, rot=<0.0118 -0.013 0.0066 -0.9998>}
 index_metacarpal_r {parent=hand_r, loc=<-1.6535 -0.0183 2.96>, rot=<0.0158 -0.0124 -0.0173 -0.9996>}
 index_01_r {parent=index_metacarpal_r, loc=<-8.3008 -0.0357 0.0524>, rot=<-0.0044 0.0485 0.0924 -0.9945>}
 index_02_r {parent=index_01_r, loc=<-3.3587 0.3004 -0.0452>, rot=<-0.002 0.0027 0.3225 -0.9466>}
 index_03_r {parent=index_02_r, loc=<-2.5578 0.123 0.0252>, rot=<-0.002 0.0067 0.2558 -0.9667>}
 middle_metacarpal_r {parent=hand_r, loc=<-2.1856 0.1725 0.5984>, rot=<0.006 0.0168 -0.0329 -0.9993>}
 middle_01_r {parent=middle_metacarpal_r, loc=<-7.3275 0.2447 -0.0129>, rot=<-0.0208 0.0042 0.0858 -0.9961>}
 middle_02_r {parent=middle_01_r, loc=<-4.0123 -0.1401 -0.1175>, rot=<-0.0218 0.0258 0.3512 -0.9357>}
 middle_03_r {parent=middle_02_r, loc=<-2.6077 -0.0534 -0.1373>, rot=<0.0082 0.0188 0.1708 -0.9851>}
 ring_metacarpal_r {parent=hand_r, loc=<-2.2074 0.3625 -1.3638>, rot=<-0.0284 0.0722 -0.0458 -0.9959>}
 ring_01_r {parent=ring_metacarpal_r, loc=<-7.0295 0.6197 0.1446>, rot=<-0.0489 -0.0252 0.1892 -0.9804>}
 ring_02_r {parent=ring_01_r, loc=<-3.6174 -0.4647 0.1027>, rot=<0 0 0.2806 -0.9598>}
 ring_03_r {parent=ring_02_r, loc=<-2.4152 -0.114 -0.0199>, rot=<0 0 0.1189 -0.9929>}
 pinky_metacarpal_r {parent=hand_r, loc=<-1.7616 0.9753 -3.3814>, rot=<-0.0397 0.0814 -0.069 -0.9935>}
 pinky_01_r {parent=pinky_metacarpal_r, loc=<-7.1709 0.4831 0.1094>, rot=<-0.0654 -0.0035 0.2242 -0.9723>}
 pinky_02_r {parent=pinky_01_r, loc=<-2.8274 -0.0893 0.1043>, rot=<0.0037 0.0011 0.2854 -0.9584>}
 pinky_03_r {parent=pinky_02_r, loc=<-2.2046 -0.0307 0.0078>, rot=<-0.0002 0.0098 0.1445 -0.9895>}
 deform_wrist_dn_base_r {parent=hand_r, loc=<-0.0001 1.5328 0>, rot=<0 0 0 -1>}
 deform_wrist_dn_r {parent=deform_wrist_dn_base_r, loc=<0 0 0>, rot=<0 0 0 1>}
 deform_wrist_up_base_r {parent=hand_r, loc=<-0.0002 -2.3007 -0.0001>, rot=<0 0 0 -1>}
 deform_wrist_up_r {parent=deform_wrist_up_base_r, loc=<0 0 0>, rot=<0 0 0 1>}
 lowerarm_twist_02_r {parent=lowerarm_r, loc=<-10.6707 0 0>, rot=<0 0 0 -1>}
 lowerarm_twist_01_r {parent=lowerarm_r, loc=<-18.316 0 0>, rot=<0 0 0 -1>}
 deform_elbow_in_base_r {parent=lowerarm_r, loc=<-0.1834 -3.6062 0.2572>, rot=<0 0 0 -1>}
 deform_elbow_in_r {parent=deform_elbow_in_base_r, loc=<0 0 0>, rot=<0 0 0 1>}
 deform_elbow_out_base_r {parent=lowerarm_r, loc=<0 4 -0.0001>, rot=<0 0 0 -1>}
 deform_elbow_out_r {parent=deform_elbow_out_base_r, loc=<0 0 0>, rot=<0 0 0 1>}
 Dyn_sleeve_Out_R {parent=lowerarm_r, loc=<-3.2904 4.7531 -0.0679>, rot=<-0.0196 -0.7465 -0.6638 -0.0417>}
 Dyn_sleeve_Up_R {parent=lowerarm_r, loc=<-4.251 -1.603 -3.6842>, rot=<-0.0345 -0.992 0.1209 -0.0129>}
 Dyn_sleeve_In_R {parent=lowerarm_r, loc=<-4.9168 -4.9093 2.9854>, rot=<-0.0098 -0.6143 0.7889 0.0129>}
 Dyn_sleeve_Down_R {parent=lowerarm_r, loc=<-3.9687 1.1701 6.5911>, rot=<0.0192 0.123 0.992 0.0213>}
 upperarm_twist_01_r {parent=upperarm_r, loc=<-7.2135 0.1134 0>, rot=<0 0 0 -1>}
 deform_upperarm_bk_base_r {parent=upperarm_twist_01_r, loc=<6.4053 4.3746 -0.314>, rot=<0 0 0 -1>}
 deform_upperarm_bk_r {parent=deform_upperarm_bk_base_r, loc=<0 0 0>, rot=<0 0 0 1>}
 deform_upperarm_fr_base_r {parent=upperarm_twist_01_r, loc=<5.6312 -5.092 -0.3354>, rot=<0 0 0 -1>}
 deform_upperarm_fr_r {parent=deform_upperarm_fr_base_r, loc=<0 0 0>, rot=<0 0 0 1>}
 deform_upperarm_up_base_r {parent=upperarm_twist_01_r, loc=<8.891 -0.4825 -3.9517>, rot=<0 0 0 -1>}
 deform_upperarm_up_r {parent=deform_upperarm_up_base_r, loc=<0 0 0>, rot=<0 0 0 1>}
 Base_Cape_R {parent=upperarm_twist_01_r, loc=<-1.9773 7.5974 1.3454>, rot=<-0.5047 -0.5845 -0.4161 0.48>}
 Dyn_Cape_R {parent=Base_Cape_R, loc=<0.2981 0 0>, rot=<0 0 0 1>}
 upperarm_twist_02_r {parent=upperarm_r, loc=<-19.7998 0.1134 0>, rot=<0 0 0 -1>}
 attach_backpack {parent=spine_05, loc=<-0.123 -13.473 0>, rot=<-0.0093 -0.707 0.0093 -0.707>}
 Backpack_BR {parent=attach_backpack, type=Null, loc=<0 0.285 0>, rot=<0 0 0 1>}
 attach_cape {parent=spine_05, loc=<9.017 -11.714 0>, rot=<0.0616 -0.7044 -0.0616 -0.7044>}
 cape {parent=attach_cape, type=Null, loc=<0 2.8496 0>, rot=<0 0 0 1>}
 Base_string {parent=spine_05, loc=<2.3659 9.6567 -3.0672>, rot=<-0.161 -0.6927 -0.684 -0.1624>}
 Dyn_string_01 {parent=Base_string, loc=<0.1 0 0>, rot=<0 0 0 1>}
 Dyn_string_02 {parent=Dyn_string_01, loc=<4.0068 0 0>, rot=<0 0 0 1>}
 deform_pec_aim_base_l {parent=spine_04, loc=<3.2456 9.3309 -1.2237>, rot=<0.0603 -0.4033 0.3318 -0.8507>}
 deform_pec_aim_l {parent=deform_pec_aim_base_l, loc=<0 0 0>, rot=<0 0 0 1>}
 deform_pec_base_l {parent=deform_pec_aim_l, loc=<9.3665 0 0>, rot=<0 0 0 1>}
 deform_pec_l {parent=deform_pec_base_l, loc=<0 0 0>, rot=<0 0 0 1>}
 deform_pec_aim_base_r {parent=spine_04, loc=<3.2452 9.3309 1.2237>, rot=<-0.4033 -0.0603 0.8507 0.3318>}
 deform_pec_aim_r {parent=deform_pec_aim_base_r, loc=<0 0 0>, rot=<0 0 0 1>}
 deform_pec_base_r {parent=deform_pec_aim_r, loc=<-9.3667 -0.0003 -0.0004>, rot=<0 0 0 1>}
 deform_pec_r {parent=deform_pec_base_r, loc=<0 0 0>, rot=<0 0 0 1>}
 Base_Cape_C {parent=spine_04, loc=<7.1286 -10.9059 -0.0009>, rot=<0.0657 -0.704 0.704 -0.0657>}
 Dyn_Cape_C_01 {parent=Base_Cape_C, loc=<0.2 0 0>, rot=<0 0 0 1>}
 Dyn_Cape_Bk_01_L {parent=Dyn_Cape_C_01, loc=<4.0114 4.4624 -0.0251>, rot=<-0.0041 -0.0616 -0.1124 -0.9917>}
 Dyn_Cape_Bk_02_L {parent=Dyn_Cape_Bk_01_L, loc=<6.2881 -0.0089 0.0196>, rot=<0 0 0 1>}
 Dyn_Cape_Bk_01_R {parent=Dyn_Cape_C_01, loc=<4.0117 -5.6331 -0.0251>, rot=<0.0041 -0.0616 0.1124 -0.9917>}
 Dyn_Cape_Bk_02_R {parent=Dyn_Cape_Bk_01_R, loc=<6.288 0.0089 0.0197>, rot=<0 0 0 1>}
 Master_knife {parent=spine_03, loc=<-2.9137 5.8641 -8.3186>, rot=<0.2278 -0.9228 -0.2941 -0.1005>}
 Base_knife {parent=Master_knife, loc=<0 0 0>, rot=<0 0 0 1>}
 Dyn_knife_01 {parent=Base_knife, loc=<0.0574 2.0187 1.3605>, rot=<0.1616 -0.0211 0.0125 -0.9865>}
 Dyn_knife_03 {parent=Dyn_knife_01, loc=<0.2 0 0>, rot=<0 0 0 1>}
 Dyn_knife_02 {parent=Base_knife, loc=<0.013 -2.3625 1.6795>, rot=<0.0484 -0.0065 -0.0089 -0.9988>}
 Dyn_knife_04 {parent=Dyn_knife_02, loc=<0.187 0.0007 -0.0005>, rot=<0 0 0 1>}
 Master_rope {parent=spine_01, loc=<1.7879 5.9326 9.4297>, rot=<0.1121 0.2854 0.9443 -0.1196>}
 Base_rope {parent=Master_rope, loc=<0 0 0>, rot=<0 0 0 1>}
 Dyn_rope_01 {parent=Base_rope, loc=<0 0 0>, rot=<0 0 0 1>}
 Dyn_rope_02 {parent=Dyn_rope_01, loc=<0.4065 -1.0868 -0.348>, rot=<0.0026 -0.039 0.2616 -0.9644>}
 Dyn_rope_03 {parent=Dyn_rope_02, loc=<3.5587 0 0>, rot=<-0.1413 0.0795 -0.2228 -0.9613>}
 Dyn_rope_04 {parent=Dyn_rope_03, loc=<3.4251 0 0>, rot=<-0.0572 0.2691 -0.0011 -0.9614>}
 Dyn_rope_05 {parent=Dyn_rope_01, loc=<0.2272 1.3541 0.0238>, rot=<-0.0622 -0.0306 -0.2139 -0.9744>}
 Dyn_rope_06 {parent=Dyn_rope_05, loc=<3.5208 0 0>, rot=<0.1401 0.077 0.0023 -0.9871>}
 Dyn_rope_07 {parent=Dyn_rope_06, loc=<3.4255 0 0>, rot=<-0.0016 0.0117 0.0855 -0.9963>}
 Dyn_rope_08 {parent=Dyn_rope_07, loc=<3.5567 0 0>, rot=<0.0191 -0.0091 0.0899 -0.9957>}
 thigh_l {parent=pelvis, loc=<-1.0855 -0.2109 -8.7053>, rot=<0.0009 0.0355 0.0155 -0.9993>}
 calf_l {parent=thigh_l, loc=<-42.5284 -0.0019 -0.7297>, rot=<0.0006 -0.0161 -0.0391 -0.9991>}
 foot_l {parent=calf_l, loc=<-46.7401 -2.5014 -0.592>, rot=<0.0427 -0.0238 0.0353 -0.9982>}
 ball_l {parent=foot_l, loc=<-3.5658 10.8066 -0.1544>, rot=<0.0001 -0.0001 -0.7063 -0.7079>}
 Correct_foot_L {parent=foot_l, loc=<1.8234 -0.2521 0.596>, rot=<0.5056 -0.4919 0.5132 -0.4889>}
 calf_twist_02_l {parent=calf_l, loc=<-19.1546 -1.7943 -0.2638>, rot=<0.0237 0.0095 -0.0035 -0.9997>}
 calf_twist_01_l {parent=calf_l, loc=<-35.8483 -1.9197 -0.5794>, rot=<0.0237 0.0095 -0.0035 -0.9997>}
 deform_knee_in_base_l {parent=calf_l, loc=<0.8031 -5.0231 0.2456>, rot=<0 0 0 -1>}
 deform_knee_in_l {parent=deform_knee_in_base_l, loc=<0 0 0>, rot=<0 0 0 1>}
 deform_knee_out_base_l {parent=calf_l, loc=<0 4 0>, rot=<0 0 0 -1>}
 deform_knee_out_l {parent=deform_knee_out_base_l, loc=<0 0 0>, rot=<0 0 0 1>}
 thigh_twist_01_l {parent=thigh_l, loc=<-27.952 0 0>, rot=<0.0013 0 0 -1>}
 deform_groin_base_l {parent=thigh_l, loc=<0.4801 5.6776 -0.2486>, rot=<0 0 0 -1>}
 deform_groin_l {parent=deform_groin_base_l, loc=<0 0 0>, rot=<0 0 0 1>}
 Base_skirt_Fr_L {parent=thigh_l, loc=<1.4759 5.8695 0.354>, rot=<-0.1094 -0.8015 -0.586 -0.0466>}
 Dyn_skirt_Fr_01_L {parent=Base_skirt_Fr_L, loc=<0.2 0 0>, rot=<0 0 0 1>}
 Dyn_skirt_Fr_02_L {parent=Dyn_skirt_Fr_01_L, loc=<6.4211 0 0>, rot=<-0.0001 -0.0124 0.0109 -0.9999>}
 Dyn_skirt_Fr_03_L {parent=Dyn_skirt_Fr_02_L, loc=<5.4298 0 0>, rot=<-0.0025 -0.0354 0.0713 -0.9968>}
 Base_skirt_Side_L {parent=thigh_l, loc=<1.0179 -0.9065 -4.5554>, rot=<0.0248 -0.9935 0.0184 -0.1092>}
 Dyn_skirt_Side_01_L {parent=Base_skirt_Side_L, loc=<0.1453 0 0>, rot=<0 0 0 1>}
 Dyn_skirt_Side_02_L {parent=Dyn_skirt_Side_01_L, loc=<6.6749 0 0.0001>, rot=<0 0.0545 0 -0.9985>}
 Dyn_skirt_Side_03_L {parent=Dyn_skirt_Side_02_L, loc=<6.7404 0 -0.0001>, rot=<0 0 0 1>}
 Base_skirt_Bk_L {parent=thigh_l, loc=<-5.0631 -10.5558 0.4657>, rot=<0.021 -0.7786 0.6268 -0.0208>}
 Dyn_skirt_Bk_01_L {parent=Base_skirt_Bk_L, loc=<0.1 0 0>, rot=<0 0 0 1>}
 Dyn_skirt_Bk_02_L {parent=Dyn_skirt_Bk_01_L, loc=<6.7404 0 -0.0001>, rot=<0 0 0 1>}
 thigh_r {parent=pelvis, loc=<-1.0855 -0.2109 8.7053>, rot=<0.0355 -0.0009 0.9993 0.0155>}
 calf_r {parent=thigh_r, loc=<42.5284 0.0019 0.7297>, rot=<0.0006 -0.0161 -0.0391 -0.9991>}
 foot_r {parent=calf_r, loc=<46.7401 2.5014 0.592>, rot=<0.0427 -0.0238 0.0353 -0.9982>}
 ball_r {parent=foot_r, loc=<3.5658 -10.8066 0.1544>, rot=<0.0001 -0.0001 -0.7063 -0.7079>}
 calf_twist_02_r {parent=calf_r, loc=<19.1546 1.7943 0.2638>, rot=<0.0237 0.0095 -0.0035 -0.9997>}
 calf_twist_01_r {parent=calf_r, loc=<35.8483 1.9197 0.5794>, rot=<0.0237 0.0095 -0.0035 -0.9997>}
 Correct_foot_R {parent=calf_twist_01_r, loc=<9.0572 0.8786 -0.6515>, rot=<0.5231 0.4421 -0.5133 -0.5172>}
 deform_knee_in_base_r {parent=calf_r, loc=<-0.8031 5.0231 -0.2457>, rot=<0 0 0 -1>}
 deform_knee_in_r {parent=deform_knee_in_base_r, loc=<0 0 0>, rot=<0 0 0 1>}
 deform_knee_out_base_r {parent=calf_r, loc=<0 -4 0>, rot=<0 0 0 -1>}
 deform_knee_out_r {parent=deform_knee_out_base_r, loc=<0 0 0>, rot=<0 0 0 1>}
 thigh_twist_01_r {parent=thigh_r, loc=<27.952 0 0>, rot=<0.0013 0 0 -1>}
 deform_groin_base_r {parent=thigh_r, loc=<-0.4801 -5.6776 0.2486>, rot=<0 0 0 -1>}
 deform_groin_r {parent=deform_groin_base_r, loc=<0 0 0>, rot=<0 0 0 1>}
 Base_skirt_Fr_R {parent=thigh_r, loc=<-1.4759 -5.8695 -0.354>, rot=<-0.679 0.0881 0.0466 -0.7273>}
 Dyn_skirt_Fr_01_R {parent=Base_skirt_Fr_R, loc=<0.2 0 0>, rot=<0 0 0 1>}
 Dyn_skirt_Fr_02_R {parent=Dyn_skirt_Fr_01_R, loc=<6.4211 0 0>, rot=<0 0 0 1>}
 Dyn_skirt_Fr_03_R {parent=Dyn_skirt_Fr_02_R, loc=<5.4298 0 0>, rot=<0 0 0 1>}
 Base_skirt_Side_R {parent=thigh_r, loc=<-1.3229 0.1875 4.3921>, rot=<-0.013 0.1297 0.0081 -0.9914>}
 Dyn_skirt_Side_01_R {parent=Base_skirt_Side_R, loc=<0.2 0 0>, rot=<0 0 0 1>}
 Dyn_skirt_Side_02_R {parent=Dyn_skirt_Side_01_R, loc=<6.6749 0 0>, rot=<0 0 0 1>}
 Dyn_skirt_Side_03_R {parent=Dyn_skirt_Side_02_R, loc=<6.7404 0 0>, rot=<0 0 0 1>}
 Base_skirt_Bk_R {parent=thigh_r, loc=<5.0632 10.5558 -0.4657>, rot=<0.6268 0.0208 -0.021 -0.7786>}
 Dyn_skirt_Bk_01_R {parent=Base_skirt_Bk_R, loc=<0.1 0 0>, rot=<0 0 0 1>}
 Dyn_skirt_Bk_02_R {parent=Dyn_skirt_Bk_01_R, loc=<6.7404 0 0>, rot=<0 0 0 1>}
 deform_glute_base_l {parent=pelvis, loc=<-7.3466 -9.0213 -9.1245>, rot=<0.0172 0.031 -0.4477 -0.8935>}
 deform_glute_l {parent=deform_glute_base_l, loc=<0 0 0>, rot=<0 0 0 1>}
 deform_glute_base_r {parent=pelvis, loc=<-7.3464 -9.0213 9.1245>, rot=<0.031 -0.0172 0.8935 -0.4477>}
 deform_glute_r {parent=deform_glute_base_r, loc=<0 0 0>, rot=<0 0 0 1>}
]
variables [
 Crouch {type=double, default=0}
 DBNO {type=bool, default=False}
 FrontEndOn {type=bool, default=False}
 Gauntlets {type=bool, default=False}
 Lod2 {type=bool, default=False}
 No stand {type=double, default=0}
 NotRban {type=bool, default=False}
 PlayingEmote {type=bool, default=False}
 Sit Lower {type=double, default=0}
 Stand {type=double, default=0}
 Velocity {type=double, default=0}
]
