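# Physics asset for the F_MED_ToadLeaf female body cosmetic, apparently its frontend variant, covering a dangling rope and right-side skirt bones.
# Capsule rope bodies hang from spine_01 via Master and Base bones, branching at Dyn_rope_01 into two strands with swing limits that tighten toward the tips.
# Box bodies form front and side right-skirt chains off thigh_r, with free angular constraints, as on the pelvis, spine and thigh capsules.
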PhysicsAsset /BRCosmetics/Characters/Player/Female/Medium/Bodies/F_MED_ToadLeaf/Meshes/F_MED_ToadLeaf_Fronend_Physics
bodies {
 Base_rope [{shapeType=Capsule, length=0.5, radius=0.5, shapeName=Base_rope_capsule}]
 Dyn_rope_01 [{shapeType=Capsule, length=0.5571, radius=3.8107, shapeName=Dyn_rope_01_capsule, center=<3.4 0 3.6846>, rotation=<rot 90 0 0>}]
 Dyn_rope_02 [{shapeType=Capsule, length=2.5581, radius=1.4419, shapeName=Dyn_rope_02_capsule, center=<3 0 0.4>, rotation=<rot 90 0 0>}]
 Dyn_rope_03 [{shapeType=Capsule, length=2.3166, radius=1.6834, shapeName=Dyn_rope_03_capsule, center=<3 0 0.3>, rotation=<rot 90 0 0>}]
 Dyn_rope_04 [{shapeType=Capsule, length=1.1691, radius=2.8309, shapeName=Dyn_rope_04_capsule, center=<3.2 0 0.7>, rotation=<rot 90 0 0>}]
 Dyn_rope_05 [{shapeType=Capsule, length=2.5863, radius=1.4137, shapeName=Dyn_rope_05_capsule, center=<3 0 0>, rotation=<rot 90 0 0>}]
 Dyn_rope_06 [{shapeType=Capsule, length=2.4234, radius=1.5766, shapeName=Dyn_rope_06_capsule, center=<3 0 0.2>, rotation=<rot 90 0 0>}]
 Dyn_rope_07 [{shapeType=Capsule, length=2, radius=2, shapeName=Dyn_rope_07_capsule, center=<3 0 0.5>, rotation=<rot 90 0 0>}]
 Dyn_rope_08 [{shapeType=Capsule, length=0.8078, radius=3.1922, shapeName=Dyn_rope_08_capsule, center=<3.4 0 1>, rotation=<rot 90 0 0>}]
 Dyn_skirt_Fr_01_R [
  {
   shapeType=Box
   extentX=10
   extentY=10.488
   extentZ=2.7564
   shapeName=Dyn_skirt_Fr_01_R_box
   center=<3.2874 -0.8513 -1.8255>
   rotation=<rot 0.0003 0 25>
  }
  {
   shapeType=Box
   extentX=7.5
   extentY=11.6536
   extentZ=3.5095
   shapeName=Dyn_skirt_Fr_01_R_box
   center=<3.2124 -8.6209 0.9554>
   rotation=<rot 0.0005 0 10.0001>
  }
 ]
 Dyn_skirt_Fr_02_R [
  {
   shapeType=Box
   extentX=10
   extentY=11.2131
   extentZ=3.4814
   shapeName=Dyn_skirt_Fr_02_R_box
   center=<3.2874 -0.8011 -1.7179>
   rotation=<rot 0.0003 0 25>
  }
  {
   shapeType=Box
   extentX=7.5
   extentY=11.6536
   extentZ=3.5095
   shapeName=Dyn_skirt_Fr_02_R_box
   center=<3.2124 -8.6755 1.3285>
   rotation=<rot 0.0005 0 10.0001>
  }
 ]
 Dyn_skirt_Fr_03_R [{
   shapeType=Box
   extentX=10
   extentY=12.664
   extentZ=2.7564
   shapeName=Dyn_skirt_Fr_03_R_box
   center=<3.2618 -0.2866 -0.6145>
   rotation=<rot 0.0004 0 25.0001>
  }]
 Dyn_skirt_Side_01_R [{
   shapeType=Box
   extentX=10
   extentY=10
   extentZ=3.4734
   shapeName=Dyn_skirt_Side_01_R_box
   center=<3.9489 0.1608 -1.8381>
   rotation=<rot 0 -0.0001 -5>
  }]
 Dyn_skirt_Side_02_R [{
   shapeType=Box
   extentX=10
   extentY=10.6004
   extentZ=4.5308
   shapeName=Dyn_skirt_Side_02_R_box
   center=<3.9489 0.1608 -1.8381>
   rotation=<rot 0 0 -5>
  }]
 Dyn_skirt_Side_03_R [{
   shapeType=Box
   extentX=10
   extentY=11.7424
   extentZ=4.5308
   shapeName=Dyn_skirt_Side_03_R_box
   center=<3.9489 0.1608 -1.8381>
   rotation=<rot 0 0 -5>
  }]
 Master_rope [{shapeType=Capsule, length=0.5, radius=0.5, shapeName=Master_rope_capsule}]
 pelvis [{shapeType=Capsule, length=7.6591, radius=8.3409, shapeName=pelvis_capsule}]
 spine_01 [{shapeType=Capsule, length=8.6591, radius=6.8409, shapeName=spine_01_capsule}]
 thigh_r [{shapeType=Capsule, length=22.9898, radius=6.5102, shapeName=thigh_r_capsule, center=<14.7274 0 0>, rotation=<rot 90 0 0>}]
}
constraints [
 {
  bone1Name=Base_rope
  bone2Name=Master_rope
  swing1LimitDegrees=45
  swing1Motion=Free
  swing2LimitDegrees=45
  swing2Motion=Free
  twistLimitDegrees=45
  twistMotion=Free
 }
 {
  bone1Name=Master_rope
  bone2Name=spine_01
  swing1LimitDegrees=45
  swing1Motion=Free
  swing2LimitDegrees=45
  swing2Motion=Free
  twistLimitDegrees=45
  twistMotion=Free
 }
 {
  bone1Name=Dyn_skirt_Fr_01_R
  bone2Name=thigh_r
  swing1LimitDegrees=45
  swing1Motion=Free
  swing2LimitDegrees=45
  swing2Motion=Free
  twistLimitDegrees=45
  twistMotion=Free
 }
 {
  bone1Name=Dyn_skirt_Side_01_R
  bone2Name=thigh_r
  swing1LimitDegrees=45
  swing1Motion=Free
  swing2LimitDegrees=45
  swing2Motion=Free
  twistLimitDegrees=45
  twistMotion=Free
 }
 {
  bone1Name=Dyn_skirt_Side_02_R
  bone2Name=Dyn_skirt_Side_01_R
  swing1LimitDegrees=45
  swing1Motion=Free
  swing2LimitDegrees=45
  swing2Motion=Free
  twistLimitDegrees=45
  twistMotion=Free
 }
 {
  bone1Name=Dyn_skirt_Side_03_R
  bone2Name=Dyn_skirt_Side_02_R
  swing1LimitDegrees=45
  swing1Motion=Free
  swing2LimitDegrees=45
  swing2Motion=Free
  twistLimitDegrees=45
  twistMotion=Free
 }
 {
  bone1Name=Dyn_skirt_Fr_02_R
  bone2Name=Dyn_skirt_Fr_01_R
  swing1LimitDegrees=45
  swing1Motion=Free
  swing2LimitDegrees=45
  swing2Motion=Free
  twistLimitDegrees=45
  twistMotion=Free
 }
 {
  bone1Name=Dyn_skirt_Fr_03_R
  bone2Name=Dyn_skirt_Fr_02_R
  swing1LimitDegrees=45
  swing1Motion=Free
  swing2LimitDegrees=45
  swing2Motion=Free
  twistLimitDegrees=45
  twistMotion=Free
 }
 {
  bone1Name=spine_01
  bone2Name=pelvis
  swing1LimitDegrees=45
  swing1Motion=Free
  swing2LimitDegrees=45
  swing2Motion=Free
  twistLimitDegrees=45
  twistMotion=Free
 }
 {
  bone1Name=thigh_r
  bone2Name=pelvis
  swing1LimitDegrees=45
  swing1Motion=Free
  swing2LimitDegrees=45
  swing2Motion=Free
  twistLimitDegrees=45
  twistMotion=Free
 }
 {bone1Name=Dyn_rope_01, bone2Name=Base_rope, swing1LimitDegrees=45, swing2LimitDegrees=45, twistLimitDegrees=10}
 {bone1Name=Dyn_rope_02, bone2Name=Dyn_rope_01, swing1LimitDegrees=25, swing2LimitDegrees=30, twistLimitDegrees=5}
 {bone1Name=Dyn_rope_03, bone2Name=Dyn_rope_02, swing1LimitDegrees=20, swing2LimitDegrees=25, twistLimitDegrees=0}
 {bone1Name=Dyn_rope_04, bone2Name=Dyn_rope_03, swing1LimitDegrees=15, swing2LimitDegrees=25, twistLimitDegrees=0}
 {bone1Name=Dyn_rope_05, bone2Name=Dyn_rope_01, swing1LimitDegrees=40, swing2LimitDegrees=40, twistLimitDegrees=5}
 {bone1Name=Dyn_rope_06, bone2Name=Dyn_rope_05, swing1LimitDegrees=30, swing2LimitDegrees=30, twistLimitDegrees=0}
 {bone1Name=Dyn_rope_07, bone2Name=Dyn_rope_06, swing1LimitDegrees=25, swing2LimitDegrees=25, twistLimitDegrees=0}
 {bone1Name=Dyn_rope_08, bone2Name=Dyn_rope_07, swing1LimitDegrees=15, swing2LimitDegrees=15, twistLimitDegrees=0}
]
