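# Animation Blueprint for the F_MED_ToadLeaf female body cosmetic, adding simulated secondary motion to sleeve and skirt bones.
# Per-bone AnimDynamics chains use angular springs and gravity, in root-relative and world-space variants.
# RigidBody nodes simulate a frontend physics asset in pelvis space from a cached frontend pose, feeding the skirt chains.
# Spherical limits on the thighs constrain the middle side-skirt bones against the legs.
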AnimBlueprint /BRCosmetics/Characters/Player/Female/Medium/Bodies/F_MED_ToadLeaf/Meshes/F_MED_ToadLeaf_AnimBP
target_skeleton=&/F_MED_ToadLeaf_Skeleton
graphs [
 AnimGraph {nodes [
   n1 = Anim Dynamics - Bone: Dyn_sleeve_Down_L!(ComponentPose, bAlphaBoolEnabled:bool=True, ExternalForce="0,0,0") -> (Pose) {
    alphaInputType=Bool
    angularBiasOverride=0.9999
    angularDampingOverride=1
    angularSpringConstant=500
    bAngularSpring=true
    bLinearSpring=true
    bOverrideAngularDamping=true
    bOverrideLinearDamping=true
    bUseGravityOverride=true
    linearDampingOverride=0.9999
    linearSpringConstant=100
    simulationSpace=RootRelative
    alphaBoolBlend {blendOption=Linear}
    boundBone {boneName=Dyn_sleeve_Down_L}
    componentAppliedLinearAccClamp=<2000 2000 2000>
    componentLinearAccScale=<0.5 0.5 0.5>
    componentLinearVelScale=<0.5 0.5 0.5>
    gravityOverride {z=-980}
    physicsBodyDefinitions [{
      boundBone {boneName=Dyn_sleeve_Down_L}
      boxExtents=<10 10 10>
      constraintSetup {angularLimitsMax=<0 0 5>, angularLimitsMin=<0 -7 -5>, angularTarget=<1 0 0>, linearAxesMax=<1 0 1>, linearAxesMin=<-1 0 0>}
      localJointOffset=<1 0 0>
     }]
   }
   n2 = Anim Dynamics - Bone: Dyn_sleeve_Up_L!(ComponentPose=@n1, bAlphaBoolEnabled:bool=True, ExternalForce="0,0,0") -> (Pose) {
    alphaInputType=Bool
    angularBiasOverride=0.9999
    angularDampingOverride=1
    angularSpringConstant=500
    bAngularSpring=true
    bLinearSpring=true
    bOverrideAngularDamping=true
    bOverrideLinearDamping=true
    bUseGravityOverride=true
    linearDampingOverride=0.9999
    linearSpringConstant=100
    simulationSpace=RootRelative
    alphaBoolBlend {blendOption=Linear}
    boundBone {boneName=Dyn_sleeve_Up_L}
    componentAppliedLinearAccClamp=<2000 2000 2000>
    componentLinearAccScale=<0.5 0.5 0.5>
    componentLinearVelScale=<0.5 0.5 0.5>
    gravityOverride {z=-980}
    physicsBodyDefinitions [{
      boundBone {boneName=Dyn_sleeve_Up_L}
      boxExtents=<10 10 10>
      constraintSetup {angularLimitsMax=<0 0 5>, angularLimitsMin=<0 -7 -5>, angularTarget=<1 0 0>, linearAxesMax=<1 0 1>, linearAxesMin=<-1 0 0>}
      localJointOffset=<1 0 0>
     }]
   }
   n3 = Anim Dynamics - Bone: Dyn_sleeve_In_L!(ComponentPose=@n2, bAlphaBoolEnabled:bool=True, ExternalForce="0,0,0") -> (Pose) {
    alphaInputType=Bool
    angularBiasOverride=0.9999
    angularDampingOverride=1
    angularSpringConstant=500
    bAngularSpring=true
    bLinearSpring=true
    bOverrideAngularDamping=true
    bOverrideLinearDamping=true
    bUseGravityOverride=true
    linearDampingOverride=0.9999
    linearSpringConstant=100
    simulationSpace=RootRelative
    alphaBoolBlend {blendOption=Linear}
    boundBone {boneName=Dyn_sleeve_In_L}
    componentAppliedLinearAccClamp=<2000 2000 2000>
    componentLinearAccScale=<0.5 0.5 0.5>
    componentLinearVelScale=<0.5 0.5 0.5>
    gravityOverride {z=-980}
    physicsBodyDefinitions [{
      boundBone {boneName=Dyn_sleeve_In_L}
      boxExtents=<10 10 10>
      constraintSetup {angularLimitsMax=<0 0 5>, angularLimitsMin=<0 -7 -5>, angularTarget=<1 0 0>, linearAxesMax=<1 0 1>, linearAxesMin=<-1 0 0>}
      localJointOffset=<1 0 0>
     }]
   }
   Anim Dynamics - Bone: Dyn_sleeve_Out_L!(ComponentPose=@n3, bAlphaBoolEnabled:bool=True, ExternalForce="0,0,0") -> (Pose) {
    alphaInputType=Bool
    angularBiasOverride=0.9999
    angularDampingOverride=1
    angularSpringConstant=500
    bAngularSpring=true
    bLinearSpring=true
    bOverrideAngularDamping=true
    bOverrideLinearDamping=true
    bUseGravityOverride=true
    linearDampingOverride=0.9999
    linearSpringConstant=100
    simulationSpace=RootRelative
    alphaBoolBlend {blendOption=Linear}
    boundBone {boneName=Dyn_sleeve_Out_L}
    componentAppliedLinearAccClamp=<2000 2000 2000>
    componentLinearAccScale=<0.5 0.5 0.5>
    componentLinearVelScale=<0.5 0.5 0.5>
    gravityOverride {z=-980}
    physicsBodyDefinitions [{
      boundBone {boneName=Dyn_sleeve_Out_L}
      boxExtents=<10 10 10>
      constraintSetup {angularLimitsMax=<0 0 5>, angularLimitsMin=<0 -7 -5>, angularTarget=<1 0 0>, linearAxesMax=<1 0 1>, linearAxesMin=<-1 0 0>}
      localJointOffset=<1 0 0>
     }]
   }
   n4 = Anim Dynamics - Bone: Dyn_sleeve_Down_R!(ComponentPose, bAlphaBoolEnabled:bool=True, ExternalForce="0,0,0") -> (Pose) {
    alphaInputType=Bool
    angularBiasOverride=0.9999
    angularDampingOverride=1
    angularSpringConstant=500
    bAngularSpring=true
    bLinearSpring=true
    bOverrideAngularDamping=true
    bOverrideLinearDamping=true
    bUseGravityOverride=true
    linearDampingOverride=0.9999
    linearSpringConstant=100
    simulationSpace=RootRelative
    alphaBoolBlend {blendOption=Linear}
    boundBone {boneName=Dyn_sleeve_Down_R}
    componentAppliedLinearAccClamp=<2000 2000 2000>
    componentLinearAccScale=<0.5 0.5 0.5>
    componentLinearVelScale=<0.5 0.5 0.5>
    gravityOverride {z=-980}
    physicsBodyDefinitions [{
      boundBone {boneName=Dyn_sleeve_Down_R}
      boxExtents=<10 10 10>
      constraintSetup {angularLimitsMax=<0 0 5>, angularLimitsMin=<0 -7 -5>, angularTarget=<1 0 0>, linearAxesMax=<1 0 1>, linearAxesMin=<-1 0 0>}
      localJointOffset=<1 0 0>
     }]
   }
   n5 = Anim Dynamics - Bone: Dyn_sleeve_Up_R!(ComponentPose=@n4, bAlphaBoolEnabled:bool=True, ExternalForce="0,0,0") -> (Pose) {
    alphaInputType=Bool
    angularBiasOverride=0.9999
    angularDampingOverride=1
    angularSpringConstant=500
    bAngularSpring=true
    bLinearSpring=true
    bOverrideAngularDamping=true
    bOverrideLinearDamping=true
    bUseGravityOverride=true
    linearDampingOverride=0.9999
    linearSpringConstant=100
    simulationSpace=RootRelative
    alphaBoolBlend {blendOption=Linear}
    boundBone {boneName=Dyn_sleeve_Up_R}
    componentAppliedLinearAccClamp=<2000 2000 2000>
    componentLinearAccScale=<0.5 0.5 0.5>
    componentLinearVelScale=<0.5 0.5 0.5>
    gravityOverride {z=-980}
    physicsBodyDefinitions [{
      boundBone {boneName=Dyn_sleeve_Up_R}
      boxExtents=<10 10 10>
      constraintSetup {angularLimitsMax=<0 0 5>, angularLimitsMin=<0 -7 -5>, angularTarget=<1 0 0>, linearAxesMax=<1 0 1>, linearAxesMin=<-1 0 0>}
      localJointOffset=<1 0 0>
     }]
   }
   n6 = Anim Dynamics - Bone: Dyn_sleeve_Out_R!(ComponentPose=@n5, bAlphaBoolEnabled:bool=True, ExternalForce="0,0,0") -> (Pose) {
    alphaInputType=Bool
    angularBiasOverride=0.9999
    angularDampingOverride=1
    angularSpringConstant=500
    bAngularSpring=true
    bLinearSpring=true
    bOverrideAngularDamping=true
    bOverrideLinearDamping=true
    bUseGravityOverride=true
    linearDampingOverride=0.9999
    linearSpringConstant=100
    simulationSpace=RootRelative
    alphaBoolBlend {blendOption=Linear}
    boundBone {boneName=Dyn_sleeve_Out_R}
    componentAppliedLinearAccClamp=<2000 2000 2000>
    componentLinearAccScale=<0.5 0.5 0.5>
    componentLinearVelScale=<0.5 0.5 0.5>
    gravityOverride {z=-980}
    physicsBodyDefinitions [{
      boundBone {boneName=Dyn_sleeve_Out_R}
      boxExtents=<10 10 10>
      constraintSetup {angularLimitsMax=<0 0 5>, angularLimitsMin=<0 -7 -5>, angularTarget=<1 0 0>, linearAxesMax=<1 0 1>, linearAxesMin=<-1 0 0>}
      localJointOffset=<1 0 0>
     }]
   }
   Anim Dynamics - Bone: Dyn_sleeve_In_R!(ComponentPose=@n6, bAlphaBoolEnabled:bool=True, ExternalForce="0,0,0") -> (Pose) {
    alphaInputType=Bool
    angularBiasOverride=0.9999
    angularDampingOverride=1
    angularSpringConstant=500
    bAngularSpring=true
    bLinearSpring=true
    bOverrideAngularDamping=true
    bOverrideLinearDamping=true
    bUseGravityOverride=true
    linearDampingOverride=0.9999
    linearSpringConstant=100
    simulationSpace=RootRelative
    alphaBoolBlend {blendOption=Linear}
    boundBone {boneName=Dyn_sleeve_In_R}
    componentAppliedLinearAccClamp=<2000 2000 2000>
    componentLinearAccScale=<0.5 0.5 0.5>
    componentLinearVelScale=<0.5 0.5 0.5>
    gravityOverride {z=-980}
    physicsBodyDefinitions [{
      boundBone {boneName=Dyn_sleeve_In_R}
      boxExtents=<10 10 10>
      constraintSetup {angularLimitsMax=<0 0 5>, angularLimitsMin=<0 -7 -5>, angularTarget=<1 0 0>, linearAxesMax=<1 0 1>, linearAxesMin=<-1 0 0>}
      localJointOffset=<1 0 0>
     }]
   }
   n7 = Anim Dynamics - Bone: Dyn_sleeve_Down_L(bAlphaBoolEnabled=True, ExternalForce="0,0,0") {
    alphaInputType=Bool
    angularBiasOverride=0.9999
    angularDampingOverride=1
    angularSpringConstant=500
    bAngularSpring=true
    bOverrideAngularDamping=true
    bUseGravityOverride=true
    linearDampingOverride=0.9999
    linearSpringConstant=100
    simulationSpace=World
    alphaBoolBlend {blendOption=Linear}
    boundBone {boneName=Dyn_sleeve_Down_L}
    componentAppliedLinearAccClamp=<2000 2000 2000>
    gravityOverride {z=-980}
    physicsBodyDefinitions [{
      boundBone {boneName=Dyn_sleeve_Down_L}
      boxExtents=<10 10 10>
      constraintSetup {angularLimitsMax=<0 0 0>, angularLimitsMin=<0 -7 0>, angularTarget=<1 0 0>}
      localJointOffset=<0.4 0 0>
     }]
   }
   n8 = Anim Dynamics - Bone: Dyn_sleeve_Up_L(ComponentPose=@n7, bAlphaBoolEnabled=True, ExternalForce="0,0,0") {
    alphaInputType=Bool
    angularBiasOverride=0.9999
    angularDampingOverride=1
    angularSpringConstant=500
    bAngularSpring=true
    bOverrideAngularDamping=true
    bUseGravityOverride=true
    linearDampingOverride=0.9999
    linearSpringConstant=100
    simulationSpace=World
    alphaBoolBlend {blendOption=Linear}
    boundBone {boneName=Dyn_sleeve_Up_L}
    componentAppliedLinearAccClamp=<2000 2000 2000>
    gravityOverride {z=-980}
    physicsBodyDefinitions [{
      boundBone {boneName=Dyn_sleeve_Up_L}
      boxExtents=<10 10 10>
      constraintSetup {angularLimitsMax=<0 0 0>, angularLimitsMin=<0 -7 0>, angularTarget=<1 0 0>}
      localJointOffset=<0.4 0 0>
     }]
   }
   n9 = Anim Dynamics - Bone: Dyn_sleeve_In_L(ComponentPose=@n8, bAlphaBoolEnabled=True, ExternalForce="0,0,0") {
    alphaInputType=Bool
    angularBiasOverride=0.9999
    angularDampingOverride=1
    angularSpringConstant=500
    bAngularSpring=true
    bOverrideAngularDamping=true
    bUseGravityOverride=true
    linearDampingOverride=0.9999
    linearSpringConstant=100
    simulationSpace=World
    alphaBoolBlend {blendOption=Linear}
    boundBone {boneName=Dyn_sleeve_In_L}
    componentAppliedLinearAccClamp=<2000 2000 2000>
    gravityOverride {z=-980}
    physicsBodyDefinitions [{
      boundBone {boneName=Dyn_sleeve_In_L}
      boxExtents=<10 10 10>
      constraintSetup {angularLimitsMax=<0 0 0>, angularLimitsMin=<0 -7 0>, angularTarget=<1 0 0>}
      localJointOffset=<0.4 0 0>
     }]
   }
   n10 = Anim Dynamics - Bone: Dyn_sleeve_Out_L(ComponentPose=@n9, bAlphaBoolEnabled=True, ExternalForce="0,0,0") {
    alphaInputType=Bool
    angularBiasOverride=0.9999
    angularDampingOverride=1
    angularSpringConstant=500
    bAngularSpring=true
    bOverrideAngularDamping=true
    bUseGravityOverride=true
    linearDampingOverride=0.9999
    linearSpringConstant=100
    simulationSpace=World
    alphaBoolBlend {blendOption=Linear}
    boundBone {boneName=Dyn_sleeve_Out_L}
    componentAppliedLinearAccClamp=<2000 2000 2000>
    gravityOverride {z=-980}
    physicsBodyDefinitions [{
      boundBone {boneName=Dyn_sleeve_Out_L}
      boxExtents=<10 10 10>
      constraintSetup {angularLimitsMax=<0 0 0>, angularLimitsMin=<0 -7 0>, angularTarget=<1 0 0>}
      localJointOffset=<0.4 0 0>
     }]
   }
   n11 = Anim Dynamics - Bone: Dyn_sleeve_Down_R(ComponentPose=@n10, bAlphaBoolEnabled=True, ExternalForce="0,0,0") {
    alphaInputType=Bool
    angularBiasOverride=0.9999
    angularDampingOverride=1
    angularSpringConstant=500
    bAngularSpring=true
    bOverrideAngularDamping=true
    bOverrideLinearDamping=true
    bUseGravityOverride=true
    linearDampingOverride=0.9999
    linearSpringConstant=100
    simulationSpace=World
    alphaBoolBlend {blendOption=Linear}
    boundBone {boneName=Dyn_sleeve_Down_R}
    componentAppliedLinearAccClamp=<2000 2000 2000>
    gravityOverride {z=-980}
    physicsBodyDefinitions [{
      boundBone {boneName=Dyn_sleeve_Down_R}
      boxExtents=<10 10 10>
      constraintSetup {angularLimitsMax=<0 0 5>, angularLimitsMin=<0 -7 -5>, angularTarget=<1 0 0>}
      localJointOffset=<0.4 0 0>
     }]
   }
   n12 = Anim Dynamics - Bone: Dyn_sleeve_Up_R(ComponentPose=@n11, bAlphaBoolEnabled=True, ExternalForce="0,0,0") {
    alphaInputType=Bool
    angularBiasOverride=0.9999
    angularDampingOverride=1
    angularSpringConstant=500
    bAngularSpring=true
    bOverrideAngularDamping=true
    bOverrideLinearDamping=true
    bUseGravityOverride=true
    linearDampingOverride=0.9999
    linearSpringConstant=100
    simulationSpace=World
    alphaBoolBlend {blendOption=Linear}
    boundBone {boneName=Dyn_sleeve_Up_R}
    componentAppliedLinearAccClamp=<2000 2000 2000>
    gravityOverride {z=-980}
    physicsBodyDefinitions [{
      boundBone {boneName=Dyn_sleeve_Up_R}
      boxExtents=<10 10 10>
      constraintSetup {angularLimitsMax=<0 0 5>, angularLimitsMin=<0 -7 -5>, angularTarget=<1 0 0>}
      localJointOffset=<0.4 0 0>
     }]
   }
   n13 = Anim Dynamics - Bone: Dyn_sleeve_Out_R(ComponentPose=@n12, bAlphaBoolEnabled=True, ExternalForce="0,0,0") {
    alphaInputType=Bool
    angularBiasOverride=0.9999
    angularDampingOverride=1
    angularSpringConstant=500
    bAngularSpring=true
    bOverrideAngularDamping=true
    bOverrideLinearDamping=true
    bUseGravityOverride=true
    linearDampingOverride=0.9999
    linearSpringConstant=100
    simulationSpace=World
    alphaBoolBlend {blendOption=Linear}
    boundBone {boneName=Dyn_sleeve_Out_R}
    componentAppliedLinearAccClamp=<2000 2000 2000>
    gravityOverride {z=-980}
    physicsBodyDefinitions [{
      boundBone {boneName=Dyn_sleeve_Out_R}
      boxExtents=<10 10 10>
      constraintSetup {angularLimitsMax=<0 0 5>, angularLimitsMin=<0 -7 -5>, angularTarget=<1 0 0>}
      localJointOffset=<0.4 0 0>
     }]
   }
   Anim Dynamics - Bone: Dyn_sleeve_In_R(ComponentPose=@n13, bAlphaBoolEnabled=True, ExternalForce="0,0,0") {
    alphaInputType=Bool
    angularBiasOverride=0.9999
    angularDampingOverride=1
    angularSpringConstant=500
    bAngularSpring=true
    bOverrideAngularDamping=true
    bOverrideLinearDamping=true
    bUseGravityOverride=true
    linearDampingOverride=0.9999
    linearSpringConstant=100
    simulationSpace=World
    alphaBoolBlend {blendOption=Linear}
    boundBone {boneName=Dyn_sleeve_In_R}
    componentAppliedLinearAccClamp=<2000 2000 2000>
    gravityOverride {z=-980}
    physicsBodyDefinitions [{
      boundBone {boneName=Dyn_sleeve_In_R}
      boxExtents=<10 10 10>
      constraintSetup {angularLimitsMax=<0 0 5>, angularLimitsMin=<0 -7 -5>, angularTarget=<1 0 0>}
      localJointOffset=<0.4 0 0>
     }]
   }
   RigidBody!(ComponentPose, Alpha, OverrideWorldGravity:Vector="0,0,-980", ExternalForce) -> (Pose) {
    bClampLinearTranslationLimitToRefPose=true
    bOverrideWorldGravity=true
    overridePhysicsAsset=&/F_MED_ToadLeaf_Fronend_Physics
    simulationSpace=BaseBoneSpace
    alphaBoolBlend {blendOption=Linear}
    baseBoneRef {boneName=pelvis}
    componentAppliedLinearAccClamp=<5000 5000 5000>
    overrideWorldGravity {z=-980}
    simSpaceSettings {
     dampingAlpha=0
     maxAngularAcceleration=500
     maxAngularVelocity=3
     maxLinearAcceleration=400
     maxLinearVelocity=250
     velocityScaleZ=0.8
     worldAlpha=0.8
     externalLinearDragV=<0.1 0.1 0.1>
    }
   }
   n14 = RigidBody(ComponentPose=Local To Component(LocalPose=Use cached pose 'For Frontebd'()), OverrideWorldGravity="0,0,-980") {
    bClampLinearTranslationLimitToRefPose=true
    bOverrideWorldGravity=true
    overridePhysicsAsset=&F_MED_ToadLeaf_Fronend_Physics
    simulationSpace=BaseBoneSpace
    alphaBoolBlend {blendOption=Linear}
    baseBoneRef {boneName=pelvis}
    componentAppliedLinearAccClamp=<5000 5000 5000>
    overrideWorldGravity {z=-980}
    simSpaceSettings {
     dampingAlpha=0
     maxAngularAcceleration=500
     maxAngularVelocity=3
     maxLinearAcceleration=1200
     maxLinearVelocity=250
     velocityScaleZ=0.8
     worldAlpha=0.8
     externalLinearDragV=<0.1 0.1 0.1>
    }
   }
   n15 = Anim Dynamics - Bone: Dyn_skirt_Fr_01_L!(ComponentPose=@n14, bAlphaBoolEnabled:bool=True, ExternalForce="0,0,0") -> (Pose) {
    alphaInputType=Bool
    angularDampingOverride=1
    angularSpringConstant=500
    bAngularSpring=true
    bOverrideAngularDamping=true
    bOverrideLinearDamping=true
    bUseGravityOverride=true
    linearDampingOverride=0.9999
    simulationSpace=RootRelative
    alphaBoolBlend {blendOption=Linear}
    boundBone {boneName=Dyn_skirt_Fr_01_L}
    componentAppliedLinearAccClamp=<2000 2000 2000>
    componentLinearAccScale=<0.5 0.5 0.5>
    componentLinearVelScale=<0.5 0.5 0.5>
    gravityOverride {z=-980}
    physicsBodyDefinitions [{
      boundBone {boneName=Dyn_skirt_Fr_01_L}
      boxExtents=<10 10 10>
      constraintSetup {angularLimitsMax=<5 0 5>, angularLimitsMin=<-5 -25 -5>, angularTarget=<1 0 0>}
      localJointOffset=<1 0 0>
     }]
   }
   n16 = Anim Dynamics - Bone: Dyn_skirt_Fr_02_L!(ComponentPose=@n15, bAlphaBoolEnabled:bool=True, ExternalForce="0,0,0") -> (Pose) {
    alphaInputType=Bool
    angularDampingOverride=1
    angularSpringConstant=500
    bAngularSpring=true
    bOverrideAngularDamping=true
    bOverrideLinearDamping=true
    bUseGravityOverride=true
    linearDampingOverride=0.9999
    simulationSpace=RootRelative
    alphaBoolBlend {blendOption=Linear}
    boundBone {boneName=Dyn_skirt_Fr_02_L}
    componentAppliedLinearAccClamp=<2000 2000 2000>
    componentLinearAccScale=<0.5 0.5 0.5>
    componentLinearVelScale=<0.5 0.5 0.5>
    gravityOverride {z=-980}
    physicsBodyDefinitions [{
      boundBone {boneName=Dyn_skirt_Fr_02_L}
      boxExtents=<10 10 10>
      constraintSetup {angularLimitsMax=<5 0 5>, angularLimitsMin=<-5 -15 -5>, angularTarget=<1 0 0>}
      localJointOffset=<1 0 0>
     }]
   }
   n17 = Anim Dynamics - Bone: Dyn_skirt_Fr_03_L!(ComponentPose=@n16, bAlphaBoolEnabled:bool=True, ExternalForce="0,0,0") -> (Pose) {
    alphaInputType=Bool
    angularDampingOverride=1
    angularSpringConstant=500
    bAngularSpring=true
    bOverrideAngularDamping=true
    bOverrideLinearDamping=true
    bUseGravityOverride=true
    linearDampingOverride=0.9999
    simulationSpace=RootRelative
    alphaBoolBlend {blendOption=Linear}
    boundBone {boneName=Dyn_skirt_Fr_03_L}
    componentAppliedLinearAccClamp=<2000 2000 2000>
    componentLinearAccScale=<0.5 0.5 0.5>
    componentLinearVelScale=<0.5 0.5 0.5>
    gravityOverride {z=-980}
    physicsBodyDefinitions [{
      boundBone {boneName=Dyn_skirt_Fr_03_L}
      boxExtents=<10 10 10>
      constraintSetup {angularLimitsMax=<5 0 5>, angularLimitsMin=<-5 -10 -5>, angularTarget=<1 0 0>}
      localJointOffset=<1 0 0>
     }]
   }
   n18 = Anim Dynamics - Bone: Dyn_skirt_Fr_01_R!(ComponentPose=@n17, bAlphaBoolEnabled:bool=True, ExternalForce="0,0,0") -> (Pose) {
    alphaInputType=Bool
    angularDampingOverride=1
    angularSpringConstant=500
    bAngularSpring=true
    bOverrideAngularDamping=true
    bOverrideLinearDamping=true
    bUseGravityOverride=true
    linearDampingOverride=0.9999
    simulationSpace=RootRelative
    alphaBoolBlend {blendOption=Linear}
    boundBone {boneName=Dyn_skirt_Fr_01_R}
    componentAppliedLinearAccClamp=<2000 2000 2000>
    componentLinearAccScale=<0.5 0.5 0.5>
    componentLinearVelScale=<0.5 0.5 0.5>
    gravityOverride {z=-980}
    physicsBodyDefinitions [{
      boundBone {boneName=Dyn_skirt_Fr_01_R}
      boxExtents=<10 10 10>
      constraintSetup {angularLimitsMax=<0 0 0>, angularLimitsMin=<0 -15 0>, angularTarget=<1 0 0>}
      localJointOffset=<1 0 0>
     }]
   }
   n19 = Anim Dynamics - Bone: Dyn_skirt_Fr_02_R!(ComponentPose=@n18, bAlphaBoolEnabled:bool=True, ExternalForce="0,0,0") -> (Pose) {
    alphaInputType=Bool
    angularDampingOverride=1
    angularSpringConstant=500
    bAngularSpring=true
    bOverrideAngularDamping=true
    bOverrideLinearDamping=true
    bUseGravityOverride=true
    linearDampingOverride=0.9999
    simulationSpace=RootRelative
    alphaBoolBlend {blendOption=Linear}
    boundBone {boneName=Dyn_skirt_Fr_02_R}
    componentAppliedLinearAccClamp=<2000 2000 2000>
    componentLinearAccScale=<0.5 0.5 0.5>
    componentLinearVelScale=<0.5 0.5 0.5>
    gravityOverride {z=-980}
    physicsBodyDefinitions [{
      boundBone {boneName=Dyn_skirt_Fr_02_R}
      boxExtents=<10 10 10>
      constraintSetup {angularLimitsMax=<0 0 0>, angularLimitsMin=<0 -10 0>, angularTarget=<1 0 0>}
      localJointOffset=<1 0 0>
     }]
   }
   n20 = Anim Dynamics - Bone: Dyn_skirt_Fr_03_R!(ComponentPose=@n19, bAlphaBoolEnabled:bool=True, ExternalForce="0,0,0") -> (Pose) {
    alphaInputType=Bool
    angularDampingOverride=1
    angularSpringConstant=500
    bAngularSpring=true
    bOverrideAngularDamping=true
    bOverrideLinearDamping=true
    bUseGravityOverride=true
    linearDampingOverride=0.9999
    simulationSpace=RootRelative
    alphaBoolBlend {blendOption=Linear}
    boundBone {boneName=Dyn_skirt_Fr_03_R}
    componentAppliedLinearAccClamp=<2000 2000 2000>
    componentLinearAccScale=<0.5 0.5 0.5>
    componentLinearVelScale=<0.5 0.5 0.5>
    gravityOverride {z=-980}
    physicsBodyDefinitions [{
      boundBone {boneName=Dyn_skirt_Fr_03_R}
      boxExtents=<10 10 10>
      constraintSetup {angularLimitsMax=<0 0 0>, angularLimitsMin=<0 -7 0>, angularTarget=<1 0 0>}
      localJointOffset=<1 0 0>
     }]
   }
   n21 = Anim Dynamics - Bone: Dyn_skirt_Side_01_L!(ComponentPose=@n20, bAlphaBoolEnabled:bool=True, ExternalForce="0,0,0") -> (Pose) {
    alphaInputType=Bool
    angularDampingOverride=1
    angularSpringConstant=500
    bAngularSpring=true
    bOverrideAngularDamping=true
    bOverrideLinearDamping=true
    bUseGravityOverride=true
    linearDampingOverride=0.9999
    simulationSpace=RootRelative
    alphaBoolBlend {blendOption=Linear}
    boundBone {boneName=Dyn_skirt_Side_01_L}
    componentAppliedLinearAccClamp=<2000 2000 2000>
    componentLinearAccScale=<0.5 0.5 0.5>
    componentLinearVelScale=<0.5 0.5 0.5>
    gravityOverride {z=-980}
    physicsBodyDefinitions [{
      boundBone {boneName=Dyn_skirt_Side_01_L}
      boxExtents=<10 10 10>
      constraintSetup {angularLimitsMax=<0 0 5>, angularLimitsMin=<0 -15 -15>, angularTarget=<1 0 0>}
      localJointOffset=<1 0 0>
     }]
   }
   n22 = Anim Dynamics - Bone: Dyn_skirt_Side_02_L!(ComponentPose=@n21, bAlphaBoolEnabled:bool=True, ExternalForce="0,0,0") -> (Pose) {
    alphaInputType=Bool
    angularDampingOverride=1
    angularSpringConstant=500
    bAngularSpring=true
    bOverrideAngularDamping=true
    bOverrideLinearDamping=true
    bUseGravityOverride=true
    bUseSphericalLimits=true
    linearDampingOverride=0.9999
    simulationSpace=RootRelative
    alphaBoolBlend {blendOption=Linear}
    boundBone {boneName=Dyn_skirt_Side_02_L}
    componentAppliedLinearAccClamp=<2000 2000 2000>
    componentLinearAccScale=<0.5 0.5 0.5>
    componentLinearVelScale=<0.5 0.5 0.5>
    gravityOverride {z=-980}
    physicsBodyDefinitions [{
      collisionType=CustomSphere
      sphereCollisionRadius=4
      boundBone {boneName=Dyn_skirt_Side_02_L}
      boxExtents=<10 10 10>
      constraintSetup {angularLimitsMax=<0 0 5>, angularLimitsMin=<0 -10 -10>, angularTarget=<1 0 0>}
      localJointOffset=<1 0 0>
     }]
    sphericalLimits [{limitRadius=6, drivingBone {boneName=thigh_l}, sphereLocalOffset=<-15 0 0>}]
   }
   n23 = Anim Dynamics - Bone: Dyn_skirt_Side_03_L!(ComponentPose=@n22, bAlphaBoolEnabled:bool=True, ExternalForce="0,0,0") -> (Pose) {
    alphaInputType=Bool
    angularDampingOverride=1
    angularSpringConstant=500
    bAngularSpring=true
    bOverrideAngularDamping=true
    bOverrideLinearDamping=true
    bUseGravityOverride=true
    linearDampingOverride=0.9999
    simulationSpace=RootRelative
    alphaBoolBlend {blendOption=Linear}
    boundBone {boneName=Dyn_skirt_Side_03_L}
    componentAppliedLinearAccClamp=<2000 2000 2000>
    componentLinearAccScale=<0.5 0.5 0.5>
    componentLinearVelScale=<0.5 0.5 0.5>
    gravityOverride {z=-980}
    physicsBodyDefinitions [{
      boundBone {boneName=Dyn_skirt_Side_03_L}
      boxExtents=<10 10 10>
      constraintSetup {angularLimitsMax=<0 0 0>, angularLimitsMin=<0 -5 -5>, angularTarget=<1 0 0>}
      localJointOffset=<1 0 0>
     }]
   }
   n24 = Anim Dynamics - Bone: Dyn_skirt_Side_01_R!(ComponentPose=@n23, bAlphaBoolEnabled:bool=True, ExternalForce="0,0,0") -> (Pose) {
    alphaInputType=Bool
    angularDampingOverride=1
    angularSpringConstant=500
    bAngularSpring=true
    bOverrideAngularDamping=true
    bOverrideLinearDamping=true
    bUseGravityOverride=true
    linearDampingOverride=0.9999
    simulationSpace=RootRelative
    alphaBoolBlend {blendOption=Linear}
    boundBone {boneName=Dyn_skirt_Side_01_R}
    componentAppliedLinearAccClamp=<2000 2000 2000>
    componentLinearAccScale=<0.5 0.5 0.5>
    componentLinearVelScale=<0.5 0.5 0.5>
    gravityOverride {z=-980}
    physicsBodyDefinitions [{
      boundBone {boneName=Dyn_skirt_Side_01_R}
      boxExtents=<10 10 10>
      constraintSetup {angularLimitsMax=<0 0 0>, angularLimitsMin=<0 -20 0>, angularTarget=<1 0 0>}
      localJointOffset=<1 0 0>
     }]
   }
   n25 = Anim Dynamics - Bone: Dyn_skirt_Side_02_R!(ComponentPose=@n24, bAlphaBoolEnabled:bool=True, ExternalForce="0,0,0") -> (Pose) {
    alphaInputType=Bool
    angularDampingOverride=1
    angularSpringConstant=500
    bAngularSpring=true
    bOverrideAngularDamping=true
    bOverrideLinearDamping=true
    bUseGravityOverride=true
    bUseSphericalLimits=true
    linearDampingOverride=0.9999
    simulationSpace=RootRelative
    alphaBoolBlend {blendOption=Linear}
    boundBone {boneName=Dyn_skirt_Side_02_R}
    componentAppliedLinearAccClamp=<2000 2000 2000>
    componentLinearAccScale=<0.5 0.5 0.5>
    componentLinearVelScale=<0.5 0.5 0.5>
    gravityOverride {z=-980}
    physicsBodyDefinitions [{
      collisionType=CustomSphere
      sphereCollisionRadius=5
      boundBone {boneName=Dyn_skirt_Side_02_R}
      boxExtents=<10 10 10>
      constraintSetup {angularLimitsMax=<0 0 0>, angularLimitsMin=<0 -15 0>, angularTarget=<1 0 0>}
      localJointOffset=<1 0 0>
     }]
    sphericalLimits [{limitRadius=6, drivingBone {boneName=thigh_r}, sphereLocalOffset=<15 0 0>}]
   }
   n26 = Anim Dynamics - Bone: Dyn_skirt_Side_03_R!(ComponentPose=@n25, bAlphaBoolEnabled:bool=True, ExternalForce="0,0,0") -> (Pose) {
    alphaInputType=Bool
    angularDampingOverride=1
    angularSpringConstant=500
    bAngularSpring=true
    bOverrideAngularDamping=true
    bOverrideLinearDamping=true
    bUseGravityOverride=true
    linearDampingOverride=0.9999
    simulationSpace=RootRelative
    alphaBoolBlend {blendOption=Linear}
    boundBone {boneName=Dyn_skirt_Side_03_R}
    componentAppliedLinearAccClamp=<2000 2000 2000>
    componentLinearAccScale=<0.5 0.5 0.5>
    componentLinearVelScale=<0.5 0.5 0.5>
    gravityOverride {z=-980}
    physicsBodyDefinitions [{
      boundBone {boneName=Dyn_skirt_Side_03_R}
      boxExtents=<10 10 10>
      constraintSetup {angularLimitsMax=<0 0 0>, angularLimitsMin=<0 -10 0>, angularTarget=<1 0 0>}
      localJointOffset=<1 0 0>
     }]
   }
   n27 = Anim Dynamics - Bone: Dyn_skirt_Bk_01_L!(ComponentPose=@n26, bAlphaBoolEnabled:bool=True, ExternalForce="0,0,0") -> (Pose) {
    alphaInputType=Bool
    angularDampingOverride=1
    angularSpringConstant=500
    bAngularSpring=true
    bOverrideAngularDamping=true
    bOverrideLinearDamping=true
    bUseGravityOverride=true
    linearDampingOverride=0.9999
    simulationSpace=RootRelative
    alphaBoolBlend {blendOption=Linear}
    boundBone {boneName=Dyn_skirt_Bk_01_L}
    componentAppliedLinearAccClamp=<2000 2000 2000>
    componentLinearAccScale=<0.5 0.5 0.5>
    componentLinearVelScale=<0.5 0.5 0.5>
    gravityOverride {z=-980}
    physicsBodyDefinitions [{
      boundBone {boneName=Dyn_skirt_Bk_01_L}
      boxExtents=<10 10 10>
      constraintSetup {angularLimitsMax=<0 0 5>, angularLimitsMin=<0 -25 -5>, angularTarget=<1 0 0>}
      localJointOffset=<1 0 0>
     }]
   }
   n28 = Anim Dynamics - Bone: Dyn_skirt_Bk_02_L!(ComponentPose=@n27, bAlphaBoolEnabled:bool=True, ExternalForce="0,0,0") -> (Pose) {
    alphaInputType=Bool
    angularDampingOverride=1
    angularSpringConstant=500
    bAngularSpring=true
    bOverrideAngularDamping=true
    bOverrideLinearDamping=true
    bUseGravityOverride=true
    linearDampingOverride=0.9999
    simulationSpace=RootRelative
    alphaBoolBlend {blendOption=Linear}
    boundBone {boneName=Dyn_skirt_Bk_02_L}
    componentAppliedLinearAccClamp=<2000 2000 2000>
    componentLinearAccScale=<0.5 0.5 0.5>
    componentLinearVelScale=<0.5 0.5 0.5>
    gravityOverride {z=-980}
    physicsBodyDefinitions [{
      boundBone {boneName=Dyn_skirt_Bk_02_L}
      boxExtents=<10 10 10>
      constraintSetup {angularLimitsMax=<0 0 5>, angularLimitsMin=<0 -15 -5>, angularTarget=<1 0 0>}
      localJointOffset=<1 0 0>
     }]
   }
   n29 = Anim Dynamics - Bone: Dyn_skirt_Bk_01_R!(ComponentPose=@n28, bAlphaBoolEnabled:bool=True, ExternalForce="0,0,0") -> (Pose) {
    alphaInputType=Bool
    angularDampingOverride=1
    angularSpringConstant=500
    bAngularSpring=true
    bOverrideAngularDamping=true
    bOverrideLinearDamping=true
    bUseGravityOverride=true
    linearDampingOverride=0.9999
    simulationSpace=RootRelative
    alphaBoolBlend {blendOption=Linear}
    boundBone {boneName=Dyn_skirt_Bk_01_R}
    componentAppliedLinearAccClamp=<2000 2000 2000>
    componentLinearAccScale=<0.5 0.5 0.5>
    componentLinearVelScale=<0.5 0.5 0.5>
    gravityOverride {z=-980}
    physicsBodyDefinitions [{
      boundBone {boneName=Dyn_skirt_Bk_01_R}
      boxExtents=<10 10 10>
      constraintSetup {angularLimitsMax=<0 0 0>, angularLimitsMin=<0 -25 0>, angularTarget=<1 0 0>}
      localJointOffset=<1 0 0>
     }]
   }
   n30 = Anim Dynamics - Bone: Dyn_skirt_Bk_02_R!(ComponentPose=@n29, bAlphaBoolEnabled:bool=True, ExternalForce="0,0,0") -> (Pose) {
    alphaInputType=Bool
    angularDampingOverride=1
    angularSpringConstant=500
    bAngularSpring=true
    bOverrideAngularDamping=true
    bOverrideLinearDamping=true
    bUseGravityOverride=true
    bUsePlanarLimit=false
    linearDampingOverride=0.9999
    simulationSpace=RootRelative
    alphaBoolBlend {blendOption=Linear}
    boundBone {boneName=Dyn_skirt_Bk_02_R}
    componentAppliedLinearAccClamp=<2000 2000 2000>
    componentLinearAccScale=<0.5 0.5 0.5>
    componentLinearVelScale=<0.5 0.5 0.5>
    gravityOverride {z=-980}
    physicsBodyDefinitions [{
      sphereCollisionRadius=5
      boundBone {boneName=Dyn_skirt_Bk_02_R}
      boxExtents=<10 10 10>
      constraintSetup {angularLimitsMax=<0 0 0>, angularLimitsMin=<0 -15 0>, angularTarget=<1 0 0>}
      localJointOffset=<1 0 0>
     }]
   }
   n31 = RigidBody(ComponentPose=Local To Component(LocalPose=Use cached pose 'For Frontebd'()), OverrideWorldGravity="0,0,-980") {
    bClampLinearTranslationLimitToRefPose=true
    bOverrideWorldGravity=true
    overridePhysicsAsset=&F_MED_ToadLeaf_Fronend_Physics
    simulationSpace=BaseBoneSpace
    alphaBoolBlend {blendOption=Linear}
    baseBoneRef {boneName=pelvis}
    componentAppliedLinearAccClamp=<5000 5000 5000>
    overrideWorldGravity {z=-980}
    simSpaceSettings {
     dampingAlpha=0
     maxAngularAcceleration=500
     maxAngularVelocity=3
     maxLinearAcceleration=400
     maxLinearVelocity=250
     velocityScaleZ=0.8
     worldAlpha=0.6
     externalLinearDragV=<0.1 0.1 0.1>
    }
   }
   n32 = Anim Dynamics - Bone: Dyn_skirt_Fr_01_L!(ComponentPose=@n31, Alpha:float=1, ExternalForce="0,0,0") -> (Pose) {
    angularDampingOverride=1
    angularSpringConstant=500
    bAngularSpring=true
    bOverrideAngularDamping=true
    bUseGravityOverride=true
    linearDampingOverride=0.9999
    simulationSpace=World
    alphaBoolBlend {blendOption=Linear}
    boundBone {boneName=Dyn_skirt_Fr_01_L}
    componentAppliedLinearAccClamp=<2000 2000 2000>
    gravityOverride {z=-980}
    physicsBodyDefinitions [{
      boundBone {boneName=Dyn_skirt_Fr_01_L}
      boxExtents=<10 10 10>
      constraintSetup {angularLimitsMax=<0 0 0>, angularLimitsMin=<0 -25 0>, angularTarget=<1 0 0>}
      localJointOffset=<0.3 0 0>
     }]
   }
   n33 = Anim Dynamics - Bone: Dyn_skirt_Fr_02_L!(ComponentPose=@n32, Alpha:float=1, ExternalForce="0,0,0") -> (Pose) {
    angularDampingOverride=1
    angularSpringConstant=500
    bAngularSpring=true
    bOverrideAngularDamping=true
    bUseGravityOverride=true
    linearDampingOverride=0.9999
    simulationSpace=World
    alphaBoolBlend {blendOption=Linear}
    boundBone {boneName=Dyn_skirt_Fr_02_L}
    componentAppliedLinearAccClamp=<2000 2000 2000>
    gravityOverride {z=-980}
    physicsBodyDefinitions [{
      boundBone {boneName=Dyn_skirt_Fr_02_L}
      boxExtents=<10 10 10>
      constraintSetup {angularLimitsMax=<0 0 0>, angularLimitsMin=<0 -15 0>, angularTarget=<1 0 0>}
      localJointOffset=<0.3 0 0>
     }]
   }
   n34 = Anim Dynamics - Bone: Dyn_skirt_Fr_03_L!(ComponentPose=@n33, Alpha:float=1, ExternalForce="0,0,0") -> (Pose) {
    angularDampingOverride=1
    angularSpringConstant=500
    bAngularSpring=true
    bOverrideAngularDamping=true
    bUseGravityOverride=true
    linearDampingOverride=0.9999
    simulationSpace=World
    alphaBoolBlend {blendOption=Linear}
    boundBone {boneName=Dyn_skirt_Fr_03_L}
    componentAppliedLinearAccClamp=<2000 2000 2000>
    gravityOverride {z=-980}
    physicsBodyDefinitions [{
      boundBone {boneName=Dyn_skirt_Fr_03_L}
      boxExtents=<10 10 10>
      constraintSetup {angularLimitsMax=<0 0 0>, angularLimitsMin=<0 -10 0>, angularTarget=<1 0 0>}
      localJointOffset=<0.3 0 0>
     }]
   }
   n35 = Anim Dynamics - Bone: Dyn_skirt_Fr_01_R!(ComponentPose=@n34, Alpha:float=1, ExternalForce="0,0,0") -> (Pose) {
    angularDampingOverride=1
    angularSpringConstant=500
    bAngularSpring=true
    bOverrideAngularDamping=true
    bUseGravityOverride=true
    linearDampingOverride=0.9999
    simulationSpace=World
    alphaBoolBlend {blendOption=Linear}
    boundBone {boneName=Dyn_skirt_Fr_01_R}
    componentAppliedLinearAccClamp=<2000 2000 2000>
    gravityOverride {z=-980}
    physicsBodyDefinitions [{
      boundBone {boneName=Dyn_skirt_Fr_01_R}
      boxExtents=<10 10 10>
      constraintSetup {angularLimitsMax=<0 0 0>, angularLimitsMin=<0 -15 0>, angularTarget=<1 0 0>}
      localJointOffset=<0.3 0 0>
     }]
   }
   n36 = Anim Dynamics - Bone: Dyn_skirt_Fr_02_R!(ComponentPose=@n35, Alpha:float=1, ExternalForce="0,0,0") -> (Pose) {
    angularDampingOverride=1
    angularSpringConstant=500
    bAngularSpring=true
    bOverrideAngularDamping=true
    bUseGravityOverride=true
    linearDampingOverride=0.9999
    simulationSpace=World
    alphaBoolBlend {blendOption=Linear}
    boundBone {boneName=Dyn_skirt_Fr_02_R}
    componentAppliedLinearAccClamp=<2000 2000 2000>
    gravityOverride {z=-980}
    physicsBodyDefinitions [{
      boundBone {boneName=Dyn_skirt_Fr_02_R}
      boxExtents=<10 10 10>
      constraintSetup {angularLimitsMax=<0 0 0>, angularLimitsMin=<0 -10 0>, angularTarget=<1 0 0>}
      localJointOffset=<0.3 0 0>
     }]
   }
   n37 = Anim Dynamics - Bone: Dyn_skirt_Fr_03_R!(ComponentPose=@n36, Alpha:float=1, ExternalForce="0,0,0") -> (Pose) {
    angularDampingOverride=1
    angularSpringConstant=500
    bAngularSpring=true
    bOverrideAngularDamping=true
    bUseGravityOverride=true
    linearDampingOverride=0.9999
    simulationSpace=World
    alphaBoolBlend {blendOption=Linear}
    boundBone {boneName=Dyn_skirt_Fr_03_R}
    componentAppliedLinearAccClamp=<2000 2000 2000>
    gravityOverride {z=-980}
    physicsBodyDefinitions [{
      boundBone {boneName=Dyn_skirt_Fr_03_R}
      boxExtents=<10 10 10>
      constraintSetup {angularLimitsMax=<0 0 0>, angularLimitsMin=<0 -7 0>, angularTarget=<1 0 0>}
      localJointOffset=<0.3 0 0>
     }]
   }
   n38 = Anim Dynamics - Bone: Dyn_skirt_Side_01_L!(ComponentPose=@n37, Alpha:float=1, ExternalForce="0,0,0") -> (Pose) {
    angularDampingOverride=1
    angularSpringConstant=500
    bAngularSpring=true
    bOverrideAngularDamping=true
    bUseGravityOverride=true
    linearDampingOverride=0.9999
    simulationSpace=World
    alphaBoolBlend {blendOption=Linear}
    boundBone {boneName=Dyn_skirt_Side_01_L}
    componentAppliedLinearAccClamp=<2000 2000 2000>
    gravityOverride {z=-980}
    physicsBodyDefinitions [{
      boundBone {boneName=Dyn_skirt_Side_01_L}
      boxExtents=<10 10 10>
      constraintSetup {angularLimitsMax=<0 0 0>, angularLimitsMin=<0 -20 0>, angularTarget=<1 0 0>}
      localJointOffset=<0.3 0 0>
     }]
   }
   n39 = Anim Dynamics - Bone: Dyn_skirt_Side_02_L!(ComponentPose=@n38, Alpha:float=1, ExternalForce="0,0,0") -> (Pose) {
    angularDampingOverride=1
    angularSpringConstant=500
    bAngularSpring=true
    bOverrideAngularDamping=true
    bUseGravityOverride=true
    bUseSphericalLimits=true
    linearDampingOverride=0.9999
    simulationSpace=World
    alphaBoolBlend {blendOption=Linear}
    boundBone {boneName=Dyn_skirt_Side_02_L}
    componentAppliedLinearAccClamp=<2000 2000 2000>
    gravityOverride {z=-980}
    physicsBodyDefinitions [{
      collisionType=CustomSphere
      sphereCollisionRadius=4
      boundBone {boneName=Dyn_skirt_Side_02_L}
      boxExtents=<10 10 10>
      constraintSetup {angularLimitsMax=<0 0 0>, angularLimitsMin=<0 -15 0>, angularTarget=<1 0 0>}
      localJointOffset=<0.3 0 0>
     }]
    sphericalLimits [{limitRadius=6, drivingBone {boneName=thigh_l}, sphereLocalOffset=<-15 0 0>}]
   }
   n40 = Anim Dynamics - Bone: Dyn_skirt_Side_03_L!(ComponentPose=@n39, Alpha:float=1, ExternalForce="0,0,0") -> (Pose) {
    angularDampingOverride=1
    angularSpringConstant=500
    bAngularSpring=true
    bOverrideAngularDamping=true
    bUseGravityOverride=true
    linearDampingOverride=0.9999
    simulationSpace=World
    alphaBoolBlend {blendOption=Linear}
    boundBone {boneName=Dyn_skirt_Side_03_L}
    componentAppliedLinearAccClamp=<2000 2000 2000>
    gravityOverride {z=-980}
    physicsBodyDefinitions [{
      boundBone {boneName=Dyn_skirt_Side_03_L}
      boxExtents=<10 10 10>
      constraintSetup {angularLimitsMax=<0 0 0>, angularLimitsMin=<0 -10 0>, angularTarget=<1 0 0>}
      localJointOffset=<0.3 0 0>
     }]
   }
   n41 = Anim Dynamics - Bone: Dyn_skirt_Side_01_R!(ComponentPose=@n40, Alpha:float=1, ExternalForce="0,0,0") -> (Pose) {
    angularDampingOverride=1
    angularSpringConstant=500
    bAngularSpring=true
    bOverrideAngularDamping=true
    bUseGravityOverride=true
    linearDampingOverride=0.9999
    simulationSpace=World
    alphaBoolBlend {blendOption=Linear}
    boundBone {boneName=Dyn_skirt_Side_01_R}
    componentAppliedLinearAccClamp=<2000 2000 2000>
    gravityOverride {z=-980}
    physicsBodyDefinitions [{
      boundBone {boneName=Dyn_skirt_Side_01_R}
      boxExtents=<10 10 10>
      constraintSetup {angularLimitsMax=<0 0 0>, angularLimitsMin=<0 -20 0>, angularTarget=<1 0 0>}
      localJointOffset=<0.3 0 0>
     }]
   }
   n42 = Anim Dynamics - Bone: Dyn_skirt_Side_02_R!(ComponentPose=@n41, Alpha:float=1, ExternalForce="0,0,0") -> (Pose) {
    angularDampingOverride=1
    angularSpringConstant=500
    bAngularSpring=true
    bOverrideAngularDamping=true
    bUseGravityOverride=true
    bUseSphericalLimits=true
    linearDampingOverride=0.9999
    simulationSpace=World
    alphaBoolBlend {blendOption=Linear}
    boundBone {boneName=Dyn_skirt_Side_02_R}
    componentAppliedLinearAccClamp=<2000 2000 2000>
    gravityOverride {z=-980}
    physicsBodyDefinitions [{
      collisionType=CustomSphere
      sphereCollisionRadius=5
      boundBone {boneName=Dyn_skirt_Side_02_R}
      boxExtents=<10 10 10>
      constraintSetup {angularLimitsMax=<0 0 0>, angularLimitsMin=<0 -15 0>, angularTarget=<1 0 0>}
      localJointOffset=<0.3 0 0>
     }]
    sphericalLimits [{limitRadius=6, drivingBone {boneName=thigh_r}, sphereLocalOffset=<15 0 0>}]
   }
   n43 = Anim Dynamics - Bone: Dyn_skirt_Side_03_R!(ComponentPose=@n42, Alpha:float=1, ExternalForce="0,0,0") -> (Pose) {
    angularDampingOverride=1
    angularSpringConstant=500
    bAngularSpring=true
    bOverrideAngularDamping=true
    bUseGravityOverride=true
    linearDampingOverride=0.9999
    simulationSpace=World
    alphaBoolBlend {blendOption=Linear}
    boundBone {boneName=Dyn_skirt_Side_03_R}
    componentAppliedLinearAccClamp=<2000 2000 2000>
    gravityOverride {z=-980}
    physicsBodyDefinitions [{
      boundBone {boneName=Dyn_skirt_Side_03_R}
      boxExtents=<10 10 10>
      constraintSetup {angularLimitsMax=<0 0 0>, angularLimitsMin=<0 -10 0>, angularTarget=<1 0 0>}
      localJointOffset=<0.3 0 0>
     }]
   }
   n44 = Anim Dynamics - Bone: Dyn_skirt_Bk_01_L!(ComponentPose=@n43, Alpha:float=1, ExternalForce="0,0,0") -> (Pose) {
    angularDampingOverride=1
    angularSpringConstant=500
    bAngularSpring=true
    bOverrideAngularDamping=true
    bUseGravityOverride=true
    linearDampingOverride=0.9999
    simulationSpace=World
    alphaBoolBlend {blendOption=Linear}
    boundBone {boneName=Dyn_skirt_Bk_01_L}
    componentAppliedLinearAccClamp=<2000 2000 2000>
    gravityOverride {z=-980}
    physicsBodyDefinitions [{
      boundBone {boneName=Dyn_skirt_Bk_01_L}
      boxExtents=<10 10 10>
      constraintSetup {angularLimitsMax=<0 0 0>, angularLimitsMin=<0 -25 0>, angularTarget=<1 0 0>}
      localJointOffset=<0.3 0 0>
     }]
   }
   n45 = Anim Dynamics - Bone: Dyn_skirt_Bk_02_L!(ComponentPose=@n44, Alpha:float=1, ExternalForce="0,0,0") -> (Pose) {
    angularDampingOverride=1
    angularSpringConstant=500
    bAngularSpring=true
    bOverrideAngularDamping=true
    bUseGravityOverride=true
    linearDampingOverride=0.9999
    simulationSpace=World
    alphaBoolBlend {blendOption=Linear}
    boundBone {boneName=Dyn_skirt_Bk_02_L}
    componentAppliedLinearAccClamp=<2000 2000 2000>
    gravityOverride {z=-980}
    physicsBodyDefinitions [{
      boundBone {boneName=Dyn_skirt_Bk_02_L}
      boxExtents=<10 10 10>
      constraintSetup {angularLimitsMax=<0 0 0>, angularLimitsMin=<0 -15 0>, angularTarget=<1 0 0>}
      localJointOffset=<0.3 0 0>
     }]
   }
   n46 = Anim Dynamics - Bone: Dyn_skirt_Bk_01_R!(ComponentPose=@n45, Alpha:float=1, ExternalForce="0,0,0") -> (Pose) {
    angularDampingOverride=1
    angularSpringConstant=500
    bAngularSpring=true
    bOverrideAngularDamping=true
    bUseGravityOverride=true
    linearDampingOverride=0.9999
    simulationSpace=World
    alphaBoolBlend {blendOption=Linear}
    boundBone {boneName=Dyn_skirt_Bk_01_R}
    componentAppliedLinearAccClamp=<2000 2000 2000>
    gravityOverride {z=-980}
    physicsBodyDefinitions [{
      boundBone {boneName=Dyn_skirt_Bk_01_R}
      boxExtents=<10 10 10>
      constraintSetup {angularLimitsMax=<0 0 0>, angularLimitsMin=<0 -25 0>, angularTarget=<1 0 0>}
      localJointOffset=<0.3 0 0>
     }]
   }
   n47 = Anim Dynamics - Bone: Dyn_skirt_Bk_02_R!(ComponentPose=@n46, Alpha:float=1, ExternalForce="0,0,0") -> (Pose) {
    angularDampingOverride=1
    angularSpringConstant=500
    bAngularSpring=true
    bOverrideAngularDamping=true
    bUseGravityOverride=true
    linearDampingOverride=0.9999
    simulationSpace=World
    alphaBoolBlend {blendOption=Linear}
    boundBone {boneName=Dyn_skirt_Bk_02_R}
    componentAppliedLinearAccClamp=<2000 2000 2000>
    gravityOverride {z=-980}
    physicsBodyDefinitions [{
      boundBone {boneName=Dyn_skirt_Bk_02_R}
      boxExtents=<10 10 10>
      constraintSetup {angularLimitsMax=<0 0 0>, angularLimitsMin=<0 -15 0>, angularTarget=<1 0 0>}
      localJointOffset=<0.3 0 0>
     }]
   }
   n48 = "Transform (Modify) Bone - Bone: Base_skirt_Fr_R"!(ComponentPose=@n47, bAlphaBoolEnabled:bool=True, Translation="0,0,0", Rotation="10,0,0", Scale="1,1,1") -> (Pose) {
    alphaInputType=Bool
    rotationMode=BMM_Additive
    scaleMode=BMM_Additive
    translationMode=BMM_Additive
    alphaBoolBlend {blendOption=Linear}
    boneToModify {boneName=Base_skirt_Fr_R}
    rotation {pitch=10}
   }
   Output Pose(Result=Blend Poses by bool(BlendPose_0=Component To Local(ComponentPose=@n48), BlendPose_1=Component To Local(ComponentPose=@n30)) {blendType=Linear}) {layerGroup=None}
   n49 = Control Rig!(
    Source=Copy Pose From Mesh() {bUseAttachedParent=true}
    Alpha
    PlayingEmote:bool=Get bIsPlayingEmote()
    Velocity:double=Get Velocity character() -> (Velocity character:double)
    DBNO:bool=Get bIsDBNO()
    Lod2:bool=Get Lod2() -> (Lod2:bool)
    Gauntlets:bool=Get GauntlesEquip() -> (GauntlesEquip:bool)
    NotRban:bool=Get NotRban() -> (NotRban:bool)
    FrontEndOn:bool=Get bIsInFrontEnd()
   ) -> (Pose) {
    alphaBoolBlend {blendOption=Linear}
    controlRigAssetReference {blueprintRigClass=&/F_MED_ToadLeaf_CtrlRig.F_MED_ToadLeaf_CtrlRig_C}
   }
   n50 = "Constraint - Master_knife (PPPPPP)"!(
    ComponentPose=Local To Component(LocalPose=@n49)
    Alpha=1
    ConstraintWeights_0=0.0046
    ConstraintWeights_1:float=0.3747
    ConstraintWeights_2:float=0.1038
    ConstraintWeights_3:float=0.357
    ConstraintWeights_4:float=0.0415
    ConstraintWeights_5:float=0.1184
   ) -> (Pose) {
    alphaBoolBlend {blendOption=Linear}
    boneToModify {boneName=Master_knife}
    constraintSetup [
     {targetBone {boneName=spine_05}}
     {targetBone {boneName=spine_04}}
     {targetBone {boneName=spine_03}}
     {targetBone {boneName=spine_02}}
     {targetBone {boneName=spine_01}}
     {targetBone {boneName=pelvis}}
    ]
    constraintWeights [1,1,1,1,1,1]
   }
   n51 = "Constraint - Base_string (PPPPPPPP)"!(
    ComponentPose=@n50
    Alpha=1
    ConstraintWeights_0=0.0031
    ConstraintWeights_1:float=0.0056
    ConstraintWeights_2:float=0.5926
    ConstraintWeights_3:float=0.038
    ConstraintWeights_4:float=0.2702
    ConstraintWeights_5:float=0.001
    ConstraintWeights_6:float=0.0007
    ConstraintWeights_7:float=0.0888
   ) -> (Pose) {
    alphaBoolBlend {blendOption=Linear}
    boneToModify {boneName=Base_string}
    constraintSetup [
     {targetBone {boneName=neck_02}}
     {targetBone {boneName=neck_01}}
     {targetBone {boneName=spine_05}}
     {targetBone {boneName=clavicle_l}}
     {targetBone {boneName=spine_04}}
     {targetBone {boneName=upperarm_twist_01_l}}
     {targetBone {boneName=spine_03}}
     {targetBone {boneName=deform_pec_l}}
    ]
    constraintWeights [1,1,1,1,1,1,1,1]
   }
   n52 = "Transform (Modify) Bone - Bone: attach_cape"!(ComponentPose=@n51, bAlphaBoolEnabled:bool=Get Lod1() -> (Lod1:bool), Translation="0,-1,0", Rotation="4,0,7", Scale="1,1,1") -> (Pose) {
    alphaInputType=Bool
    rotationMode=BMM_Additive
    translationMode=BMM_Additive
    alphaBoolBlend {blendOption=Linear}
    boneToModify {boneName=attach_cape}
    rotation {pitch=4, roll=7}
    translation {y=-1}
   }
   n53 = Get Lod1_cape() -> (Lod1_cape:bool)
   n54 = "Transform (Modify) Bone - Bone: Base_Cape_C"!(ComponentPose=@n52, bAlphaBoolEnabled:bool=@n53, Translation="0,0,0", Rotation="0,0,-5", Scale="1,1,1") -> (Pose) {
    alphaInputType=Bool
    rotationMode=BMM_Additive
    translationMode=BMM_Additive
    alphaBoolBlend {blendOption=Linear}
    boneToModify {boneName=Base_Cape_C}
    rotation {roll=-5}
   }
   n55 = "Transform (Modify) Bone - Bone: attach_cape"(ComponentPose=@n54, bAlphaBoolEnabled=@n53, Translation="0,-0.5,0", Rotation="0,0,7", Scale="1,1,1") {
    alphaInputType=Bool
    rotationMode=BMM_Additive
    translationMode=BMM_Additive
    alphaBoolBlend {blendOption=Linear}
    boneToModify {boneName=attach_cape}
    rotation {roll=7}
    translation {y=-0.5}
   }
   Save cached pose 'Base'(Pose=Component To Local(ComponentPose=@n55)) {cacheName=Base}
   n56 = RigidBody(ComponentPose=Local To Component(LocalPose=Use cached pose 'Base'()), OverrideWorldGravity="0,0,-980") {
    bClampLinearTranslationLimitToRefPose=true
    bOverrideWorldGravity=true
    overridePhysicsAsset=&/F_MED_ToadLeaf_Physics_Cape
    simulationSpace=BaseBoneSpace
    alphaBoolBlend {blendOption=Linear}
    baseBoneRef {boneName=spine_05}
    componentAppliedLinearAccClamp=<1000 1000 1000>
    overrideWorldGravity {z=-980}
    simSpaceSettings {
     dampingAlpha=0
     maxAngularAcceleration=100
     maxAngularVelocity=10
     maxLinearAcceleration=1300
     maxLinearVelocity=1300
     worldAlpha=1
     externalLinearDragV=<0.1 0.1 0.1>
    }
   }
   n57 = Component To Local(ComponentPose=@n56)
   n58 = RigidBody(ComponentPose=Local To Component(LocalPose=Use cached pose 'Base'()), OverrideWorldGravity="0,0,-980") {
    bClampLinearTranslationLimitToRefPose=true
    bOverrideWorldGravity=true
    overridePhysicsAsset=&/F_MED_ToadLeaf_Physics_Cape_FrontEnd2
    simulationSpace=BaseBoneSpace
    alphaBoolBlend {blendOption=Linear}
    baseBoneRef {boneName=spine_05}
    componentAppliedLinearAccClamp=<1000 1000 1000>
    overrideWorldGravity {z=-980}
    simSpaceSettings {
     dampingAlpha=0
     maxAngularAcceleration=100
     maxAngularVelocity=10
     maxLinearAcceleration=1300
     maxLinearVelocity=1300
     worldAlpha=1
     externalLinearDragV=<0.1 0.1 0.1>
    }
   }
   n59 = Component To Local(ComponentPose=@n58)
   n60 = RigidBody(ComponentPose=Local To Component(LocalPose=Use cached pose 'Base'()), OverrideWorldGravity="0,0,-980") {
    bClampLinearTranslationLimitToRefPose=true
    bOverrideWorldGravity=true
    overridePhysicsAsset=&/F_MED_ToadLeaf_Physics_Cape_withBackpack
    simulationSpace=BaseBoneSpace
    alphaBoolBlend {blendOption=Linear}
    baseBoneRef {boneName=spine_05}
    componentAppliedLinearAccClamp=<1000 1000 1000>
    overrideWorldGravity {z=-980}
    simSpaceSettings {
     dampingAlpha=0
     maxAngularAcceleration=100
     maxAngularVelocity=10
     maxLinearAcceleration=1300
     maxLinearVelocity=1300
     worldAlpha=1
     externalLinearDragV=<0.1 0.1 0.1>
    }
   }
   Save cached pose 'For Frontebd'(Pose=Blend Poses by bool(
     BlendPose_0=Component To Local(ComponentPose=@n60)
     BlendPose_1=Blend Poses by bool(BlendPose_0=@n59, BlendPose_1=@n57) {blendType=Linear}
     bActiveValue=Get Is Backbling A Cape() -> (Is Backbling A Cape:bool)
    ) {blendType=Linear}) {cacheName=For Frontebd}
  ]}
 EventGraph {nodes [
   Event BlueprintUpdateAnimation()
   > n1 = Sequence!(execute) -> (then_0, then_1, then_2:Exec, then_3:Exec)
   > n2 = Gravity Override State Function DEPRECATED(CharacterPartAnimInstance=Self-Reference(), GameTimeInSeconds=GetGameTimeInSeconds(), GravityOverrideParams=Get Gravity Override Params())
   > Set Gravity Override(Gravity Override=@n2)
   n3 = CheckForBackbling()
   > Set Is Backbling A Cape(Is Backbling A Cape:bool=OR Boolean(A=@n3, B=@n3.bIsBackblingACape)) -> (Output_Get:bool)
   n4 = Get As Fort Player Pawn Athena() -> (As Fort Player Pawn Athena:Fort Player Pawn Athena Object Reference)
   > Is Valid(InputObject=@n4)
   n5 = GetDisplayName(Object=GetSkeletalMeshForPartType(self=@n4, PartType=Gameplay))
   > Set GauntlesEquip(GauntlesEquip:bool=OR Boolean!(
     A=Contains(SearchIn=@n5, Substring=GauntletsTemplate_Gadget)
     B=Contains(SearchIn=@n5, Substring=SprocketPoppy_Gauntlet)
     C:bool=Contains(SearchIn=@n5, Substring=CoffeePuff_Gauntlet)
    ) -> (ReturnValue)) -> (Output_Get:bool)
   n6 = Cast To Character!(execute=@n1.then_1, Object=GetOwningActor()) -> (then, CastFailed, AsCharacter:Character Object Reference) {is=title K2Node_DynamicCast}
   > n7 = Set As Character(As Character:Character Object Reference=@n6) -> (Output_Get:Character Object Reference)
   > n8 = Set Character Movement(Character Movement:Character Movement Component Object Reference=Get CharacterMovement!(self:Character Object Reference=@n7) -> (CharacterMovement:Character Movement Component Object Reference)) -> (Output_Get:Character Movement Component Object Reference)
   > Set Velocity character(Velocity character:double=Vector Length(A=Get Velocity!(self:Movement Component Object Reference=@n8) -> (Velocity:Vector))) -> (Output_Get:double)
   n9 = Get CurrentLOD()
   > n10 = Set Lod1(Lod1:bool="integer >= integer"(A=@n9, B=1)) -> (Output_Get:bool)
   n11 = NOT Boolean(A=Get bRigidBodyEnabled())
   > Set Lod1_cape(Lod1_cape:bool=OR Boolean(A=@n3.bIsBackblingACape, B=@n10, C=@n11)) -> (Output_Get:bool)
   > Set NotRban(NotRban:bool=@n11) -> (Output_Get:bool)
   > Set Lod2(Lod2:bool="integer >= integer"(A=@n9, B=2)) -> (Output_Get:bool)
   n12 = Get bIsPlayingEmote()
   Set FrontEndNotEmote(execute=@n1.then_2, FrontEndNotEmote:bool=AND Boolean(A=Get bIsInFrontEnd(), B=NOT Boolean(A=@n12))) -> (Output_Get:bool)
   n13 = Get bRigidBodyEnabled()
   n14 = Branch(execute=@n1.then_3, Condition=AND Boolean(A=Get bIsPlayingEmote(), B=@n13))
   n15 = "To Float (Boolean)"(InBool=@n13)
   > Set FrontEndAninDyn_Decrease(FrontEndAninDyn_Decrease:double=float * float(A=@n15, B=0.5) {is=title K2Node_PromotableOperator}) -> (Output_Get:double)
   Set FrontEndAninDyn_Decrease(execute=@n14.else, FrontEndAninDyn_Decrease=float * float(A=@n15, B=0.05))
   Event BlueprintInitializeAnimation()
   n16 = TryGetPawnOwner()
   > Is Valid(InputObject=@n16)
   > n17 = Cast To FortPlayerPawnAthena(Object=@n16)
   > Set As Fort Player Pawn Athena(As Fort Player Pawn Athena:Fort Player Pawn Athena Object Reference=@n17) -> (Output_Get:Fort Player Pawn Athena Object Reference)
  ]}
]
variables [
 As Character {type=Character, default=null}
 As Fort Player Pawn Athena {type=FortPlayerPawnAthena}
 Character Movement {type=CharacterMovementComponent, default=null}
 FrontEndAninDyn_Decrease {type=number}
 FrontEndNotEmote {type=boolean}
 GauntlesEquip {type=boolean}
 Gravity Override {type=Vector}
 Gravity Override Params {type=GravityOverrideParamsStruct}
 Is Backbling A Cape {type=boolean}
 Lod1 {type=boolean}
 Lod1_cape {type=boolean}
 Lod2 {type=boolean}
 NotRban {type=boolean}
 Velocity character {type=number}
]
