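# Front-end physics asset for the cape and string of the female medium ToadLeaf player body.
# Capsules on the upper spine, clavicles and upper arms form the body, joined by free constraints that also attach tiny base anchors.
# Dynamic capsules hang from those anchors with tight angular limits: a center cape splitting into two two-segment back chains, single left and right pieces on the upper arms, and a two-segment string.
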PhysicsAsset /BRCosmetics/Characters/Player/Female/Medium/Bodies/F_MED_ToadLeaf/Meshes/F_MED_ToadLeaf_Physics_Cape_FrontEnd
bodies {
 Base_Cape_C [{shapeType=Capsule, length=0.505, radius=0.505, shapeName=Base_Cape_C_capsule}]
 Base_Cape_L [{shapeType=Capsule, length=0.505, radius=0.505, shapeName=Base_Cape_L_capsule, rotation=<rot 0 -90 0>}]
 Base_Cape_R [{shapeType=Capsule, length=0.505, radius=0.505, shapeName=Base_Cape_R_capsule, rotation=<rot 0 -90 0>}]
 Base_string [{shapeType=Capsule, length=0.505, radius=0.505, shapeName=Base_string_capsule, rotation=<rot 0 -90 0>}]
 Dyn_Cape_Bk_01_L [{shapeType=Capsule, length=4.6859, radius=5.0577, shapeName=Dyn_Cape_Bk_01_L_capsule, center=<4.5 0 3.35>, rotation=<rot 0 0 90>}]
 Dyn_Cape_Bk_01_R [{shapeType=Capsule, length=4.6859, radius=5.0577, shapeName=Dyn_Cape_Bk_01_R_capsule, center=<4.5 0 3.35>, rotation=<rot 0 0 90>}]
 Dyn_Cape_Bk_02_L [{shapeType=Capsule, length=4.6859, radius=5.0577, shapeName=Dyn_Cape_Bk_02_L_capsule, center=<4.5 0 3.35>, rotation=<rot 0 0 90>}]
 Dyn_Cape_Bk_02_R [{shapeType=Capsule, length=4.6859, radius=5.0577, shapeName=Dyn_Cape_Bk_02_R_capsule, center=<4.5 0 3.35>, rotation=<rot 0 0 90>}]
 Dyn_Cape_C_01 [{shapeType=Capsule, length=4.6859, radius=5.0577, shapeName=Dyn_Cape_C_01_capsule, center=<6.5 0 3.35>, rotation=<rot 0 0 90>}]
 Dyn_Cape_L [{
   shapeType=Capsule
   length=7.0882
   radius=2.5706
   shapeName=Dyn_Cape_L_capsule
   center=<6.1246 9.1981 -3.9828>
   rotation=<rot -0.0001 -19.9999 90.0002>
  }]
 Dyn_Cape_R [{
   shapeType=Capsule
   length=7.5334
   radius=2.2474
   shapeName=Dyn_Cape_R_capsule
   center=<5.3874 -9.2361 -3.2602>
   rotation=<rot -0.0001 15 89.9999>
  }]
 Dyn_string_01 [{shapeType=Capsule, length=2.0298, radius=1.1732, shapeName=Dyn_string_01_capsule, center=<2.8351 0 0>, rotation=<rot 90 0 0>}]
 Dyn_string_02 [{shapeType=Capsule, length=2.2965, radius=1.1269, shapeName=Dyn_string_02_capsule, center=<2.8351 0 0>, rotation=<rot 90 0 0>}]
 clavicle_l [{shapeType=Capsule, length=5.8836, radius=5.5086, shapeName=clavicle_l_capsule, center=<5.0884 0 0>, rotation=<rot 90 0 0>}]
 clavicle_r [{shapeType=Capsule, length=5.8836, radius=5.5086, shapeName=clavicle_r_capsule, center=<-5.0884 0 0>, rotation=<rot 90 0 180>}]
 spine_02 [{shapeType=Capsule, length=9.3898, radius=6.8718, shapeName=spine_03_capsule, center=<0.8583 -1.6021 0>}]
 spine_03 [{shapeType=Capsule, length=8.5222, radius=7.7394, shapeName=spine_03_capsule, center=<-0.1077 -1.0994 0>}]
 spine_04 [{shapeType=Capsule, length=6.5932, radius=9.6683, shapeName=spine_04_capsule, center=<0 1.8557 0>}]
 spine_05 [{shapeType=Capsule, length=7.7275, radius=8.534, shapeName=spine_05_capsule, center=<-0.5998 1.0044 0>}]
 upperarm_l [{shapeType=Capsule, length=7.4469, radius=6.6401, shapeName=upperarm_l_capsule, center=<9 0 0>, rotation=<rot 90 0 0>}]
 upperarm_r [{shapeType=Capsule, length=7.4469, radius=6.6401, shapeName=upperarm_r_capsule, center=<-9 0 0>, rotation=<rot 90 0 0>}]
}
constraints [
 {
  bone1Name=Base_Cape_C
  bone2Name=spine_05
  swing1LimitDegrees=45
  swing1Motion=Free
  swing2LimitDegrees=45
  swing2Motion=Free
  twistLimitDegrees=45
  twistMotion=Free
 }
 {bone1Name=Dyn_Cape_C_01, bone2Name=Base_Cape_C, swing1LimitDegrees=10, swing2LimitDegrees=25, twistLimitDegrees=5}
 {bone1Name=Dyn_Cape_Bk_01_L, bone2Name=Dyn_Cape_C_01, swing1LimitDegrees=10, swing2LimitDegrees=18, twistLimitDegrees=5}
 {bone1Name=Dyn_Cape_Bk_02_L, bone2Name=Dyn_Cape_Bk_01_L, swing1LimitDegrees=5, swing2LimitDegrees=15, twistLimitDegrees=5}
 {bone1Name=Dyn_Cape_Bk_01_R, bone2Name=Dyn_Cape_C_01, swing1LimitDegrees=10, swing2LimitDegrees=18, twistLimitDegrees=5}
 {bone1Name=Dyn_Cape_Bk_02_R, bone2Name=Dyn_Cape_Bk_01_R, swing1LimitDegrees=5, swing2LimitDegrees=15, twistLimitDegrees=5}
 {
  bone1Name=clavicle_l
  bone2Name=spine_05
  swing1LimitDegrees=45
  swing1Motion=Free
  swing2LimitDegrees=45
  swing2Motion=Free
  twistLimitDegrees=45
  twistMotion=Free
 }
 {
  bone1Name=upperarm_l
  bone2Name=clavicle_l
  swing1LimitDegrees=45
  swing1Motion=Free
  swing2LimitDegrees=45
  swing2Motion=Free
  twistLimitDegrees=45
  twistMotion=Free
 }
 {
  bone1Name=upperarm_r
  bone2Name=clavicle_r
  swing1LimitDegrees=45
  swing1Motion=Free
  swing2LimitDegrees=45
  swing2Motion=Free
  twistLimitDegrees=45
  twistMotion=Free
 }
 {
  bone1Name=spine_05
  bone2Name=spine_04
  swing1LimitDegrees=45
  swing1Motion=Free
  swing2LimitDegrees=45
  swing2Motion=Free
  twistLimitDegrees=45
  twistMotion=Free
 }
 {
  bone1Name=clavicle_r
  bone2Name=spine_04
  swing1LimitDegrees=45
  swing1Motion=Free
  swing2LimitDegrees=45
  swing2Motion=Free
  twistLimitDegrees=45
  twistMotion=Free
 }
 {
  bone1Name=spine_04
  bone2Name=spine_03
  swing1LimitDegrees=45
  swing1Motion=Free
  swing2LimitDegrees=45
  swing2Motion=Free
  twistLimitDegrees=45
  twistMotion=Free
 }
 {bone1Name=Dyn_Cape_L, bone2Name=Base_Cape_L, swing1LimitDegrees=10, swing2LimitDegrees=15, twistLimitDegrees=0}
 {bone1Name=Dyn_Cape_R, bone2Name=Base_Cape_R, swing1LimitDegrees=10, swing2LimitDegrees=15, twistLimitDegrees=0}
 {
  bone1Name=Base_Cape_L
  bone2Name=upperarm_l
  swing1LimitDegrees=45
  swing1Motion=Free
  swing2LimitDegrees=45
  swing2Motion=Free
  twistLimitDegrees=45
  twistMotion=Free
 }
 {
  bone1Name=Base_Cape_R
  bone2Name=upperarm_r
  swing1LimitDegrees=45
  swing1Motion=Free
  swing2LimitDegrees=45
  swing2Motion=Free
  twistLimitDegrees=45
  twistMotion=Free
 }
 {
  bone1Name=Base_string
  bone2Name=spine_05
  swing1LimitDegrees=45
  swing1Motion=Free
  swing2LimitDegrees=45
  swing2Motion=Free
  twistLimitDegrees=45
  twistMotion=Free
 }
 {bone1Name=Dyn_string_01, bone2Name=Base_string, swing1LimitDegrees=10, swing2LimitDegrees=10, twistLimitDegrees=0}
 {bone1Name=Dyn_string_02, bone2Name=Dyn_string_01, swing1LimitDegrees=10, swing2LimitDegrees=10, twistLimitDegrees=0}
 {
  bone1Name=spine_03
  bone2Name=spine_02
  swing1LimitDegrees=45
  swing1Motion=Free
  swing2LimitDegrees=45
  swing2Motion=Free
  twistLimitDegrees=45
  twistMotion=Free
 }
]
