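# Corrective Control Rig for the ToadLeaf female face accessory that repositions hair bones as the head moves.
# Head rotation relative to spine_05 is remapped into weights that blend offsets onto the Base_hair and Vorrect_hair bones.
# A second pass scales two staged rotation ranges by how far the pelvis drops, suggesting crouch correction.
# When PlayingEmote is true, a further correction follows the head's yaw delta from reference.
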ControlRigBlueprint /BRCosmetics/Characters/Player/Female/Medium/Bodies/F_MED_ToadLeaf/Meshes/Parts/F_MED_ToadLeaf_FaceAcc_CtrlRig
graphs [
 RigVMFunctionLibrary {links [], nodes []}
 RigVMFunction_Sequence_1_ContainedGraph {nodes [
   n1 = RigVMFunctionEntryNode()
   RigVMFunction_Sequence = &RigVMFunction_Sequence(ExecuteContext=@n1.ExecuteContext)
   n2 = &RigVMFunction_Sequence(ExecuteContext=@RigVMFunction_Sequence.B)
   RigVMFunctionReturnNode(A=@RigVMFunction_Sequence.A, B=@n2.A, C=@n2.B)
  ]}
 RigVMModel {nodes [
   RigUnit_BeginExecution = &RigUnit_BeginExecution()
   RigVMFunction_Sequence = RigVMAggregateNode(ExecuteContext=@RigUnit_BeginExecution.ExecutePin)
   RigVMFunction_ControlFlowBranch = &RigVMFunction_ControlFlowBranch(ExecuteContext=@RigVMFunction_Sequence.C, Condition=RigVMVariableNode() {variable=PlayingEmote}.Value)
   Get Delta Ref = RigVMFunctionReferenceNode(Item=(Type=Bone,Name="head"), ExecuteContext=@RigVMFunction_ControlFlowBranch.True)
   n1 = &RigVMFunction_MathDoubleRemap(SourceMaximum=-40, SourceMinimum=-10, TargetMaximum=1, TargetMinimum=0, Value=@"Get Delta Ref".Result.Rotation.Yaw)
   > &RigUnit_ModifyTransforms(
    ItemToModify=((Item=(Type=Bone,Name="Base_hair_C"),Transform=(Rotation=(X=0,Y=0,Z=-0.3007,W=0.9537),Translation=(X=0,Y=-1,Z=0),Scale3D=(X=1,Y=1,Z=1))))
    WeightMinimum=0
    Weight=@n1.Result
   )
   > &RigUnit_ModifyTransforms(
    ItemToModify=((Item=(Type=Bone,Name="Vorrect_hair_C"),Transform=(Rotation=(X=0,Y=0,Z=0,W=1),Translation=(X=-4,Y=-4,Z=0),Scale3D=(X=1,Y=1,Z=1))))
    WeightMinimum=0
    Weight=@n1.Result
   )
   n2 = &RigVMFunction_MathQuaternionToEuler(
    RotationOrder=ZYX
    Value=&RigUnit_GetRelativeTransformForItem(Child=(Type=Bone,Name="head"), Parent=(Type=Bone,Name="spine_05"), bChildInitial=False, bParentInitial=False).RelativeTransform.Rotation
   )
   n3 = &RigVMFunction_MathDoubleRemap(
    SourceMaximum=75
    SourceMinimum=90
    TargetMaximum=1
    TargetMinimum=0
    Value=&RigUnit_GetRelativeTransformForItem(Child=(Type=Bone,Name="pelvis"), Parent=(Type=Bone,Name="root"), bChildInitial=False, bParentInitial=False).RelativeTransform.Translation.Z
   )
   Multiply = &RigVMFunction_MathDoubleMul(
    A=&RigVMFunction_MathDoubleRemap(SourceMaximum=-35, SourceMinimum=-15, TargetMaximum=1, TargetMinimum=0, Value=@n2.Result.Z).Result
    B=@n3.Result
   )
   &RigUnit_ModifyTransforms(
    ItemToModify=((Item=(Type=Bone,Name="Base_hair_C"),Transform=(Rotation=(X=0,Y=0,Z=-0.0436,W=0.999),Translation=(X=0,Y=0,Z=0),Scale3D=(X=1,Y=1,Z=1))))
    WeightMinimum=0
    Weight=@Multiply.Result
    ExecutePin=@RigVMFunction_Sequence.B
   )
   > &RigUnit_ModifyTransforms(
    ItemToModify=((Item=(Type=Bone,Name="Base_hair_R"),Transform=(Rotation=(X=0,Y=0,Z=0,W=1),Translation=(X=-1,Y=0,Z=0),Scale3D=(X=1,Y=1,Z=1))))
    WeightMinimum=0
    Weight=@Multiply.Result
   )
   n4 = &RigVMFunction_MathDoubleMul(
    B=@n3.Result
    A=&RigVMFunction_MathDoubleRemap(SourceMaximum=-50, SourceMinimum=-35, TargetMaximum=1, TargetMinimum=0, Value=@n2.Result.Z).Result
   )
   > &RigUnit_ModifyTransforms(
    ItemToModify=((Item=(Type=Bone,Name="Base_hair_C"),Transform=(Rotation=(X=0,Y=0,Z=-0.1305,W=0.9914),Translation=(X=0,Y=-1,Z=0),Scale3D=(X=1,Y=1,Z=1))))
    WeightMinimum=0
    Weight=@n4.Result
   )
   > &RigUnit_ModifyTransforms(
    ItemToModify=((Item=(Type=Bone,Name="Base_hair_R"),Transform=(Rotation=(X=0,Y=0,Z=0.2588,W=0.9659),Translation=(X=0,Y=0,Z=0),Scale3D=(X=1,Y=1,Z=1))))
    WeightMinimum=0
    Weight=@n4.Result
   )
   > &RigUnit_ModifyTransforms(
    ItemToModify=((Item=(Type=Bone,Name="Vorrect_hair_C"),Transform=(Rotation=(X=0,Y=0,Z=0,W=1),Translation=(X=-3,Y=0,Z=0),Scale3D=(X=1,Y=1,Z=1))))
    WeightMinimum=0
    Weight=@n4.Result
   )
   n5 = &RigVMFunction_MathDoubleRemap(
    SourceMaximum=-30
    SourceMinimum=0
    TargetMaximum=1
    TargetMinimum=0
    Value=&RigVMFunction_MathQuaternionToEuler(
     RotationOrder=ZYX
     Value=&RigUnit_GetRelativeTransformForItem(Child=(Type=Bone,Name="head"), Parent=(Type=Bone,Name="spine_05"), bChildInitial=False, bParentInitial=False).RelativeTransform.Rotation
    ).Result.Z
   )
   &RigUnit_ModifyTransforms(
    ItemToModify=((Item=(Type=Bone,Name="Base_hair_C"),Transform=(Rotation=(X=0.0149,Y=0.041,Z=-0.3417,W=0.9388),Translation=(X=0,Y=1,Z=0),Scale3D=(X=1,Y=1,Z=1))))
    WeightMinimum=0
    Weight=@n5.Result
    ExecutePin=@RigVMFunction_Sequence.A
   )
   > &RigUnit_ModifyTransforms(
    ItemToModify=((Item=(Type=Bone,Name="Vorrect_hair_C"),Transform=(Rotation=(X=0,Y=0,Z=0,W=1),Translation=(X=-5,Y=-5,Z=0),Scale3D=(X=1,Y=1,Z=1))))
    WeightMinimum=0
    Weight=@n5.Result
   )
   > &RigUnit_ModifyTransforms(
    ItemToModify=((Item=(Type=Bone,Name="Base_hair_L"),Transform=(Rotation=(X=0,Y=-0.1305,Z=0,W=0.9914),Translation=(X=-1,Y=0,Z=1),Scale3D=(X=1,Y=1,Z=1))))
    WeightMinimum=0
    Weight=@n5.Result
   )
   > &RigUnit_ModifyTransforms(
    ItemToModify=((Item=(Type=Bone,Name="Base_hair_R"),Transform=(Rotation=(X=0,Y=0,Z=-0.0872,W=0.9962),Translation=(X=-1,Y=0,Z=0),Scale3D=(X=1,Y=1,Z=1))))
    WeightMinimum=0
    Weight=@n5.Result
   )
  ]}
]
hierarchy [
 root {loc=<0 0 0>, rot=<0 0 0 1>}
 pelvis {parent=root, loc=<0 -0.5301 98.0571>, rot=<0.0127 0.707 -0.0127 -0.707>}
 spine_01 {parent=pelvis, loc=<6.1208 -0.9328 0>, rot=<0 0 0.034 -0.9994>}
 spine_02 {parent=spine_01, loc=<6.5997 -0.2095 0>, rot=<0 0 0.0238 -0.9997>}
 spine_03 {parent=spine_02, loc=<6.4914 -0.1237 0>, rot=<0 0 0.0149 -0.9999>}
 spine_04 {parent=spine_03, loc=<6.9598 0.0794 0>, rot=<0 0 -0.0132 -0.9999>}
 spine_05 {parent=spine_04, loc=<6.9987 0 0>, rot=<0 0 -0.0015 -1>}
 neck_01 {parent=spine_05, loc=<11.3138 0.1529 0>, rot=<0 0 -0.2003 -0.9797>}
 neck_02 {parent=neck_01, loc=<4.5627 0 0>, rot=<0 0 0.0504 -0.9987>}
 head {parent=neck_02, loc=<5.3914 0 0>, rot=<0 0 0.1108 -0.9938>}
 Hair_Side_L_02 {parent=head, loc=<2.182 -0.3558 -7.7958>, rot=<-0.0653 -0.9869 -0.0663 -0.132>}
 Hair_Side_R_01 {parent=head, loc=<4.1391 4.2089 7.2946>, rot=<-0.1482 0.1997 0.9163 0.314>}
 Hair_Side_R_02 {parent=head, loc=<3.3082 -0.5936 8.6459>, rot=<-0.0746 -0.0949 0.972 0.2015>}
 Hair_Side_R_03 {parent=head, loc=<5.8669 2.8204 8.2768>, rot=<0.4502 -0.103 0.8866 -0.0268>}
 Hair_Front_01 {parent=head, loc=<7.0421 6.61 5.8414>, rot=<-0.036 0.3663 0.9139 0.1713>}
 Hair_Front_02 {parent=head, loc=<8.232 8.5653 3.2609>, rot=<-0.042 0.5179 0.8543 0.0145>}
 Hair_Front_03 {parent=head, loc=<8.539 9.1889 -0.4135>, rot=<0.0107 -0.751 -0.6602 -0.0017>}
 Hair_Front_04 {parent=head, loc=<8.3146 8.1378 -3.5523>, rot=<-0.0568 -0.8405 -0.538 0.03>}
 Hair_Bk_R_01 {parent=head, loc=<-2.6473 -0.2013 6.3686>, rot=<0.0695 -0.6007 0.7448 0.2821>}
 Hair_Front_05 {parent=head, loc=<8.0893 6.9092 -4.8078>, rot=<-0.0712 -0.9141 -0.3987 -0.0183>}
 Hair_Side_L_01 {parent=head, loc=<3.9088 3.1807 -7.7947>, rot=<-0.2528 -0.935 -0.209 0.1347>}
 Base_hair_C {parent=head, loc=<0.4314 -5.6986 -0.4479>, rot=<0 -1 0 0>}
 Hair_Bk_C_01 {parent=Base_hair_C, loc=<3.7246 -0.2179 0.5926>, rot=<0 0 0.1979 -0.9802>}
 Hair_Bk_C_02 {parent=Base_hair_C, loc=<0.7601 -1.4576 1.2311>, rot=<0.1663 -0.0482 0.1147 -0.9782>}
 Hair_Bk_L_03 {parent=Base_hair_C, loc=<-0.2225 -1.3859 -3.8571>, rot=<-0.2551 -0.0812 0.1488 -0.952>}
 Base_hair_L {parent=Base_hair_C, loc=<0.2873 1.3925 3.7265>, rot=<-0.3789 0.2621 -0.1268 -0.8785>}
 Hair_Bk_L_01 {parent=Base_hair_L, loc=<2.4112 1.8583 -0.6875>, rot=<0.0975 -0.1442 0.0337 -0.9842>}
 Hair_Bk_L_02 {parent=Base_hair_L, loc=<3.0113 -1.0361 -0.3462>, rot=<0 0 0 1>}
 Base_hair_R {parent=Base_hair_C, loc=<-0.4671 2.9172 -5.0047>, rot=<-0.4437 -0.1219 0.0235 -0.8876>}
 Hair_Bk_R_02 {parent=Base_hair_R, loc=<3.661 0.3423 2.8858>, rot=<0.1346 0.0389 0.1204 -0.9828>}
 Vorrect_hair_C {parent=head, loc=<6.1325 -7.8622 -0.4479>, rot=<0 -1 0 0>}
 FACIAL_C_FacialRoot {parent=head, loc=<2.1462 -0.7638 -0.0002>, rot=<0.4858 -0.5138 -0.4858 -0.5138>}
 FACIAL_C_Forehead {parent=FACIAL_C_FacialRoot, loc=<-0.0009 -2.9256 8.867>, rot=<-0.0492 0 0 -0.9988>}
 FACIAL_L_ForeheadIn {parent=FACIAL_C_FacialRoot, loc=<1.6442 -2.8794 8.7178>, rot=<-0.0664 -0.0696 0.0039 -0.9954>}
 FACIAL_R_ForeheadIn {parent=FACIAL_C_FacialRoot, loc=<-1.6457 -2.8797 8.7163>, rot=<-0.0664 0.0697 -0.0038 -0.9953>}
 FACIAL_L_ForeheadMid {parent=FACIAL_C_FacialRoot, loc=<3.063 -3.1235 8.2093>, rot=<-0.0766 -0.1755 -0.0216 -0.9813>}
 FACIAL_R_ForeheadMid {parent=FACIAL_C_FacialRoot, loc=<-3.0637 -3.1236 8.2077>, rot=<-0.0766 0.1756 0.0216 -0.9812>}
 FACIAL_L_ForeheadOut {parent=FACIAL_C_FacialRoot, loc=<4.3265 -3.0733 7.3279>, rot=<-0.1269 -0.2551 -0.0806 -0.9552>}
 FACIAL_R_ForeheadOut {parent=FACIAL_C_FacialRoot, loc=<-4.3265 -3.073 7.3269>, rot=<-0.1269 0.2552 0.0807 -0.9551>}
 FACIAL_L_EyesackUpper {parent=FACIAL_C_FacialRoot, loc=<3.124 -2.0358 8.7328>, rot=<0.0431 -0.0838 -0.065 -0.9934>}
 FACIAL_R_EyesackUpper {parent=FACIAL_C_FacialRoot, loc=<-3.1234 -2.0365 8.734>, rot=<0.0431 0.0839 0.065 -0.9934>}
 FACIAL_L_EyelidUpperFurrow {parent=FACIAL_C_FacialRoot, loc=<2.9977 -1.8032 8.6547>, rot=<-0.0357 -0.1503 -0.0805 -0.9847>}
 FACIAL_R_EyelidUpperFurrow {parent=FACIAL_C_FacialRoot, loc=<-2.9957 -1.8051 8.6564>, rot=<-0.0357 0.1504 0.0806 -0.9847>}
 FACIAL_L_EyelidUpperA {parent=FACIAL_C_FacialRoot, loc=<2.7937 -1.0337 6.7665>, rot=<-0.1535 0 0 -0.9881>}
 FACIAL_R_EyelidUpperA {parent=FACIAL_C_FacialRoot, loc=<-2.7922 -1.0348 6.7696>, rot=<-0.1535 0 0 -0.9881>}
 FACIAL_L_Eye {parent=FACIAL_C_FacialRoot, loc=<2.7904 -1.0475 6.7647>, rot=<-0.0058 0 0 -1>}
 FACIAL_R_Eye {parent=FACIAL_C_FacialRoot, loc=<-2.7915 -1.0468 6.7674>, rot=<-0.0058 0 0 -1>}
 FACIAL_L_EyelidLowerA {parent=FACIAL_C_FacialRoot, loc=<2.7937 -1.0337 6.7665>, rot=<0.2785 0 0 -0.9604>}
 FACIAL_R_EyelidLowerA {parent=FACIAL_C_FacialRoot, loc=<-2.7922 -1.0348 6.7696>, rot=<0.2785 0 0 -0.9604>}
 FACIAL_L_EyeCornerInner {parent=FACIAL_C_FacialRoot, loc=<1.5718 -0.2293 8.0765>, rot=<-0.0058 0 0 -1>}
 FACIAL_R_EyeCornerInner {parent=FACIAL_C_FacialRoot, loc=<-1.5714 -0.2367 8.0799>, rot=<-0.0058 0 0 -1>}
 FACIAL_L_EyeCornerOuter {parent=FACIAL_C_FacialRoot, loc=<4.5678 -0.881 7.3683>, rot=<-0.0053 -0.3826 -0.0022 -0.9239>}
 FACIAL_R_EyeCornerOuter {parent=FACIAL_C_FacialRoot, loc=<-4.5669 -0.8801 7.37>, rot=<-0.0053 0.3826 0.0022 -0.9239>}
 FACIAL_L_EyesackLower {parent=FACIAL_C_FacialRoot, loc=<3.288 0.6064 7.3155>, rot=<-0.028 -0.1395 -0.023 -0.9896>}
 FACIAL_R_EyesackLower {parent=FACIAL_C_FacialRoot, loc=<-3.2867 0.6018 7.3134>, rot=<-0.028 0.1396 0.0231 -0.9895>}
 FACIAL_L_CheekInner {parent=FACIAL_C_FacialRoot, loc=<2.0453 1.315 7.7799>, rot=<-0.1288 -0.0868 -0.1976 -0.9679>}
 FACIAL_R_CheekInner {parent=FACIAL_C_FacialRoot, loc=<-2.0463 1.3157 7.7782>, rot=<-0.1288 0.0869 0.1976 -0.9679>}
 FACIAL_L_CheekOuter {parent=FACIAL_C_FacialRoot, loc=<3.8159 1.3856 6.1618>, rot=<0.0499 -0.3775 0.0982 -0.9194>}
 FACIAL_R_CheekOuter {parent=FACIAL_C_FacialRoot, loc=<-3.8151 1.3865 6.1599>, rot=<0.0499 0.3775 -0.0981 -0.9194>}
 FACIAL_L_NasolabialBulge {parent=FACIAL_C_FacialRoot, loc=<2.9755 4.6013 6.0608>, rot=<0.0526 -0.3774 0.0217 -0.9243>}
 FACIAL_R_NasolabialBulge {parent=FACIAL_C_FacialRoot, loc=<-2.978 4.6027 6.0583>, rot=<0.0526 0.3775 -0.0216 -0.9243>}
 FACIAL_L_NasolabialFurrow {parent=FACIAL_C_FacialRoot, loc=<2.7188 3.4221 8.4841>, rot=<-0.0386 -0.1237 -0.3643 -0.9222>}
 FACIAL_R_NasolabialFurrow {parent=FACIAL_C_FacialRoot, loc=<-2.7196 3.4213 8.4832>, rot=<-0.0386 0.1237 0.3644 -0.9222>}
 FACIAL_L_CheekLower {parent=FACIAL_C_FacialRoot, loc=<3.4056 4.3172 5.3527>, rot=<0.0294 -0.4272 0.0553 -0.902>}
 FACIAL_R_CheekLower {parent=FACIAL_C_FacialRoot, loc=<-3.4069 4.3182 5.35>, rot=<0.0294 0.4272 -0.0553 -0.902>}
 FACIAL_C_Nose {parent=FACIAL_C_FacialRoot, loc=<-0.0007 0.6564 9.2655>, rot=<-0.1361 0 0 -0.9907>}
 FACIAL_C_NoseLower {parent=FACIAL_C_Nose, loc=<0.0002 2.6158 -0.377>, rot=<0.1305 0 0 -0.9914>}
 FACIAL_L_Nostril {parent=FACIAL_C_Nose, loc=<0.8582 2.0803 -0.3389>, rot=<0.0173 0.1305 0.0023 -0.9913>}
 FACIAL_R_Nostril {parent=FACIAL_C_Nose, loc=<-0.8563 2.0813 -0.3429>, rot=<0.0173 -0.1305 -0.0023 -0.9913>}
 FACIAL_C_NoseTip {parent=FACIAL_C_Nose, loc=<-0.0003 1.973 0.5252>, rot=<0.0872 0 0 -0.9962>}
 FACIAL_C_MouthUpper {parent=FACIAL_C_FacialRoot, loc=<-0.0017 4.7371 7.0688>, rot=<-0.0057 0 0 -1>}
 FACIAL_C_LipUpper {parent=FACIAL_C_MouthUpper, loc=<0.0005 -0.1655 1.9674>, rot=<-0.0872 0 0 -0.9962>}
 FACIAL_L_LipUpper {parent=FACIAL_C_MouthUpper, loc=<0.8885 -0.1208 1.8453>, rot=<-0.0778 -0.1601 -0.0379 -0.9833>}
 FACIAL_R_LipUpper {parent=FACIAL_C_MouthUpper, loc=<-0.8886 -0.1214 1.8448>, rot=<-0.0778 0.1601 0.0379 -0.9833>}
 FACIAL_L_LipUpperOuter {parent=FACIAL_C_MouthUpper, loc=<1.6302 0.0083 1.3419>, rot=<-0.0984 -0.2293 -0.0983 -0.9634>}
 FACIAL_R_LipUpperOuter {parent=FACIAL_C_MouthUpper, loc=<-1.63 0.0085 1.339>, rot=<-0.0984 0.2293 0.0983 -0.9634>}
 FACIAL_L_LipCorner {parent=FACIAL_C_MouthUpper, loc=<2.2279 0.3866 0.6858>, rot=<0 -0.2078 0 -0.9782>}
 FACIAL_R_LipCorner {parent=FACIAL_C_MouthUpper, loc=<-2.2287 0.3872 0.6868>, rot=<0 0.2078 0 -0.9782>}
 FACIAL_L_Masseter {parent=FACIAL_C_FacialRoot, loc=<5.8947 1.9119 3.8935>, rot=<0.0637 0.1129 0.0086 -0.9915>}
 FACIAL_R_Masseter {parent=FACIAL_C_FacialRoot, loc=<-5.8955 1.9136 3.8914>, rot=<0.0636 -0.1128 -0.0085 -0.9915>}
 FACIAL_C_UnderChin {parent=FACIAL_C_FacialRoot, loc=<0 8.8679 3.9382>, rot=<0.4949 0 0 -0.869>}
 FACIAL_L_UnderChin {parent=FACIAL_C_FacialRoot, loc=<2.7204 7.6459 2.7899>, rot=<0.3977 -0.3553 0.0803 -0.8421>}
 FACIAL_R_UnderChin {parent=FACIAL_C_FacialRoot, loc=<-2.7213 7.6463 2.7896>, rot=<0.3977 0.3553 -0.0802 -0.8421>}
 FACIAL_C_TeethUpper {parent=FACIAL_C_FacialRoot, loc=<-0.003 5.1341 7.9669>, rot=<-0.0057 0 0 -1>}
 FACIAL_C_LowerLipRotation {parent=FACIAL_C_FacialRoot, loc=<0.0012 2.6819 2.1495>, rot=<0.3366 0 0 -0.9416>}
 FACIAL_C_MouthLower {parent=FACIAL_C_LowerLipRotation, loc=<-0.0027 -1.5294 5.1073>, rot=<-0.342 0 0 -0.9397>}
 FACIAL_C_LipLower {parent=FACIAL_C_MouthLower, loc=<-0.0016 0.9423 1.6917>, rot=<0.0872 0 0 -0.9962>}
 FACIAL_L_LipLower {parent=FACIAL_C_MouthLower, loc=<0.7999 0.9254 1.5331>, rot=<0.0927 -0.1294 0.0272 -0.9869>}
 FACIAL_R_LipLower {parent=FACIAL_C_MouthLower, loc=<-0.8038 0.9264 1.5309>, rot=<0.0927 0.1294 -0.0272 -0.9869>}
 FACIAL_L_LipLowerOuter {parent=FACIAL_C_MouthLower, loc=<1.5324 0.7667 1.0904>, rot=<0.0913 -0.2321 0.0318 -0.9679>}
 FACIAL_R_LipLowerOuter {parent=FACIAL_C_MouthLower, loc=<-1.5347 0.7668 1.0882>, rot=<0.0913 0.2321 -0.0318 -0.9679>}
 FACIAL_C_TeethLower {parent=FACIAL_C_LowerLipRotation, loc=<-0.0039 -1.7131 6.0386>, rot=<-0.342 0 0 -0.9397>}
 FACIAL_C_Tongue1 {parent=FACIAL_C_TeethLower, loc=<0.0034 0.1773 -4.6246>, rot=<-0.2588 0 0 -0.9659>}
 FACIAL_C_Tongue2 {parent=FACIAL_C_Tongue1, loc=<-0.0008 0.3914 1.3319>, rot=<0.2588 0 0 -0.9659>}
 FACIAL_C_Tongue3 {parent=FACIAL_C_Tongue2, loc=<-0.0007 0.1987 1.2451>, rot=<0.1305 0 0 -0.9914>}
 FACIAL_C_Tongue4 {parent=FACIAL_C_Tongue3, loc=<-0.0001 -0.1744 1.0793>, rot=<0.0436 0 0 -0.999>}
 FACIAL_C_TongueUpper3 {parent=FACIAL_C_Tongue3, loc=<0.0006 -0.3192 -0.1919>, rot=<0 0 0 1>}
 FACIAL_C_TongueLower3 {parent=FACIAL_C_Tongue3, loc=<0 0.8165 0.0286>, rot=<0 0 0 1>}
 FACIAL_L_TongueSide3 {parent=FACIAL_C_Tongue3, loc=<1.4976 0.0025 0.1017>, rot=<0 0 0 1>}
 FACIAL_R_TongueSide3 {parent=FACIAL_C_Tongue3, loc=<-1.4977 0.0034 0.1>, rot=<0 0 0 1>}
 FACIAL_C_TongueUpper2 {parent=FACIAL_C_Tongue2, loc=<0.0007 -0.5628 -0.2339>, rot=<0 0 0 1>}
 FACIAL_L_TongueSide2 {parent=FACIAL_C_Tongue2, loc=<1.7793 0.0005 0.122>, rot=<0 0 0 1>}
 FACIAL_R_TongueSide2 {parent=FACIAL_C_Tongue2, loc=<-1.7795 0.0011 0.12>, rot=<0 0 0 1>}
 FACIAL_C_Jaw {parent=FACIAL_C_FacialRoot, loc=<0.0012 2.6819 2.1495>, rot=<0.3366 0 0 -0.9416>}
 FACIAL_C_Jawline {parent=FACIAL_C_Jaw, loc=<-0.0017 2.2167 6.9396>, rot=<0.2586 0 0 -0.966>}
 FACIAL_L_ChinSide {parent=FACIAL_C_Jaw, loc=<2.4796 0.2882 6.7508>, rot=<-0.2208 -0.3738 0.0499 -0.8995>}
 FACIAL_R_ChinSide {parent=FACIAL_C_Jaw, loc=<-2.4812 0.288 6.7512>, rot=<-0.2208 0.3738 -0.0499 -0.8995>}
 FACIAL_C_Chin {parent=FACIAL_C_Jaw, loc=<-0.0017 0.7893 7.1548>, rot=<-0.2591 0 0 -0.9659>}
 FACIAL_L_Ear {parent=FACIAL_C_FacialRoot, loc=<6.2834 -0.0974 0.7791>, rot=<-0.032 -0.0005 -0.0261 -0.9991>}
 FACIAL_R_Ear {parent=FACIAL_C_FacialRoot, loc=<-6.2817 -0.0954 0.7736>, rot=<-0.032 0.0006 0.0262 -0.9991>}
 FACIAL_L_Jawline {parent=FACIAL_C_FacialRoot, loc=<4.0158 5.778 4.3247>, rot=<0.2038 0.1581 -0.0537 -0.9647>}
 FACIAL_R_Jawline {parent=FACIAL_C_FacialRoot, loc=<-4.0183 5.7792 4.3243>, rot=<0.2038 -0.158 0.0538 -0.9647>}
 FACIAL_C_NoseBridge {parent=FACIAL_C_FacialRoot, loc=<-0.0005 -1.3768 9.6101>, rot=<0.0381 0 0 -0.9993>}
 FACIAL_L_EyelidUpperB {parent=FACIAL_C_FacialRoot, loc=<2.7937 -1.0337 6.7665>, rot=<-0.2306 0 0 -0.9731>}
 FACIAL_L_EyelidLowerB {parent=FACIAL_C_FacialRoot, loc=<2.7937 -1.0337 6.7665>, rot=<0.3118 0 0 -0.9501>}
 FACIAL_R_EyelidLowerB {parent=FACIAL_C_FacialRoot, loc=<-2.7922 -1.0348 6.7696>, rot=<0.3201 0 0 -0.9474>}
 FACIAL_R_EyelidUpperB {parent=FACIAL_C_FacialRoot, loc=<-2.7922 -1.0348 6.7696>, rot=<-0.2306 0 0 -0.9731>}
]
variables [PlayingEmote {type=bool, default=False}]
