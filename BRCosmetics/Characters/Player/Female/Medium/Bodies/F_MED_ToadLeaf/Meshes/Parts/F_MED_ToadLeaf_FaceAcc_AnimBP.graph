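# Animation Blueprint for the ToadLeaf female cosmetic's FaceAcc mesh that adds secondary motion to its hair.
# Copies the attached parent's pose into a Control Rig, which also receives bIsPlayingEmote, then runs chained single-bone Anim Dynamics nodes with springs, damping and gravity.
# These cover side, back and front hair bones; front-section nodes allow wider swing.
# Hair_Side_R_01 and _02 are simulated twice, the second time among the front nodes with those wider limits.
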AnimBlueprint /BRCosmetics/Characters/Player/Female/Medium/Bodies/F_MED_ToadLeaf/Meshes/Parts/F_MED_ToadLeaf_FaceAcc_AnimBP
target_skeleton=&/F_MED_ToadLeaf_FaceAcc_Skeleton
graphs [
 AnimGraph {nodes [
   n1 = Control Rig!(Source=Copy Pose From Mesh() {bUseAttachedParent=true}, Alpha, PlayingEmote:bool=Get bIsPlayingEmote()) -> (Pose) {
    alphaBoolBlend {blendOption=Linear}
    controlRigAssetReference {blueprintRigClass=&/F_MED_ToadLeaf_FaceAcc_CtrlRig.F_MED_ToadLeaf_FaceAcc_CtrlRig_C}
   }
   n2 = Anim Dynamics - Bone: Hair_Side_L_01(ComponentPose=Local To Component(LocalPose=@n1), Alpha=1, ExternalForce="0,0,0") {
    angularDampingOverride=1
    angularSpringConstant=500
    bAngularSpring=true
    bOverrideAngularDamping=true
    bUseGravityOverride=true
    simulationSpace=RootRelative
    alphaBoolBlend {blendOption=Linear}
    boundBone {boneName=Hair_Side_L_01}
    componentAppliedLinearAccClamp=<3000 3000 3000>
    componentLinearAccScale=<0.5 0.5 0.5>
    componentLinearVelScale=<0.5 0.5 0.5>
    gravityOverride {z=-980}
    physicsBodyDefinitions [{
      boundBone {boneName=Hair_Side_L_01}
      boxExtents=<10 10 10>
      constraintSetup {angularLimitsMax=<0 0 5>, angularLimitsMin=<0 -9 -5>, angularTarget=<1 0 0>}
      localJointOffset=<1 0 0>
     }]
   }
   n3 = Anim Dynamics - Bone: Hair_Side_L_02(ComponentPose=@n2, Alpha=1, ExternalForce="0,0,0") {
    angularDampingOverride=1
    angularSpringConstant=500
    bAngularSpring=true
    bOverrideAngularDamping=true
    bUseGravityOverride=true
    simulationSpace=RootRelative
    alphaBoolBlend {blendOption=Linear}
    boundBone {boneName=Hair_Side_L_02}
    componentAppliedLinearAccClamp=<3000 3000 3000>
    componentLinearAccScale=<0.5 0.5 0.5>
    componentLinearVelScale=<0.5 0.5 0.5>
    gravityOverride {z=-980}
    physicsBodyDefinitions [{
      boundBone {boneName=Hair_Side_L_02}
      boxExtents=<10 10 10>
      constraintSetup {angularLimitsMax=<0 0 5>, angularLimitsMin=<0 -9 -5>, angularTarget=<1 0 0>}
      localJointOffset=<1 0 0>
     }]
   }
   n4 = Anim Dynamics - Bone: Hair_Bk_L_01(ComponentPose=@n3, Alpha=1, ExternalForce="0,0,0") {
    angularDampingOverride=1
    angularSpringConstant=500
    bAngularSpring=true
    bOverrideAngularDamping=true
    bUseGravityOverride=true
    simulationSpace=RootRelative
    alphaBoolBlend {blendOption=Linear}
    boundBone {boneName=Hair_Bk_L_01}
    componentAppliedLinearAccClamp=<3000 3000 3000>
    componentLinearAccScale=<0.5 0.5 0.5>
    componentLinearVelScale=<0.5 0.5 0.5>
    gravityOverride {z=-980}
    physicsBodyDefinitions [{
      boundBone {boneName=Hair_Bk_L_01}
      boxExtents=<10 10 10>
      constraintSetup {angularLimitsMax=<0 0 5>, angularLimitsMin=<0 -9 -5>, angularTarget=<1 0 0>}
      localJointOffset=<1 0 0>
     }]
   }
   n5 = Anim Dynamics - Bone: Hair_Bk_L_02(ComponentPose=@n4, Alpha=1, ExternalForce="0,0,0") {
    angularDampingOverride=1
    angularSpringConstant=500
    bAngularSpring=true
    bOverrideAngularDamping=true
    bUseGravityOverride=true
    simulationSpace=RootRelative
    alphaBoolBlend {blendOption=Linear}
    boundBone {boneName=Hair_Bk_L_02}
    componentAppliedLinearAccClamp=<3000 3000 3000>
    componentLinearAccScale=<0.5 0.5 0.5>
    componentLinearVelScale=<0.5 0.5 0.5>
    gravityOverride {z=-980}
    physicsBodyDefinitions [{
      boundBone {boneName=Hair_Bk_L_02}
      boxExtents=<10 10 10>
      constraintSetup {angularLimitsMax=<0 0 5>, angularLimitsMin=<0 -9 -5>, angularTarget=<1 0 0>}
      localJointOffset=<1 0 0>
     }]
   }
   n6 = Anim Dynamics - Bone: Hair_Bk_L_03(ComponentPose=@n5, Alpha=1, ExternalForce="0,0,0") {
    angularDampingOverride=1
    angularSpringConstant=500
    bAngularSpring=true
    bOverrideAngularDamping=true
    bUseGravityOverride=true
    simulationSpace=RootRelative
    alphaBoolBlend {blendOption=Linear}
    boundBone {boneName=Hair_Bk_L_03}
    componentAppliedLinearAccClamp=<3000 3000 3000>
    componentLinearAccScale=<0.5 0.5 0.5>
    componentLinearVelScale=<0.5 0.5 0.5>
    gravityOverride {z=-980}
    physicsBodyDefinitions [{
      boundBone {boneName=Hair_Bk_L_03}
      boxExtents=<10 10 10>
      constraintSetup {angularLimitsMax=<0 0 5>, angularLimitsMin=<0 -9 -5>, angularTarget=<1 0 0>}
      localJointOffset=<1 0 0>
     }]
   }
   n7 = Anim Dynamics - Bone: Hair_Bk_C_01(ComponentPose=@n6, Alpha=1, ExternalForce="0,0,0") {
    angularDampingOverride=1
    angularSpringConstant=500
    bAngularSpring=true
    bOverrideAngularDamping=true
    bUseGravityOverride=true
    simulationSpace=RootRelative
    alphaBoolBlend {blendOption=Linear}
    boundBone {boneName=Hair_Bk_C_01}
    componentAppliedLinearAccClamp=<3000 3000 3000>
    componentLinearAccScale=<0.5 0.5 0.5>
    componentLinearVelScale=<0.5 0.5 0.5>
    gravityOverride {z=-980}
    physicsBodyDefinitions [{
      boundBone {boneName=Hair_Bk_C_01}
      boxExtents=<10 10 10>
      constraintSetup {angularLimitsMax=<0 0 5>, angularLimitsMin=<0 -9 -5>, angularTarget=<1 0 0>}
      localJointOffset=<1 0 0>
     }]
   }
   n8 = Anim Dynamics - Bone: Hair_Bk_C_02(ComponentPose=@n7, Alpha=1, ExternalForce="0,0,0") {
    angularDampingOverride=1
    angularSpringConstant=500
    bAngularSpring=true
    bOverrideAngularDamping=true
    bUseGravityOverride=true
    simulationSpace=RootRelative
    alphaBoolBlend {blendOption=Linear}
    boundBone {boneName=Hair_Bk_C_02}
    componentAppliedLinearAccClamp=<3000 3000 3000>
    componentLinearAccScale=<0.5 0.5 0.5>
    componentLinearVelScale=<0.5 0.5 0.5>
    gravityOverride {z=-980}
    physicsBodyDefinitions [{
      boundBone {boneName=Hair_Bk_C_02}
      boxExtents=<10 10 10>
      constraintSetup {angularLimitsMax=<0 0 5>, angularLimitsMin=<0 -9 -5>, angularTarget=<1 0 0>}
      localJointOffset=<1 0 0>
     }]
   }
   n9 = Anim Dynamics - Bone: Hair_Bk_R_01(ComponentPose=@n8, Alpha=1, ExternalForce="0,0,0") {
    angularDampingOverride=1
    angularSpringConstant=500
    bAngularSpring=true
    bOverrideAngularDamping=true
    bUseGravityOverride=true
    simulationSpace=RootRelative
    alphaBoolBlend {blendOption=Linear}
    boundBone {boneName=Hair_Bk_R_01}
    componentAppliedLinearAccClamp=<3000 3000 3000>
    componentLinearAccScale=<0.5 0.5 0.5>
    componentLinearVelScale=<0.5 0.5 0.5>
    gravityOverride {z=-980}
    physicsBodyDefinitions [{
      boundBone {boneName=Hair_Bk_R_01}
      boxExtents=<10 10 10>
      constraintSetup {angularLimitsMax=<0 0 5>, angularLimitsMin=<0 -9 -5>, angularTarget=<1 0 0>}
      localJointOffset=<1 0 0>
     }]
   }
   n10 = Anim Dynamics - Bone: Hair_Bk_R_02(ComponentPose=@n9, Alpha=1, ExternalForce="0,0,0") {
    angularDampingOverride=1
    angularSpringConstant=500
    bAngularSpring=true
    bOverrideAngularDamping=true
    bUseGravityOverride=true
    simulationSpace=RootRelative
    alphaBoolBlend {blendOption=Linear}
    boundBone {boneName=Hair_Bk_R_02}
    componentAppliedLinearAccClamp=<3000 3000 3000>
    componentLinearAccScale=<0.5 0.5 0.5>
    componentLinearVelScale=<0.5 0.5 0.5>
    gravityOverride {z=-980}
    physicsBodyDefinitions [{
      boundBone {boneName=Hair_Bk_R_02}
      boxExtents=<10 10 10>
      constraintSetup {angularLimitsMax=<0 0 5>, angularLimitsMin=<0 -9 -5>, angularTarget=<1 0 0>}
      localJointOffset=<1 0 0>
     }]
   }
   n11 = Anim Dynamics - Bone: Hair_Side_R_01(ComponentPose=@n10, Alpha=1, ExternalForce="0,0,0") {
    angularDampingOverride=1
    angularSpringConstant=500
    bAngularSpring=true
    bOverrideAngularDamping=true
    bUseGravityOverride=true
    simulationSpace=RootRelative
    alphaBoolBlend {blendOption=Linear}
    boundBone {boneName=Hair_Side_R_01}
    componentAppliedLinearAccClamp=<3000 3000 3000>
    componentLinearAccScale=<0.5 0.5 0.5>
    componentLinearVelScale=<0.5 0.5 0.5>
    gravityOverride {z=-980}
    physicsBodyDefinitions [{
      boundBone {boneName=Hair_Side_R_01}
      boxExtents=<10 10 10>
      constraintSetup {angularLimitsMax=<0 0 5>, angularLimitsMin=<0 -9 -5>, angularTarget=<1 0 0>}
      localJointOffset=<1 0 0>
     }]
   }
   n12 = Anim Dynamics - Bone: Hair_Side_R_02(ComponentPose=@n11, Alpha=1, ExternalForce="0,0,0") {
    angularDampingOverride=1
    angularSpringConstant=500
    bAngularSpring=true
    bOverrideAngularDamping=true
    bUseGravityOverride=true
    simulationSpace=RootRelative
    alphaBoolBlend {blendOption=Linear}
    boundBone {boneName=Hair_Side_R_02}
    componentAppliedLinearAccClamp=<3000 3000 3000>
    componentLinearAccScale=<0.5 0.5 0.5>
    componentLinearVelScale=<0.5 0.5 0.5>
    gravityOverride {z=-980}
    physicsBodyDefinitions [{
      boundBone {boneName=Hair_Side_R_02}
      boxExtents=<10 10 10>
      constraintSetup {angularLimitsMax=<0 0 5>, angularLimitsMin=<0 -9 -5>, angularTarget=<1 0 0>}
      localJointOffset=<1 0 0>
     }]
   }
   n13 = Anim Dynamics - Bone: Hair_Side_R_03(ComponentPose=@n12, Alpha=1, ExternalForce="0,0,0") {
    angularDampingOverride=1
    angularSpringConstant=500
    bAngularSpring=true
    bOverrideAngularDamping=true
    bUseGravityOverride=true
    simulationSpace=RootRelative
    alphaBoolBlend {blendOption=Linear}
    boundBone {boneName=Hair_Side_R_03}
    componentAppliedLinearAccClamp=<3000 3000 3000>
    componentLinearAccScale=<0.5 0.5 0.5>
    componentLinearVelScale=<0.5 0.5 0.5>
    gravityOverride {z=-980}
    physicsBodyDefinitions [{
      boundBone {boneName=Hair_Side_R_03}
      boxExtents=<10 10 10>
      constraintSetup {angularLimitsMax=<0 0 5>, angularLimitsMin=<0 -9 -5>, angularTarget=<1 0 0>}
      localJointOffset=<1 0 0>
     }]
   }
   n14 = Anim Dynamics - Bone: Hair_Front_01(ComponentPose=@n13, Alpha=1, ExternalForce="0,0,0") {
    angularDampingOverride=1
    angularSpringConstant=500
    bAngularSpring=true
    bOverrideAngularDamping=true
    bUseGravityOverride=true
    simulationSpace=RootRelative
    alphaBoolBlend {blendOption=Linear}
    boundBone {boneName=Hair_Front_01}
    componentAppliedLinearAccClamp=<3000 3000 3000>
    componentLinearAccScale=<0.5 0.5 0.5>
    componentLinearVelScale=<0.5 0.5 0.5>
    gravityOverride {z=-980}
    physicsBodyDefinitions [{
      boundBone {boneName=Hair_Front_01}
      boxExtents=<10 10 10>
      constraintSetup {angularLimitsMax=<0 0 5>, angularLimitsMin=<0 -20 -5>, angularTarget=<1 0 0>}
      localJointOffset=<1 0 0>
     }]
   }
   n15 = Anim Dynamics - Bone: Hair_Side_R_01(ComponentPose=@n14, Alpha=1, ExternalForce="0,0,0") {
    angularDampingOverride=1
    angularSpringConstant=500
    bAngularSpring=true
    bOverrideAngularDamping=true
    bUseGravityOverride=true
    simulationSpace=RootRelative
    alphaBoolBlend {blendOption=Linear}
    boundBone {boneName=Hair_Side_R_01}
    componentAppliedLinearAccClamp=<3000 3000 3000>
    componentLinearAccScale=<0.5 0.5 0.5>
    componentLinearVelScale=<0.5 0.5 0.5>
    gravityOverride {z=-980}
    physicsBodyDefinitions [{
      boundBone {boneName=Hair_Side_R_01}
      boxExtents=<10 10 10>
      constraintSetup {angularLimitsMax=<0 0 5>, angularLimitsMin=<0 -20 -5>, angularTarget=<1 0 0>}
      localJointOffset=<1 0 0>
     }]
   }
   n16 = Anim Dynamics - Bone: Hair_Side_R_02(ComponentPose=@n15, Alpha=1, ExternalForce="0,0,0") {
    angularDampingOverride=1
    angularSpringConstant=500
    bAngularSpring=true
    bOverrideAngularDamping=true
    bUseGravityOverride=true
    simulationSpace=RootRelative
    alphaBoolBlend {blendOption=Linear}
    boundBone {boneName=Hair_Side_R_02}
    componentAppliedLinearAccClamp=<3000 3000 3000>
    componentLinearAccScale=<0.5 0.5 0.5>
    componentLinearVelScale=<0.5 0.5 0.5>
    gravityOverride {z=-980}
    physicsBodyDefinitions [{
      boundBone {boneName=Hair_Side_R_02}
      boxExtents=<10 10 10>
      constraintSetup {angularLimitsMax=<0 0 5>, angularLimitsMin=<0 -20 -5>, angularTarget=<1 0 0>}
      localJointOffset=<1 0 0>
     }]
   }
   n17 = Anim Dynamics - Bone: Hair_Front_04(ComponentPose=@n16, Alpha=1, ExternalForce="0,0,0") {
    angularDampingOverride=1
    angularSpringConstant=500
    bAngularSpring=true
    bOverrideAngularDamping=true
    bUseGravityOverride=true
    simulationSpace=RootRelative
    alphaBoolBlend {blendOption=Linear}
    boundBone {boneName=Hair_Front_04}
    componentAppliedLinearAccClamp=<3000 3000 3000>
    componentLinearAccScale=<0.5 0.5 0.5>
    componentLinearVelScale=<0.5 0.5 0.5>
    gravityOverride {z=-980}
    physicsBodyDefinitions [{
      boundBone {boneName=Hair_Front_04}
      boxExtents=<10 10 10>
      constraintSetup {angularLimitsMax=<0 0 5>, angularLimitsMin=<0 -20 -5>, angularTarget=<1 0 0>}
      localJointOffset=<1 0 0>
     }]
   }
   n18 = Anim Dynamics - Bone: Hair_Front_05(ComponentPose=@n17, Alpha=1, ExternalForce="0,0,0") {
    angularDampingOverride=1
    angularSpringConstant=500
    bAngularSpring=true
    bOverrideAngularDamping=true
    bUseGravityOverride=true
    simulationSpace=RootRelative
    alphaBoolBlend {blendOption=Linear}
    boundBone {boneName=Hair_Front_05}
    componentAppliedLinearAccClamp=<3000 3000 3000>
    componentLinearAccScale=<0.5 0.5 0.5>
    componentLinearVelScale=<0.5 0.5 0.5>
    gravityOverride {z=-980}
    physicsBodyDefinitions [{
      boundBone {boneName=Hair_Front_05}
      boxExtents=<10 10 10>
      constraintSetup {angularLimitsMax=<0 0 5>, angularLimitsMin=<0 -20 -5>, angularTarget=<1 0 0>}
      localJointOffset=<1 0 0>
     }]
   }
   Output Pose(Result=Component To Local(ComponentPose=@n18)) {layerGroup=None}
  ]}
 EventGraph {nodes [Event BlueprintUpdateAnimation()]}
]
variables []
